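# Control Rig for the F_MED_BlueMonday female character body that procedurally drives outfit and deform bones: skirt, choker, belts, straps, pouch, zipper, necklace, glutes and groin.
# Weighted parent constraints set each bone's base motion from body bones.
# Corrective offsets layer on top, weighted by remapped swing angles, spherical pose readers and inverse-distance readers on thighs, head and spine.
# An RBAN-enabled flag gates the necklace corrective, and thigh rotation also writes a curve.
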ControlRigBlueprint /BRCosmetics/Characters/Player/Female/Medium/Bodies/F_MED_BlueMonday/Meshes/F_MED_BlueMonday_CtrlRig
graphs [
 CollapseNode_ContainedGraph {nodes [
   n1 = RigVMFunctionEntryNode()
   &RigUnit_ParentConstraint(
    AdvancedSettings=(RotationOrderForFilter=XZY)
    Child=(Type=Bone,Name="base_strapArm_1_l")
    Parents=((Item=(Type=Bone,Name="deform_upperarm_fr_l"),Weight=0.017),(Item=(Type=Bone,Name="upperarm_twist_02_l"),Weight=0.2822),(Item=(Type=Bone,Name="upperarm_twist_01_l"),Weight=0.7009))
    Weight=1
    ExecutePin=@n1.ExecutePin
   )
   > &RigUnit_ParentConstraint(
    AdvancedSettings=(RotationOrderForFilter=XZY)
    Child=(Type=Bone,Name="base_strapArm_3_l")
    Parents="((Item=(Type=Bone,Name=\"upperarm_twist_01_l\"),Weight=0.0183),(Item=(Type=Bone,Name=\"deform_elbow_in_l\"),Weight=0.0492),(Item=(Type=Bone,Name=\"lowerarm_l\"),Weight=0.0512),(Item=(Type=Bone,Name=\"upperarm_twist_02_l\"),Weight=0.3796),(Item=(Type=Bone,Name=\"upp ...(truncated)"
    Weight=1
   )
   > &RigUnit_ParentConstraint(
    AdvancedSettings=(RotationOrderForFilter=XZY)
    Child=(Type=Bone,Name="base_strapArm_2_l")
    Parents=((Item=(Type=Bone,Name="upperarm_l"),Weight=0.0848),(Item=(Type=Bone,Name="upperarm_twist_01_l"),Weight=0.2998),(Item=(Type=Bone,Name="upperarm_twist_02_l"),Weight=0.6154))
    Weight=1
   )
   > &RigUnit_ParentConstraint(
    AdvancedSettings=(RotationOrderForFilter=XZY)
    Child=(Type=Bone,Name="base_strapArm_2_r")
    Parents=((Item=(Type=Bone,Name="upperarm_twist_01_r"),Weight=0.095),(Item=(Type=Bone,Name="upperarm_r"),Weight=0.2888),(Item=(Type=Bone,Name="upperarm_twist_02_r"),Weight=0.6162))
    Weight=1
   )
   > &RigUnit_ParentConstraint(
    AdvancedSettings=(RotationOrderForFilter=XZY)
    Child=(Type=Bone,Name="base_strapArm_1_r")
    Parents=((Item=(Type=Bone,Name="deform_upperarm_fr_r"),Weight=0.017),(Item=(Type=Bone,Name="upperarm_twist_02_r"),Weight=0.2822),(Item=(Type=Bone,Name="upperarm_twist_01_r"),Weight=0.7009))
    Weight=1
   )
   > &RigUnit_ParentConstraint(
    AdvancedSettings=(RotationOrderForFilter=XZY)
    Child=(Type=Bone,Name="kin_legPouch")
    Parents=((Item=(Type=Bone,Name="thigh_twist_01_l"),Weight=0.4954),(Item=(Type=Bone,Name="thigh_l"),Weight=0.5046))
    Weight=1
   )
   > GetAlphaFromQuatDelta = RigVMFunctionReferenceNode(
    Rotation Order=ZYX
    absoluteValue=True
    child=(Type=Bone,Name="thigh_l")
    onlyTwist=True
    outputMinMax=(X=-1,Y=1)
    parent=(Type=Null,Name="thigh_l_null")
    premult=1
    rotationMinMax=(X=-40,Y=40)
    twistAxis=(X=1,Y=0,Z=0)
   )
   > &RigUnit_ModifyTransforms(
    ItemToModify=((Item=(Type=Bone,Name="kin_legPouch"),Transform=(Rotation=(X=0,Y=-0.0397,Z=0,W=0.9992),Translation=(X=0.5,Y=0,Z=0),Scale3D=(X=1,Y=1,Z=1))))
    WeightMinimum=0
    Weight=@GetAlphaFromQuatDelta.alpha
   )
   Remap = &RigVMFunction_MathDoubleRemap(SourceMaximum=-90, SourceMinimum=-5, TargetMaximum=1, TargetMinimum=0, Value=RigVMVariableNode() {variable=l thighLocal rz float}.Value)
   > &RigUnit_ModifyTransforms(
    ItemToModify=((Item=(Type=Bone,Name="kin_legPouch"),Transform=(Rotation=(X=0,Y=0,Z=0,W=1),Translation=(X=0,Y=0,Z=-0.45),Scale3D=(X=1,Y=1,Z=1))))
    WeightMinimum=0
    Weight=@Remap.Result
   )
   > &RigUnit_ParentConstraint(
    AdvancedSettings=(RotationOrderForFilter=XZY)
    Child=(Type=Bone,Name="base_vestZipper_c")
    Parents=((Item=(Type=Bone,Name="spine_01"),Weight=0.0756),(Item=(Type=Bone,Name="spine_04"),Weight=0.1236),(Item=(Type=Bone,Name="spine_02"),Weight=0.3962),(Item=(Type=Bone,Name="spine_03"),Weight=0.4046))
    Weight=1
   )
   > &RigUnit_ParentConstraint(
    AdvancedSettings=(RotationOrderForFilter=XZY)
    Child=(Type=Bone,Name="base_vestZipper_l")
    Parents=((Item=(Type=Bone,Name="spine_01"),Weight=0.0278),(Item=(Type=Bone,Name="spine_02"),Weight=0.2456),(Item=(Type=Bone,Name="spine_04"),Weight=0.2935),(Item=(Type=Bone,Name="spine_03"),Weight=0.4331))
    Weight=1
   )
   > n2 = &RigUnit_ParentConstraint(
    AdvancedSettings=(RotationOrderForFilter=XZY)
    Child=(Type=Bone,Name="base_vestZipper_r")
    Parents=((Item=(Type=Bone,Name="spine_01"),Weight=0.0278),(Item=(Type=Bone,Name="spine_02"),Weight=0.2456),(Item=(Type=Bone,Name="spine_04"),Weight=0.2935),(Item=(Type=Bone,Name="spine_03"),Weight=0.4331))
    Weight=1
   )
   RigVMFunctionReturnNode(ExecutePin=@n2.ExecutePin)
  ]}
 CollapseNode_ContainedGraph {nodes [
   n1 = RigVMFunctionEntryNode()
   RigVMFunction_Sequence = &RigVMFunction_Sequence(ExecuteContext=@n1.ExecutePin)
   &RigUnit_ParentConstraint(
    AdvancedSettings=(RotationOrderForFilter=XZY)
    Child=(Type=Bone,Name="kin_chocker_1")
    Parents=((Item=(Type=Bone,Name="FACIAL_C_Jawline"),Weight=0.1524),(Item=(Type=Bone,Name="spine_05"),Weight=0.1772),(Item=(Type=Bone,Name="neck_02"),Weight=0.2601),(Item=(Type=Bone,Name="neck_01"),Weight=0.4103))
    Weight=1
    ExecutePin=@RigVMFunction_Sequence.A
   )
   > &RigUnit_ParentConstraint(
    AdvancedSettings=(RotationOrderForFilter=XZY)
    Child=(Type=Bone,Name="kin_chocker_2")
    Parents="((Item=(Type=Bone,Name=\"head\"),Weight=0.0144),(Item=(Type=Bone,Name=\"clavicle_l\"),Weight=0.02),(Item=(Type=Bone,Name=\"FACIAL_L_UnderChin\"),Weight=0.021),(Item=(Type=Bone,Name=\"FACIAL_C_Jawline\"),Weight=0.0366),(Item=(Type=Bone,Name=\"FACIAL_L_Jawline\"),Weig ...(truncated)"
    Weight=1
   )
   > &RigUnit_ParentConstraint(
    AdvancedSettings=(RotationOrderForFilter=XZY)
    Child=(Type=Bone,Name="kin_chocker_3")
    Parents=((Item=(Type=Bone,Name="clavicle_l"),Weight=0.0376),(Item=(Type=Bone,Name="neck_02"),Weight=0.2458),(Item=(Type=Bone,Name="neck_01"),Weight=0.3476),(Item=(Type=Bone,Name="spine_05"),Weight=0.369))
    Weight=1
   )
   > &RigUnit_ParentConstraint(
    AdvancedSettings=(RotationOrderForFilter=XZY)
    Child=(Type=Bone,Name="kin_chocker_4")
    Parents=((Item=(Type=Bone,Name="head"),Weight=0.0164),(Item=(Type=Bone,Name="spine_05"),Weight=0.2589),(Item=(Type=Bone,Name="neck_02"),Weight=0.2729),(Item=(Type=Bone,Name="neck_01"),Weight=0.4519))
    Weight=1
   )
   > &RigUnit_ParentConstraint(
    AdvancedSettings=(RotationOrderForFilter=XZY)
    Child=(Type=Bone,Name="kin_chocker_5")
    Parents=((Item=(Type=Bone,Name="clavicle_r"),Weight=0.0451),(Item=(Type=Bone,Name="neck_02"),Weight=0.2309),(Item=(Type=Bone,Name="neck_01"),Weight=0.3343),(Item=(Type=Bone,Name="spine_05"),Weight=0.3897))
    Weight=1
   )
   > &RigUnit_ParentConstraint(
    AdvancedSettings=(RotationOrderForFilter=XZY)
    Child=(Type=Bone,Name="kin_chocker_6")
    Parents="((Item=(Type=Bone,Name=\"head\"),Weight=0.0142),(Item=(Type=Bone,Name=\"clavicle_r\"),Weight=0.0205),(Item=(Type=Bone,Name=\"FACIAL_R_UnderChin\"),Weight=0.0207),(Item=(Type=Bone,Name=\"FACIAL_C_Jawline\"),Weight=0.0362),(Item=(Type=Bone,Name=\"FACIAL_R_Jawline\"),W ...(truncated)"
    Weight=1
   )
   RigVMFunction_MathQuaternionSwingTwist = &RigVMFunction_MathQuaternionSwingTwist(
    TwistAxis=(X=1,Y=0,Z=0)
    Input=&RigUnit_GetRelativeTransformForItem(Child=(Type=Bone,Name="head"), Parent=(Type=Null,Name="head_sp_null"), bChildInitial=False, bParentInitial=False).RelativeTransform.Rotation
   )
   RigVMFunction_MathQuaternionToEuler = &RigVMFunction_MathQuaternionToEuler(RotationOrder=ZYX, Value=@RigVMFunction_MathQuaternionSwingTwist.Swing)
   Remap = &RigVMFunction_MathDoubleRemap(SourceMaximum=-60, SourceMinimum=0, TargetMaximum=1, TargetMinimum=0, Value=@RigVMFunction_MathQuaternionToEuler.Result.Z)
   > &RigUnit_ModifyTransforms(
    ItemToModify="((Item=(Type=Bone,Name=\"kin_chocker_4\"),Transform=(Rotation=(X=0,Y=0,Z=0,W=1),Translation=(X=0,Y=-0.8,Z=0.4),Scale3D=(X=1,Y=1,Z=1))),(Item=(Type=Bone,Name=\"kin_chocker_3\"),Transform=(Rotation=(X=0,Y=0,Z=0,W=1),Translation=(X=0,Y=-0.33,Z=0.5),Scale3D=(X=1,Y ...(truncated)"
    WeightMinimum=0
    Weight=@Remap.Result
   )
   > &RigUnit_ModifyTransforms(
    ItemToModify="((Item=(Type=Bone,Name=\"kin_chocker_5\"),Transform=(Rotation=(X=0,Y=0,Z=0,W=1),Translation=(X=0,Y=-0.5,Z=0.5),Scale3D=(X=1,Y=1,Z=1))),(Item=(Type=Bone,Name=\"kin_chocker_6\"),Transform=(Rotation=(X=0,Y=0,Z=0,W=1),Translation=(X=0,Y=-0.64,Z=0.5),Scale3D=(X=1,Y ...(truncated)"
    WeightMaximum=0
    WeightMinimum=1
    Weight=&RigVMFunction_MathDoubleRemap(SourceMaximum=-40, SourceMinimum=0, TargetMaximum=1, TargetMinimum=0, Value=@RigVMFunction_MathQuaternionToEuler.Result.Y).Result
   )
   > &RigUnit_ModifyTransforms(
    ItemToModify="((Item=(Type=Bone,Name=\"kin_chocker_3\"),Transform=(Rotation=(X=0,Y=0,Z=0,W=1),Translation=(X=0,Y=-0.5,Z=0.5),Scale3D=(X=1,Y=1,Z=1))),(Item=(Type=Bone,Name=\"kin_chocker_2\"),Transform=(Rotation=(X=0,Y=0,Z=0,W=1),Translation=(X=0,Y=-0.4,Z=0.5),Scale3D=(X=1,Y= ...(truncated)"
    WeightMinimum=0
    Weight=&RigVMFunction_MathDoubleRemap(SourceMaximum=40, SourceMinimum=0, TargetMaximum=1, TargetMinimum=0, Value=@RigVMFunction_MathQuaternionToEuler.Result.Y).Result
   )
   > &RigUnit_ModifyTransforms(
    ItemToModify="((Item=(Type=Bone,Name=\"kin_chocker_6\"),Transform=(Rotation=(X=0,Y=0,Z=0,W=1),Translation=(X=0.2,Y=0,Z=0.5),Scale3D=(X=1,Y=1,Z=1))),(Item=(Type=Bone,Name=\"kin_chocker_2\"),Transform=(Rotation=(X=0,Y=0,Z=0,W=1),Translation=(X=-0.2,Y=0,Z=0.5),Scale3D=(X=1,Y=1 ...(truncated)"
    WeightMinimum=0
    Weight=&RigVMFunction_MathDoubleRemap(SourceMaximum=30, SourceMinimum=0, TargetMaximum=1, TargetMinimum=0, Value=@RigVMFunction_MathQuaternionToEuler.Result.Z).Result
   )
   > GetAlphaFromQuatDelta = RigVMFunctionReferenceNode(
    Rotation Order=ZYX
    child=(Type=Bone,Name="head")
    outputMinMax=(X=0,Y=1)
    parent=(Type=Null,Name="head_null")
    premult=1
    rotationMinMax=(X=0,Y=-30)
    triggerAxis=Z
    twistAxis=(X=1,Y=0,Z=0)
   )
   > &RigUnit_ModifyTransforms(
    ItemToModify="((Item=(Type=Bone,Name=\"kin_chocker_1\"),Transform=(Rotation=(X=0,Y=0,Z=0,W=1),Translation=(X=0,Y=0,Z=0.5),Scale3D=(X=1,Y=1,Z=1))),(Item=(Type=Bone,Name=\"kin_chocker_2\"),Transform=(Rotation=(X=0,Y=0,Z=0,W=1),Translation=(X=0,Y=0,Z=0.3),Scale3D=(X=1,Y=1,Z=1) ...(truncated)"
    WeightMinimum=0
    Weight=@GetAlphaFromQuatDelta.alpha
   )
   > n2 = &RigUnit_SphericalPoseReader(
    ActiveRegionScaleFactors=(PositiveWidth=1,NegativeWidth=1,PositiveHeight=1,NegativeHeight=1)
    ActiveRegionSize=0.02
    Debug=(bDrawDebug=false,bDraw2D=False,bDrawLocalAxes=False,DebugScale=25,DebugSegments=20,DebugThickness=0.25)
    DriverAxis=(X=1,Y=0,Z=0)
    DriverItem=(Type=Bone,Name="head")
    FalloffRegionScaleFactors=(PositiveWidth=1,NegativeWidth=1,PositiveHeight=1,NegativeHeight=1)
    FalloffSize=0.1
    FlipHeightScaling=False
    FlipWidthScaling=False
    OptionalParentItem=(Type=Bone,Name="None")
    RotationOffset=(X=-8,Y=-87,Z=7)
   )
   > &RigUnit_ModifyTransforms(
    ItemToModify="((Item=(Type=Bone,Name=\"kin_chocker_2\"),Transform=(Rotation=(X=0,Y=0,Z=0,W=1),Translation=(X=0,Y=0,Z=0.5),Scale3D=(X=1,Y=1,Z=1))),(Item=(Type=Bone,Name=\"kin_chocker_1\"),Transform=(Rotation=(X=0,Y=0,Z=0,W=1),Translation=(X=0,Y=0,Z=0.2),Scale3D=(X=1,Y=1,Z=1) ...(truncated)"
    WeightMinimum=0
    Weight=@n2.OutputParam
   )
   RigVMFunctionReturnNode(ExecutePin=@RigVMFunction_Sequence.B)
  ]}
 CollapseNode_ContainedGraph {nodes [
   n1 = RigVMFunctionEntryNode()
   RigVMFunction_MathQuaternionSwingTwist = &RigVMFunction_MathQuaternionSwingTwist(
    TwistAxis=(X=1,Y=0,Z=0)
    Input=&RigUnit_GetRelativeTransformForItem(Child=(Type=Bone,Name="thigh_l"), Parent=(Type=Null,Name="thigh_l_null"), bChildInitial=False, bParentInitial=False).RelativeTransform.Rotation
   )
   RigVMFunction_MathQuaternionToEuler = &RigVMFunction_MathQuaternionToEuler(RotationOrder=ZYX, Value=@RigVMFunction_MathQuaternionSwingTwist.Swing)
   RigVMVariableNode(ExecuteContext=@n1.ExecutePin, Value=@RigVMFunction_MathQuaternionToEuler.Result.Y) {variable=l thighLocal ry float}
   > RigVMVariableNode(Value=@RigVMFunction_MathQuaternionToEuler.Result.Z) {variable=l thighLocal rz float}
   n2 = &RigVMFunction_MathQuaternionToEuler(
    RotationOrder=ZYX
    Value=&RigVMFunction_MathQuaternionSwingTwist(
     TwistAxis=(X=1,Y=0,Z=0)
     Input=&RigUnit_GetRelativeTransformForItem(Child=(Type=Bone,Name="thigh_r"), Parent=(Type=Null,Name="thigh_r_null"), bChildInitial=False, bParentInitial=False).RelativeTransform.Rotation
    ).Swing
   )
   > RigVMVariableNode(Value=@n2.Result.Y) {variable=r thighLocal ry float}
   > n3 = RigVMVariableNode(Value=@n2.Result.Z) {variable=r thighLocal rz float}
   constraints = RigVMCollapseNode(ExecutePin=@n3.ExecuteContext)
   RigVMFunction_Sequence = RigVMAggregateNode(ExecuteContext=@constraints.ExecutePin)
   n4 = RigVMVariableNode() {variable=l thighLocal ry float}
   &RigUnit_ModifyTransforms(
    ItemToModify="((Item=(Type=Bone,Name=\"base_skirt_1_s_l\"),Transform=(Rotation=(X=0,Y=0.2208,Z=0,W=-0.9753),Translation=(X=-1.1094,Y=0,Z=0.1645),Scale3D=(X=1,Y=1,Z=1))),(Item=(Type=Bone,Name=\"base_skirt_1_f_l\"),Transform=(Rotation=(X=-0.037,Y=-0.1431,Z=-0.2473,W=0.9576),T ...(truncated)"
    WeightMinimum=0
    Weight=/Script/RigVM.RigVMFunction_MathDoublePow(
     B=1.5
     A=&RigVMFunction_MathDoubleRemap(SourceMaximum=-55, SourceMinimum=0, TargetMaximum=1, TargetMinimum=0, Value=@n4.Value).Result
    ) {is=struct RigVMUnitNode, template="Power::Execute(in A,in B,out Result)"}.Result
    ExecutePin=@RigVMFunction_Sequence.B
   )
   n5 = RigVMVariableNode() {variable=r thighLocal ry float}
   > &RigUnit_ModifyTransforms(
    ItemToModify="((Item=(Type=Bone,Name=\"base_skirt_1_s_r\"),Transform=(Rotation=(X=0,Y=0.2208,Z=0,W=-0.9753),Translation=(X=-1.1094,Y=0,Z=0.1645),Scale3D=(X=1,Y=1,Z=1))),(Item=(Type=Bone,Name=\"base_skirt_1_f_r\"),Transform=(Rotation=(X=0.037,Y=-0.1431,Z=0.2473,W=0.9576),Tra ...(truncated)"
    WeightMinimum=0
    Weight=&RigVMFunction_MathDoublePow(
     B=1.5
     A=&RigVMFunction_MathDoubleRemap(SourceMaximum=-55, SourceMinimum=0, TargetMaximum=1, TargetMinimum=0, Value=@n5.Value).Result
    ) {template="Power::Execute(in A,in B,out Result)"}.Result
   )
   > n6 = &RigUnit_SphericalPoseReader(
    ActiveRegionScaleFactors=(PositiveWidth=1,NegativeWidth=1,PositiveHeight=1,NegativeHeight=1)
    ActiveRegionSize=0.05
    Debug=(bDrawDebug=false,bDraw2D=False,bDrawLocalAxes=False,DebugScale=25,DebugSegments=20,DebugThickness=0.25)
    DriverAxis=(X=-1,Y=0,Z=0)
    DriverItem=(Type=Bone,Name="thigh_l")
    FalloffRegionScaleFactors=(PositiveWidth=1,NegativeWidth=1,PositiveHeight=1,NegativeHeight=0.5)
    FalloffSize=0.25
    FlipHeightScaling=False
    FlipWidthScaling=False
    OptionalParentItem=(Type=Bone,Name="None")
    RotationOffset=(X=120,Y=17,Z=0)
   )
   > &RigUnit_ModifyTransforms(
    ItemToModify="((Item=(Type=Bone,Name=\"base_skirt_1_s_l\"),Transform=(Rotation=(X=0.1488,Y=-0.127,Z=0.1787,W=0.9643),Translation=(X=0,Y=0,Z=0),Scale3D=(X=1,Y=1,Z=1))),(Item=(Type=Bone,Name=\"base_skirt_1_f_c\"),Transform=(Rotation=(X=0,Y=0.1948,Z=0,W=0.9808),Translation=(X= ...(truncated)"
    WeightMinimum=0
    Weight=@n6.OutputParam
   )
   > n7 = &RigUnit_SphericalPoseReader(
    ActiveRegionScaleFactors=(PositiveWidth=1,NegativeWidth=1,PositiveHeight=1,NegativeHeight=1)
    ActiveRegionSize=0.05
    Debug=(bDrawDebug=false,bDraw2D=False,bDrawLocalAxes=False,DebugScale=25,DebugSegments=20,DebugThickness=0.25)
    DriverAxis=(X=1,Y=0,Z=0)
    DriverItem=(Type=Bone,Name="thigh_r")
    FalloffRegionScaleFactors=(PositiveWidth=1,NegativeWidth=1,PositiveHeight=1,NegativeHeight=0.5)
    FalloffSize=0.25
    FlipHeightScaling=False
    FlipWidthScaling=False
    OptionalParentItem=(Type=Bone,Name="None")
    RotationOffset=(X=-60,Y=17,Z=0)
   )
   > &RigUnit_ModifyTransforms(
    ItemToModify="((Item=(Type=Bone,Name=\"base_skirt_1_s_r\"),Transform=(Rotation=(X=-0.1488,Y=-0.127,Z=-0.1787,W=0.9643),Translation=(X=0,Y=0,Z=0),Scale3D=(X=1,Y=1,Z=1))),(Item=(Type=Bone,Name=\"base_skirt_1_f_c\"),Transform=(Rotation=(X=0,Y=0.1948,Z=0,W=0.9808),Translation=( ...(truncated)"
    WeightMinimum=0
    Weight=@n7.OutputParam
   )
   RigVMFunctionReturnNode(ExecutePin=@RigVMFunction_Sequence.E)
   Remap = &RigVMFunction_MathDoubleRemap(SourceMaximum=-100, SourceMinimum=-12, TargetMaximum=1, TargetMinimum=0, Value=RigVMVariableNode() {variable=l thighLocal rz float}.Value)
   n8 = &RigVMFunction_MathDoublePow(B=1.2, A=@Remap.Result) {template="Power::Execute(in A,in B,out Result)"}
   &RigUnit_ModifyTransforms(
    ItemToModify="((Item=(Type=Bone,Name=\"base_skirt_1_f_l\"),Transform=(Rotation=(X=-0.0507,Y=-0.5165,Z=0.1365,W=0.8438),Translation=(X=-2.042,Y=0,Z=1.6046),Scale3D=(X=1,Y=1,Z=1))),(Item=(Type=Bone,Name=\"base_skirt_1_f_c\"),Transform=(Rotation=(X=-0.0468,Y=-0.7581,Z=-0.0547, ...(truncated)"
    WeightMinimum=0
    Weight=@n8.Result
    ExecutePin=@RigVMFunction_Sequence.A
   )
   n9 = &RigVMFunction_MathDoublePow(
    B=1.2
    A=&RigVMFunction_MathDoubleRemap(SourceMaximum=-100, SourceMinimum=0, TargetMaximum=1, TargetMinimum=0, Value=RigVMVariableNode() {variable=r thighLocal rz float}.Value).Result
   ) {template="Power::Execute(in A,in B,out Result)"}
   > &RigUnit_ModifyTransforms(
    ItemToModify="((Item=(Type=Bone,Name=\"base_skirt_1_f_r\"),Transform=(Rotation=(X=-0.0507,Y=0.5165,Z=0.1365,W=-0.8438),Translation=(X=-2.6347,Y=0,Z=1.9134),Scale3D=(X=1,Y=1,Z=1))),(Item=(Type=Bone,Name=\"base_skirt_1_f_c\"),Transform=(Rotation=(X=-0.0468,Y=-0.7581,Z=-0.0547 ...(truncated)"
    WeightMinimum=0
    Weight=@n9.Result
   )
   Multiply = &RigVMFunction_MathDoubleMul(B=@n8.Result, A=@n9.Result)
   Power = &RigVMFunction_MathDoublePow(B=0.7, A=@Multiply.Result) {template="Power::Execute(in A,in B,out Result)"}
   > &RigUnit_ModifyTransforms(
    ItemToModify=((Item=(Type=Bone,Name="base_skirt_1_f_c"),Transform=(Rotation=(X=0,Y=-0.866,Z=0,W=-0.5),Translation=(X=0,Y=0,Z=0),Scale3D=(X=1,Y=1,Z=1))))
    WeightMinimum=0
    Weight=@Power.Result
   )
   n10 = &RigVMFunction_MathDoublePow(
    B=2
    A=&RigVMFunction_MathDoubleRemap(SourceMaximum=50, SourceMinimum=0, TargetMaximum=1, TargetMinimum=0, Value=RigVMVariableNode() {variable=l thighLocal rz float}.Value).Result
   ) {template="Power::Execute(in A,in B,out Result)"}
   &RigUnit_ModifyTransforms(
    ItemToModify="((Item=(Type=Bone,Name=\"base_skirt_1_b_c\"),Transform=(Rotation=(X=0,Y=-0.2302,Z=0,W=0.9732),Translation=(X=0,Y=0,Z=0),Scale3D=(X=1,Y=1,Z=1))),(Item=(Type=Bone,Name=\"base_skirt_1_b_l\"),Transform=(Rotation=(X=0.0354,Y=-0.2218,Z=-0.1177,W=0.9673),Translation= ...(truncated)"
    WeightMinimum=0
    Weight=@n10.Result
    ExecutePin=@RigVMFunction_Sequence.C
   )
   n11 = &RigVMFunction_MathDoublePow(
    B=2
    A=&RigVMFunction_MathDoubleRemap(SourceMaximum=50, SourceMinimum=0, TargetMaximum=1, TargetMinimum=0, Value=RigVMVariableNode() {variable=r thighLocal rz float}.Value).Result
   ) {template="Power::Execute(in A,in B,out Result)"}
   > &RigUnit_ModifyTransforms(
    ItemToModify="((Item=(Type=Bone,Name=\"base_skirt_1_b_c\"),Transform=(Rotation=(X=0,Y=-0.2302,Z=0,W=0.9732),Translation=(X=0,Y=0,Z=0),Scale3D=(X=1,Y=1,Z=1))),(Item=(Type=Bone,Name=\"base_skirt_1_b_r\"),Transform=(Rotation=(X=-0.0354,Y=-0.2218,Z=0.1177,W=0.9673),Translation= ...(truncated)"
    WeightMinimum=0
    Weight=@n11.Result
   )
   > &RigUnit_ModifyTransforms(
    ItemToModify=((Item=(Type=Bone,Name="base_skirt_1_b_c"),Transform=(Rotation=(X=0,Y=0.1305,Z=0,W=0.9914),Translation=(X=0,Y=0,Z=0),Scale3D=(X=1,Y=1,Z=1))))
    WeightMinimum=0
    Weight=&RigVMFunction_MathDoubleMul(B=@n10.Result, A=@n11.Result).Result
   )
   n12 = &RigVMFunction_Sequence(ExecuteContext=@RigVMFunction_Sequence.D)
   InverseDistance_3 = RigVMFunctionReferenceNode(
    child=(Type=Bone,Name="thigh_l")
    drawLines=false
    ease=true
    parent=(Type=Bone,Name="pelvis")
    primaryAxis=(X=-10,Y=0,Z=0)
    remapCurve=(EditorCurveData=(Keys=((InterpMode=RCIM_Cubic,TangentMode=RCTM_User,ArriveTangent=0.4822,LeaveTangent=0.4822),(InterpMode=RCIM_Cubic,TangentMode=RCTM_User,Time=1,Value=1,ArriveTangent=1.6738,LeaveTangent=2.9869))))
    remapMult=2
    targets=((),(X=0,Y=0,Z=-0.7071,W=0.7071),(X=0.5,Y=-0.5,Z=0.5,W=0.5),(X=0,Y=0,Z=0.7071,W=0.7071),(X=0.1104,Y=0.0847,Z=-0.7856,W=0.6028),(X=-0.1241,Y=-0.0952,Z=-0.7836,W=0.6013))
    ExecuteContext=@n12.A
   )
   > &RigUnit_ModifyTransforms(
    ItemToModify="((Item=(Type=Bone,Name=\"base_skirt_1_f_c\"),Transform=(Rotation=(X=-0.0233,Y=0.1403,Z=-0.0607,W=-0.988),Translation=(X=-0.6785,Y=0,Z=0.9648),Scale3D=(X=1,Y=1,Z=1))),(Item=(Type=Bone,Name=\"base_skirt_1_f_l\"),Transform=(Rotation=(X=-0.0024,Y=0.0102,Z=0.0586,W ...(truncated)"
    WeightMinimum=0
    Weight=RigVMDispatchNode(Index=4, Array=@InverseDistance_3.weights) {template="DISPATCH_RigVMDispatch_ArrayGetAtIndex(in Array,in Index,out Element)"}.Element
   )
   > &RigUnit_ModifyTransforms(
    ItemToModify="((Item=(Type=Bone,Name=\"base_skirt_1_f_l\"),Transform=(Rotation=(X=0.0006,Y=-0.055,Z=-0.027,W=0.9981),Translation=(X=0,Y=0,Z=0),Scale3D=(X=1,Y=1,Z=1))),(Item=(Type=Bone,Name=\"base_skirt_1_f_c\"),Transform=(Rotation=(X=0.0602,Y=0.1006,Z=0.0061,W=-0.9931),Tran ...(truncated)"
    WeightMinimum=0
    Weight=RigVMDispatchNode(Index=5, Array=@InverseDistance_3.weights) {template="DISPATCH_RigVMDispatch_ArrayGetAtIndex(in Array,in Index,out Element)"}.Element
   )
   > n13 = &RigUnit_SphericalPoseReader(
    ActiveRegionScaleFactors=(PositiveWidth=1.8,NegativeWidth=1.8,PositiveHeight=1.8,NegativeHeight=3)
    ActiveRegionSize=0.02
    Debug=(bDrawDebug=false,bDraw2D=False,bDrawLocalAxes=False,DebugScale=25,DebugSegments=20,DebugThickness=0.25)
    DriverAxis=(X=-1,Y=0,Z=0)
    DriverItem=(Type=Bone,Name="thigh_l")
    FalloffRegionScaleFactors=(PositiveWidth=1,NegativeWidth=0.9,PositiveHeight=1,NegativeHeight=3)
    FalloffSize=0.1
    FlipHeightScaling=False
    FlipWidthScaling=False
    OptionalParentItem=(Type=Bone,Name="None")
    RotationOffset=(X=115.4261,Y=0,Z=-40.6148)
   )
   > &RigUnit_ModifyTransforms(
    ItemToModify="((Item=(Type=Bone,Name=\"base_skirt_1_f_l\"),Transform=(Rotation=(X=0,Y=-0.228,Z=0,W=0.9737),Translation=(X=-0.9719,Y=0,Z=1.5927),Scale3D=(X=1,Y=1,Z=1))),(Item=(Type=Bone,Name=\"base_skirt_1_s_l\"),Transform=(Rotation=(X=0,Y=0.1178,Z=0,W=-0.993),Translation=(X ...(truncated)"
    WeightMinimum=0
    Weight=@n13.OutputParam
   )
   > n14 = &RigUnit_SphericalPoseReader(
    ActiveRegionScaleFactors=(PositiveWidth=1.8,NegativeWidth=1.8,PositiveHeight=1.8,NegativeHeight=3)
    ActiveRegionSize=0.02
    Debug=(bDrawDebug=false,bDraw2D=False,bDrawLocalAxes=False,DebugScale=25,DebugSegments=20,DebugThickness=0.25)
    DriverAxis=(X=-1,Y=0,Z=0)
    DriverItem=(Type=Bone,Name="thigh_l")
    FalloffRegionScaleFactors=(PositiveWidth=1,NegativeWidth=0.9,PositiveHeight=1,NegativeHeight=3)
    FalloffSize=0.1
    FlipHeightScaling=False
    FlipWidthScaling=False
    OptionalParentItem=(Type=Bone,Name="None")
    RotationOffset=(X=129.1743,Y=-1.2121,Z=-21.076)
   )
   > &RigUnit_ModifyTransforms(
    ItemToModify=((Item=(Type=Bone,Name="base_skirt_1_s_l"),Transform=(Rotation=(X=0.1714,Y=-0.1201,Z=0.0327,W=0.9773),Translation=(X=0,Y=0,Z=0),Scale3D=(X=1,Y=1,Z=1))))
    WeightMinimum=0
    Weight=@n14.OutputParam
   )
   InverseDistance_4 = RigVMFunctionReferenceNode(
    child=(Type=Bone,Name="thigh_r")
    drawLines=false
    ease=true
    parent=(Type=Bone,Name="pelvis")
    primaryAxis=(X=10,Y=0,Z=0)
    remapCurve=(EditorCurveData=(Keys=((),(Time=0.5,Value=0.5,ArriveTangent=1.203,LeaveTangent=1.203),(Time=1,Value=1))))
    remapMult=2
    targets=((),(X=0,Y=0,Z=-0.7071,W=0.7071),(X=0.5,Y=-0.5,Z=0.5,W=0.5),(X=0,Y=0,Z=0.7071,W=0.7071),(X=0.1066,Y=0.0895,Z=-0.7586,W=0.6365))
    ExecuteContext=@n12.B
   )
   > &RigUnit_ModifyTransforms(
    ItemToModify="((Item=(Type=Bone,Name=\"base_skirt_1_f_c\"),Transform=(Rotation=(X=0,Y=0.1243,Z=0,W=-0.9922),Translation=(X=-0.6785,Y=0,Z=0.9648),Scale3D=(X=1,Y=1,Z=1))),(Item=(Type=Bone,Name=\"base_skirt_1_f_r\"),Transform=(Rotation=(X=0.0024,Y=0.0102,Z=-0.0586,W=0.9982),Tr ...(truncated)"
    WeightMinimum=0
    Weight=RigVMDispatchNode(Index=4, Array=@InverseDistance_4.weights) {template="DISPATCH_RigVMDispatch_ArrayGetAtIndex(in Array,in Index,out Element)"}.Element
   )
   > n15 = &RigUnit_SphericalPoseReader(
    ActiveRegionScaleFactors=(PositiveWidth=1,NegativeWidth=2,PositiveHeight=5,NegativeHeight=5)
    ActiveRegionSize=0.03
    Debug=(bDrawDebug=false,bDraw2D=False,bDrawLocalAxes=False,DebugScale=25,DebugSegments=20,DebugThickness=0.25)
    DriverAxis=(X=1,Y=0,Z=0)
    DriverItem=(Type=Bone,Name="thigh_r")
    FalloffRegionScaleFactors=(PositiveWidth=0.5,NegativeWidth=1,PositiveHeight=1,NegativeHeight=1)
    FalloffSize=0.2
    FlipHeightScaling=False
    FlipWidthScaling=False
    OptionalParentItem=(Type=Bone,Name="None")
    RotationOffset=(X=-90,Y=0,Z=-30)
   )
   > &RigUnit_ModifyTransforms(
    ItemToModify="((Item=(Type=Bone,Name=\"base_skirt_1_f_r\"),Transform=(Rotation=(X=0.0151,Y=-0.1699,Z=-0.0672,W=0.983),Translation=(X=0.8025,Y=0,Z=1.8037),Scale3D=(X=1,Y=1,Z=1))),(Item=(Type=Bone,Name=\"base_skirt_1_f_c\"),Transform=(Rotation=(X=0,Y=0.1938,Z=0,W=-0.981),Tran ...(truncated)"
    WeightMinimum=0
    Weight=@n15.OutputParam
   )
   > &RigUnit_ModifyTransforms(
    ItemToModify="((Item=(Type=Bone,Name=\"kin_hangBelt_main_2_F\"),Transform=(Rotation=(X=0,Y=0,Z=-0.0976,W=0.9952),Translation=(X=0.4177,Y=5.8087,Z=1.1877),Scale3D=(X=1,Y=1,Z=1))),(Item=(Type=Bone,Name=\"kin_hangBelt_main_1_F\"),Transform=(Rotation=(X=-0.218,Y=0.0613,Z=-0.036 ...(truncated)"
    WeightMinimum=0
    Weight=@n15.OutputParam
   )
  ]}
 CollapseNode_ContainedGraph {nodes [
   n1 = RigVMFunctionEntryNode()
   &RigUnit_ParentConstraint(
    AdvancedSettings=(RotationOrderForFilter=XZY)
    Child=(Type=Bone,Name="base_skirt_1_f_c")
    Filter=(TranslationFilter=(bX=True,bY=True,bZ=True),RotationFilter=(bX=True,bY=True,bZ=True),ScaleFilter=(bX=False,bY=False,bZ=False))
    Parents="((Item=(Type=Bone,Name=\"deform_groin_l\"),Weight=0.0187),(Item=(Type=Bone,Name=\"deform_groin_r\"),Weight=0.0187),(Item=(Type=Bone,Name=\"thigh_twist_01_l\"),Weight=0.0529),(Item=(Type=Bone,Name=\"thigh_twist_01_r\"),Weight=0.0529),(Item=(Type=Bone,Name=\"spine_01 ...(truncated)"
    Weight=1
    ExecutePin=@n1.ExecutePin
   )
   > &RigUnit_ParentConstraint(
    AdvancedSettings=(RotationOrderForFilter=XZY)
    Child=(Type=Bone,Name="base_skirt_1_f_l")
    Filter=(TranslationFilter=(bX=True,bY=True,bZ=True),RotationFilter=(bX=True,bY=True,bZ=True),ScaleFilter=(bX=False,bY=False,bZ=False))
    Parents=((Item=(Type=Bone,Name="spine_01"),Weight=0.1147),(Item=(Type=Bone,Name="thigh_twist_01_l"),Weight=0.1421),(Item=(Type=Bone,Name="deform_groin_l"),Weight=0.2216),(Item=(Type=Bone,Name="pelvis"),Weight=0.5217))
    Weight=1
   )
   > &RigUnit_ParentConstraint(
    AdvancedSettings=(RotationOrderForFilter=XZY)
    Child=(Type=Bone,Name="base_skirt_1_s_l")
    Filter=(TranslationFilter=(bX=True,bY=True,bZ=True),RotationFilter=(bX=True,bY=True,bZ=True),ScaleFilter=(bX=False,bY=False,bZ=False))
    Parents="((Item=(Type=Bone,Name=\"deform_glute_l\"),Weight=0.0202),(Item=(Type=Bone,Name=\"spine_01\"),Weight=0.0568),(Item=(Type=Bone,Name=\"deform_groin_l\"),Weight=0.0704),(Item=(Type=Bone,Name=\"thigh_twist_01_l\"),Weight=0.3711),(Item=(Type=Bone,Name=\"pelvis\"),Weight= ...(truncated)"
    Weight=1
   )
   > &RigUnit_ParentConstraint(
    AdvancedSettings=(RotationOrderForFilter=XZY)
    Child=(Type=Bone,Name="base_skirt_1_b_l")
    Filter=(TranslationFilter=(bX=True,bY=True,bZ=True),RotationFilter=(bX=True,bY=True,bZ=True),ScaleFilter=(bX=False,bY=False,bZ=False))
    Parents=((Item=(Type=Bone,Name="spine_01"),Weight=0.0868),(Item=(Type=Bone,Name="thigh_twist_01_l"),Weight=0.1484),(Item=(Type=Bone,Name="deform_glute_l"),Weight=0.1654),(Item=(Type=Bone,Name="pelvis"),Weight=0.5994))
    Weight=1
   )
   > &RigUnit_ParentConstraint(
    AdvancedSettings=(RotationOrderForFilter=XZY)
    Child=(Type=Bone,Name="base_skirt_1_b_c")
    Filter=(TranslationFilter=(bX=True,bY=True,bZ=True),RotationFilter=(bX=True,bY=True,bZ=True),ScaleFilter=(bX=False,bY=False,bZ=False))
    Parents="((Item=(Type=Bone,Name=\"thigh_twist_01_l\"),Weight=0.0289),(Item=(Type=Bone,Name=\"thigh_twist_01_r\"),Weight=0.0289),(Item=(Type=Bone,Name=\"deform_glute_l\"),Weight=0.0467),(Item=(Type=Bone,Name=\"deform_glute_r\"),Weight=0.0467),(Item=(Type=Bone,Name=\"spine_01 ...(truncated)"
    Weight=1
   )
   > &RigUnit_ParentConstraint(
    AdvancedSettings=(RotationOrderForFilter=XZY)
    Child=(Type=Bone,Name="base_skirt_1_f_r")
    Filter=(TranslationFilter=(bX=True,bY=True,bZ=True),RotationFilter=(bX=True,bY=True,bZ=True),ScaleFilter=(bX=False,bY=False,bZ=False))
    Parents=((Item=(Type=Bone,Name="spine_01"),Weight=0.1147),(Item=(Type=Bone,Name="thigh_twist_01_r"),Weight=0.1421),(Item=(Type=Bone,Name="deform_groin_r"),Weight=0.2216),(Item=(Type=Bone,Name="pelvis"),Weight=0.5217))
    Weight=1
   )
   > &RigUnit_ParentConstraint(
    AdvancedSettings=(RotationOrderForFilter=XZY)
    Child=(Type=Bone,Name="base_skirt_1_s_r")
    Filter=(TranslationFilter=(bX=True,bY=True,bZ=True),RotationFilter=(bX=True,bY=True,bZ=True),ScaleFilter=(bX=False,bY=False,bZ=False))
    Parents="((Item=(Type=Bone,Name=\"deform_glute_r\"),Weight=0.0202),(Item=(Type=Bone,Name=\"spine_01\"),Weight=0.0569),(Item=(Type=Bone,Name=\"deform_groin_r\"),Weight=0.0704),(Item=(Type=Bone,Name=\"thigh_twist_01_r\"),Weight=0.3711),(Item=(Type=Bone,Name=\"pelvis\"),Weight= ...(truncated)"
    Weight=1
   )
   > n2 = &RigUnit_ParentConstraint(
    AdvancedSettings=(RotationOrderForFilter=XZY)
    Child=(Type=Bone,Name="base_skirt_1_b_r")
    Filter=(TranslationFilter=(bX=True,bY=True,bZ=True),RotationFilter=(bX=True,bY=True,bZ=True),ScaleFilter=(bX=False,bY=False,bZ=False))
    Parents=((Item=(Type=Bone,Name="spine_01"),Weight=0.0868),(Item=(Type=Bone,Name="thigh_twist_01_r"),Weight=0.1484),(Item=(Type=Bone,Name="deform_glute_r"),Weight=0.1654),(Item=(Type=Bone,Name="pelvis"),Weight=0.5994))
    Weight=1
   )
   RigVMFunctionReturnNode(ExecutePin=@n2.ExecutePin)
  ]}
 CollapseNode_ContainedGraph {nodes [
   n1 = RigVMFunctionEntryNode()
   &RigUnit_ParentConstraint(
    AdvancedSettings=(RotationOrderForFilter=XZY)
    Child=(Type=Bone,Name="deform_groin_base_l")
    Parents=((Item=(Type=Bone,Name="thigh_twist_01_l"),Weight=0.7),(Item=(Type=Bone,Name="pelvis"),Weight=0.3))
    Weight=1
    ExecutePin=@n1.ExecutePin
   )
   > &RigUnit_ParentConstraint(
    AdvancedSettings=(RotationOrderForFilter=XZY)
    Child=(Type=Bone,Name="deform_groin_base_r")
    Parents=((Item=(Type=Bone,Name="thigh_twist_01_r"),Weight=0.7),(Item=(Type=Bone,Name="pelvis"),Weight=0.3))
    Weight=1
   )
   > &RigUnit_ParentConstraint(
    AdvancedSettings=(RotationOrderForFilter=XZY)
    Child=(Type=Bone,Name="deform_glute_base_l")
    Parents=((Item=(Type=Bone,Name="thigh_twist_01_l"),Weight=0.25),(Item=(Type=Bone,Name="pelvis"),Weight=0.75))
    Weight=1
   )
   > &RigUnit_ParentConstraint(
    AdvancedSettings=(RotationOrderForFilter=XZY)
    Child=(Type=Bone,Name="deform_glute_base_r")
    Parents=((Item=(Type=Bone,Name="thigh_twist_01_r"),Weight=0.25),(Item=(Type=Bone,Name="pelvis"),Weight=0.75))
    Weight=1
   )
   > n2 = &RigUnit_SphericalPoseReader(
    ActiveRegionScaleFactors=(PositiveWidth=3.5,NegativeWidth=0.5,PositiveHeight=3,NegativeHeight=3)
    ActiveRegionSize=0.1
    Debug=(bDrawDebug=false,DebugScale=25,DebugSegments=20,DebugThickness=0.25)
    DriverAxis=(X=-1,Y=0,Z=0)
    DriverItem=(Type=Bone,Name="thigh_l")
    FalloffRegionScaleFactors=(PositiveWidth=1,NegativeWidth=1,PositiveHeight=1,NegativeHeight=1)
    FalloffSize=0.45
    OptionalParentItem=(Type=Bone)
    RotationOffset=(X=90,Y=0,Z=0)
   )
   > &RigUnit_ModifyTransforms(
    ItemToModify="((Item=(Type=Bone,Name=\"deform_glute_l\"),Transform=(Rotation=(X=0,Y=0,Z=0,W=1),Translation=(X=-7,Y=0,Z=0),Scale3D=(X=1,Y=1,Z=1))),(Item=(Type=Bone,Name=\"deform_groin_l\"),Transform=(Rotation=(X=0,Y=0,Z=0,W=1),Translation=(X=-3,Y=12,Z=0),Scale3D=(X=1,Y=1,Z=1 ...(truncated)"
    Weight=@n2.OutputParam
   )
   > n3 = &RigUnit_SphericalPoseReader(
    ActiveRegionScaleFactors=(PositiveWidth=3.5,NegativeWidth=0.5,PositiveHeight=3,NegativeHeight=3)
    ActiveRegionSize=0.1
    Debug=(bDrawDebug=false,DebugScale=25,DebugSegments=20,DebugThickness=0.25)
    DriverAxis=(X=1,Y=0,Z=0)
    DriverItem=(Type=Bone,Name="thigh_r")
    FalloffRegionScaleFactors=(PositiveWidth=1,NegativeWidth=1,PositiveHeight=1,NegativeHeight=1)
    FalloffSize=0.45
    FlipHeightScaling=false
    FlipWidthScaling=true
    OptionalParentItem=(Type=Bone)
    RotationOffset=(X=-90,Y=0,Z=0)
   )
   > n4 = &RigUnit_ModifyTransforms(
    ItemToModify="((Item=(Type=Bone,Name=\"deform_glute_r\"),Transform=(Rotation=(X=0,Y=0,Z=0,W=1),Translation=(X=7,Y=0,Z=0),Scale3D=(X=1,Y=1,Z=1))),(Item=(Type=Bone,Name=\"deform_groin_r\"),Transform=(Rotation=(X=0,Y=0,Z=0,W=1),Translation=(X=3,Y=-12,Z=0),Scale3D=(X=1,Y=1,Z=1) ...(truncated)"
    Weight=@n3.OutputParam
   )
   RigVMFunctionReturnNode(ExecutePin=@n4.ExecutePin)
  ]}
 CollapseNode_ContainedGraph {nodes [
   n1 = RigVMFunctionEntryNode()
   &RigUnit_ParentConstraint(
    AdvancedSettings=(RotationOrderForFilter=XZY)
    Child=(Type=Bone,Name="base_lowerNecklace_1_c")
    Parents=((Item=(Type=Bone,Name="clavicle_l"),Weight=0.0127),(Item=(Type=Bone,Name="spine_05"),Weight=0.9873))
    Weight=1
    ExecutePin=@n1.ExecutePin
   )
   > &RigUnit_ParentConstraint(
    AdvancedSettings=(RotationOrderForFilter=XZY)
    Child=(Type=Bone,Name="kin_lowerNecklace_r")
    Parents=((Item=(Type=Bone,Name="clavicle_r"),Weight=0.1478),(Item=(Type=Bone,Name="spine_05"),Weight=0.8522))
    Weight=1
   )
   > &RigUnit_ParentConstraint(
    AdvancedSettings=(RotationOrderForFilter=XZY)
    Child=(Type=Bone,Name="kin_lowerNecklace_l")
    Parents=((Item=(Type=Bone,Name="clavicle_l"),Weight=0.1487),(Item=(Type=Bone,Name="spine_05"),Weight=0.8513))
    Weight=1
   )
   > &RigUnit_ParentConstraint(
    AdvancedSettings=(RotationOrderForFilter=XZY)
    Child=(Type=Bone,Name="coll_breast_bot_c")
    Filter=(TranslationFilter=(bX=True,bY=True,bZ=True),RotationFilter=(bX=False,bY=False,bZ=False),ScaleFilter=(bX=False,bY=False,bZ=False))
    Parents=((Item=(Type=Bone,Name="spine_04"),Weight=0.2305),(Item=(Type=Bone,Name="spine_05"),Weight=0.7695))
    Weight=1
   )
   > &RigUnit_ParentConstraint(
    AdvancedSettings=(RotationOrderForFilter=XZY)
    Child=(Type=Bone,Name="target_breast_bot_c")
    Filter=(TranslationFilter=(bX=True,bY=True,bZ=True),RotationFilter=(bX=False,bY=False,bZ=False),ScaleFilter=(bX=False,bY=False,bZ=False))
    Parents=((Item=(Type=Bone,Name="deform_pec_r"),Weight=0.0114),(Item=(Type=Bone,Name="spine_03"),Weight=0.0124),(Item=(Type=Bone,Name="spine_05"),Weight=0.3511),(Item=(Type=Bone,Name="spine_04"),Weight=0.625))
    Weight=1
   )
   > n2 = &RigUnit_AimConstraintLocalSpaceOffset(
    Child=(Type=Bone,Name="coll_breast_bot_c")
    Filter=(bX=True,bY=True,bZ=True)
    Parents=((Item=(Type=Bone,Name="target_breast_bot_c"),Weight=1))
    Weight=1
    WorldUp=(Target=(X=1,Y=0,Z=0),Kind=Direction,Space=(Type=Bone,Name="spine_05"))
   )
   RigVMFunction_ControlFlowBranch = &RigVMFunction_ControlFlowBranch(ExecuteContext=@n2.ExecutePin, Condition=RigVMVariableNode() {variable=RBAN enabled}.Value)
   RigVMFunctionReturnNode(ExecutePin=@RigVMFunction_ControlFlowBranch.Completed)
   n3 = &RigUnit_SphericalPoseReader(
    ActiveRegionScaleFactors=(PositiveWidth=1,NegativeWidth=1,PositiveHeight=1,NegativeHeight=1)
    ActiveRegionSize=0.015
    Debug=(bDrawDebug=false,bDraw2D=False,bDrawLocalAxes=False,DebugScale=25,DebugSegments=20,DebugThickness=0.1)
    DriverAxis=(X=1,Y=0,Z=0)
    DriverItem=(Type=Bone,Name="spine_05")
    FalloffRegionScaleFactors=(PositiveWidth=1,NegativeWidth=1,PositiveHeight=1,NegativeHeight=1)
    FalloffSize=0.03
    FlipHeightScaling=False
    FlipWidthScaling=False
    OptionalParentItem=(Type=Bone,Name="None")
    RotationOffset=(X=7.6,Y=-98,Z=0)
    ExecutePin=@RigVMFunction_ControlFlowBranch.False
   )
   > n4 = &RigUnit_SphericalPoseReader(
    ActiveRegionScaleFactors=(PositiveWidth=1,NegativeWidth=1,PositiveHeight=1,NegativeHeight=1)
    ActiveRegionSize=0.01
    Debug=(bDrawDebug=false,bDraw2D=False,bDrawLocalAxes=False,DebugScale=25,DebugSegments=20,DebugThickness=0.1)
    DriverAxis=(X=1,Y=0,Z=0)
    DriverItem=(Type=Bone,Name="spine_05")
    FalloffRegionScaleFactors=(PositiveWidth=1,NegativeWidth=1,PositiveHeight=1,NegativeHeight=1)
    FalloffSize=0.03
    FlipHeightScaling=False
    FlipWidthScaling=False
    OptionalParentItem=(Type=Bone,Name="None")
    RotationOffset=(X=1,Y=-85,Z=0)
   )
   > n5 = &RigUnit_SphericalPoseReader(
    ActiveRegionScaleFactors=(PositiveWidth=1,NegativeWidth=1,PositiveHeight=1,NegativeHeight=1)
    ActiveRegionSize=0.015
    Debug=(bDrawDebug=false,bDraw2D=False,bDrawLocalAxes=False,DebugScale=25,DebugSegments=20,DebugThickness=0.1)
    DriverAxis=(X=1,Y=0,Z=0)
    DriverItem=(Type=Bone,Name="spine_05")
    FalloffRegionScaleFactors=(PositiveWidth=1,NegativeWidth=1,PositiveHeight=1,NegativeHeight=1)
    FalloffSize=0.04
    FlipHeightScaling=False
    FlipWidthScaling=False
    OptionalParentItem=(Type=Bone,Name="None")
    RotationOffset=(X=7.9376,Y=-89.0092,Z=0)
   )
   Maximum = RigVMAggregateNode(A=@n5.OutputParam, C=@n3.OutputParam, B=@n4.OutputParam)
   > &RigUnit_ModifyTransforms(
    ItemToModify=((Item=(Type=Bone,Name="base_lowerNecklace_1_c"),Transform=(Rotation=(X=0.061,Y=-0.0435,Z=0.0027,W=0.9972),Translation=(X=0,Y=0,Z=0.3),Scale3D=(X=1,Y=1,Z=1))))
    WeightMinimum=0
    Weight=@Maximum.Result
   )
   RigUnit_GetTransform = &RigUnit_GetTransform(Item=(Type=Bone,Name="coll_breast_bot_c"), bInitial=False)
   RigVMFunction_DebugTransformMutableNoSpace = /Script/RigVM.RigVMFunction_DebugTransformMutableNoSpace(
    Color=(R=1,G=0,B=0,A=1)
    DebugDrawSettings=(DepthPriority=SDPG_Foreground,Lifetime=-1)
    ExecutePin=()
    Mode=Axes
    Scale=5
    Thickness=0.5
    WorldOffset=(Rotation=(X=0,Y=0,Z=0,W=1),Translation=(X=0,Y=0,Z=0),Scale3D=(X=1,Y=1,Z=1))
    bEnabled=true
    Transform=@RigUnit_GetTransform.Transform
   ) {is=struct RigVMUnitNode}
  ]}
 CollapseNode_ContainedGraph {nodes [
   n1 = RigVMFunctionEntryNode()
   n2 = &RigUnit_GetRelativeTransformForItem(Child=(Type=Bone,Name="spine_02"), Parent=(Type=Null,Name="spine_02_sp_null"), bChildInitial=False, bParentInitial=False)
   RigVMVariableNode(
    ExecuteContext=@n1.ExecutePin
    Value=&RigVMFunction_MathQuaternionToEuler(
     RotationOrder=ZYX
     Value=&RigVMFunction_MathQuaternionSwingTwist(TwistAxis=(X=1,Y=0,Z=0), Input=@n2.RelativeTransform.Rotation).Swing
    ).Result.Z
   ) {variable=c spine02RelPelvis rz}
   > RigVMFunction_Sequence = RigVMAggregateNode()
   RigVMFunctionReturnNode(ExecutePin=@RigVMFunction_Sequence.C)
   n3 = &RigVMFunction_MathDoubleRemap(SourceMaximum=-100, SourceMinimum=0, TargetMaximum=1, TargetMinimum=0, Value=RigVMVariableNode() {variable=r thighLocal rz float}.Value)
   /Script/ControlRig.RigUnit_SetCurveValue(Curve=r_thighLocal rz_0_ n100, Value=@n3.Result, ExecutePin=@RigVMFunction_Sequence.B) {is=struct RigVMUnitNode}
   &RigUnit_ParentConstraint(
    AdvancedSettings=(RotationOrderForFilter=XZY)
    Child=(Type=Bone,Name="kin_belt_1")
    Parents=((Item=(Type=Bone,Name="spine_02"),Weight=0.2009),(Item=(Type=Bone,Name="pelvis"),Weight=0.212),(Item=(Type=Bone,Name="spine_01"),Weight=0.5872))
    Weight=1
    ExecutePin=@RigVMFunction_Sequence.A
   )
   > &RigUnit_ParentConstraint(
    AdvancedSettings=(RotationOrderForFilter=XZY)
    Child=(Type=Bone,Name="kin_belt_2")
    Parents=((Item=(Type=Bone,Name="spine_03"),Weight=0.0118),(Item=(Type=Bone,Name="thigh_twist_01_l"),Weight=0.0143),(Item=(Type=Bone,Name="spine_02"),Weight=0.2033),(Item=(Type=Bone,Name="pelvis"),Weight=0.2235),(Item=(Type=Bone,Name="spine_01"),Weight=0.5471))
    Weight=1
   )
   > &RigUnit_ParentConstraint(
    AdvancedSettings=(RotationOrderForFilter=XZY)
    Child=(Type=Bone,Name="kin_belt_3")
    Parents=((Item=(Type=Bone,Name="thigh_twist_01_l"),Weight=0.0112),(Item=(Type=Bone,Name="spine_02"),Weight=0.1607),(Item=(Type=Bone,Name="pelvis"),Weight=0.3041),(Item=(Type=Bone,Name="spine_01"),Weight=0.524))
    Weight=1
   )
   Remap = &RigVMFunction_MathDoubleRemap(SourceMaximum=-100, SourceMinimum=-5, TargetMaximum=1, TargetMinimum=0, Value=RigVMVariableNode() {variable=l thighLocal rz float}.Value)
   > &RigUnit_ModifyTransforms(
    ItemToModify="((Item=(Type=Bone,Name=\"kin_belt_2\"),Transform=(Rotation=(X=0,Y=0,Z=-0.0023,W=1),Translation=(X=0.2567,Y=2.2875,Z=-0.1446),Scale3D=(X=1,Y=1,Z=1))),(Item=(Type=Bone,Name=\"kin_belt_3\"),Transform=(Rotation=(X=-0.1942,Y=-0.0653,Z=-0.1233,W=0.971),Translation=( ...(truncated)"
    WeightMinimum=0
    Weight=@Remap.Result
   )
   Power = &RigVMFunction_MathDoublePow(B=1.1, A=@n3.Result) {template="Power::Execute(in A,in B,out Result)"}
   > &RigUnit_ModifyTransforms(
    ItemToModify="((Item=(Type=Bone,Name=\"base_hangBelt_main_1\"),Transform=(Rotation=(X=-0.0611,Y=-0.002,Z=0.2569,W=0.9645),Translation=(X=-0.6726,Y=1.5933,Z=0.7865),Scale3D=(X=1,Y=1,Z=1))),(Item=(Type=Bone,Name=\"dyn_hangBelt_main_2\"),Transform=(Rotation=(X=0.1659,Y=-0.0242 ...(truncated)"
    WeightMinimum=0
    Weight=@Power.Result
   )
   > &RigUnit_ModifyTransforms(
    ItemToModify=((Item=(Type=Bone,Name="base_hangBelt_main_1"),Transform=(Rotation=(X=-0.1215,Y=0.0005,Z=-0.052,W=-0.9912),Translation=(X=0,Y=0,Z=0),Scale3D=(X=1,Y=1,Z=1))))
    WeightMinimum=0
    Weight=&RigVMFunction_MathDoubleRemap(SourceMaximum=55, SourceMinimum=0, TargetMaximum=1, TargetMinimum=0, Value=RigVMVariableNode() {variable=r thighLocal rz float}.Value).Result
   )
   n4 = RigVMVariableNode() {variable=r thighLocal ry float}
   > &RigUnit_ModifyTransforms(
    ItemToModify=((Item=(Type=Bone,Name="base_hangBelt_main_1"),Transform=(Rotation=(X=0,Y=0,Z=0.1628,W=0.9867),Translation=(X=0,Y=0,Z=0),Scale3D=(X=1,Y=1,Z=1))))
    WeightMinimum=0
    Weight=&RigVMFunction_MathDoublePow(
     B=1.2
     A=&RigVMFunction_MathDoubleRemap(SourceMaximum=-55, SourceMinimum=-5, TargetMaximum=1, TargetMinimum=0, Value=@n4.Value).Result
    ) {template="Power::Execute(in A,in B,out Result)"}.Result
   )
   > &RigUnit_ModifyTransforms(
    ItemToModify=((Item=(Type=Bone,Name="kin_spine_01_belt"),Transform=(Rotation=(X=0,Y=0,Z=0,W=-1),Translation=(X=0,Y=-0.85,Z=0),Scale3D=(X=1,Y=1,Z=1))))
    WeightMinimum=0
    Weight=&RigVMFunction_MathDoubleRemap(SourceMaximum=60, SourceMinimum=0, TargetMaximum=1, TargetMinimum=0, Value=RigVMVariableNode() {variable=c spine02RelPelvis rz}.Value).Result
   )
   > n5 = &RigUnit_SphericalPoseReader(
    ActiveRegionScaleFactors=(PositiveWidth=4,NegativeWidth=1,PositiveHeight=4,NegativeHeight=4)
    ActiveRegionSize=0.08
    Debug=(bDrawDebug=false,bDraw2D=False,bDrawLocalAxes=False,DebugScale=25,DebugSegments=20,DebugThickness=0.25)
    DriverAxis=(X=-1,Y=0,Z=0)
    DriverItem=(Type=Bone,Name="thigh_l")
    FalloffRegionScaleFactors=(PositiveWidth=1,NegativeWidth=1,PositiveHeight=1,NegativeHeight=1)
    FalloffSize=0.5
    FlipHeightScaling=False
    FlipWidthScaling=False
    OptionalParentItem=(Type=Bone,Name="None")
    RotationOffset=(X=90,Y=0,Z=0)
   )
   RigVMFunction_MathBoolNot = &RigVMFunction_MathBoolNot(Value=RigVMVariableNode() {variable=onGrindRail}.Value)
   Cast = &RigVMFunction_MathBoolToFloat(Value=@RigVMFunction_MathBoolNot.Result)
   AlphaInterp = &RigVMFunction_AlphaInterp(InterpSpeedDecreasing=10, InterpSpeedIncreasing=10, Scale=1, bClampResult=False, Value=@Cast.Result)
   Multiply = RigVMAggregateNode(
    C=&RigVMFunction_MathDoubleRemap(SourceMaximum=1, SourceMinimum=0, TargetMaximum=1, TargetMinimum=0.2, Value=@AlphaInterp.Result).Result
    B=&RigVMFunction_MathDoubleRemap(SourceMaximum=60, SourceMinimum=15, TargetMaximum=1, TargetMinimum=0, Value=RigVMVariableNode() {variable=c spine02RelPelvis rz}.Value).Result
    A=@n5.OutputParam
   )
   > &RigUnit_ModifyTransforms(
    ItemToModify="((Item=(Type=Bone,Name=\"kin_belt_2\"),Transform=(Rotation=(X=0,Y=0,Z=0.3827,W=0.9239),Translation=(X=0,Y=-1,Z=0),Scale3D=(X=1,Y=1,Z=1))),(Item=(Type=Bone,Name=\"kin_belt_3\"),Transform=(Rotation=(X=0,Y=0,Z=0.1305,W=0.9914),Translation=(X=-1.2,Y=2.35,Z=2.35),S ...(truncated)"
    WeightMinimum=0
    Weight=@Multiply.Result
   )
   > &RigUnit_OffsetTransformForItem(
    Item=(Type=Bone,Name="dyn_skirt_1_f_l")
    OffsetTransform=(Rotation=(X=0,Y=0.1736,Z=0,W=0.9848),Translation=(X=0,Y=0,Z=-2),Scale3D=(X=1,Y=1,Z=1))
    bPropagateToChildren=false
    Weight=@Multiply.Result
   )
  ]}
 CollapseNode_ContainedGraph {nodes [
   n1 = RigVMFunctionEntryNode()
   &RigUnit_ParentConstraint(
    AdvancedSettings=(RotationOrderForFilter=XZY)
    Child=(Type=Bone,Name="deform_lat_base_l")
    Filter=(TranslationFilter=(bX=True,bY=True,bZ=False),RotationFilter=(bX=False,bY=False,bZ=False),ScaleFilter=(bX=False,bY=False,bZ=False))
    Parents="((Item=(Type=Bone,Name=\"spine_03\"),Weight=0.0363),(Item=(Type=Bone,Name=\"upperarm_twist_01_l\"),Weight=0.0845),(Item=(Type=Bone,Name=\"spine_05\"),Weight=0.2608),(Item=(Type=Bone,Name=\"clavicle_l\"),Weight=0.2971),(Item=(Type=Bone,Name=\"spine_04\"),Weight=0.321 ...(truncated)"
    Weight=1
    ExecutePin=@n1.ExecutePin
   )
   > &RigUnit_ParentConstraint(
    AdvancedSettings=(RotationOrderForFilter=XZY)
    Child=(Type=Bone,Name="deform_lat_base_r")
    Filter=(TranslationFilter=(bX=True,bY=True,bZ=False),RotationFilter=(bX=False,bY=False,bZ=False),ScaleFilter=(bX=False,bY=False,bZ=False))
    Parents="((Item=(Type=Bone,Name=\"deform_pec_r\"),Weight=0.0102),(Item=(Type=Bone,Name=\"spine_03\"),Weight=0.0359),(Item=(Type=Bone,Name=\"upperarm_twist_01_r\"),Weight=0.0835),(Item=(Type=Bone,Name=\"spine_05\"),Weight=0.2586),(Item=(Type=Bone,Name=\"clavicle_r\"),Weight=0 ...(truncated)"
    Weight=1
   )
   > GetAlphaFromQuatDelta = RigVMFunctionReferenceNode(
    Rotation Order=ZYX
    child=(Type=Bone,Name="clavicle_l")
    outputMinMax=(X=0,Y=1)
    parent=(Type=Null,Name="clavicle_l_null")
    premult=1
    rotationMinMax=(X=-10,Y=-50)
    triggerAxis=Y
    twistAxis=(X=1,Y=0,Z=0)
   )
   > &RigUnit_ModifyTransforms(
    ItemToModify=((Item=(Type=Bone,Name="deform_lat_l"),Transform=(Rotation=(X=0,Y=0,Z=0,W=1),Translation=(X=-5,Y=0,Z=0),Scale3D=(X=1,Y=1,Z=1))))
    WeightMinimum=0
    Weight=@GetAlphaFromQuatDelta.alpha
   )
   > GetAlphaFromQuatDelta_1 = RigVMFunctionReferenceNode(
    Rotation Order=ZYX
    child=(Type=Bone,Name="clavicle_r")
    outputMinMax=(X=0,Y=1)
    parent=(Type=Null,Name="clavicle_r_null")
    premult=1
    rotationMinMax=(X=-10,Y=-50)
    triggerAxis=Y
    twistAxis=(X=1,Y=0,Z=0)
   )
   > n2 = &RigUnit_ModifyTransforms(
    ItemToModify=((Item=(Type=Bone,Name="deform_lat_r"),Transform=(Rotation=(X=0,Y=0,Z=0,W=1),Translation=(X=-5,Y=0,Z=0),Scale3D=(X=1,Y=1,Z=1))))
    WeightMinimum=0
    Weight=@GetAlphaFromQuatDelta_1.alpha
   )
   RigVMFunctionReturnNode(ExecutePin=@n2.ExecutePin)
  ]}
 CollapseNode_ContainedGraph {nodes [
   n1 = RigVMFunctionEntryNode()
   &RigUnit_ParentConstraint(
    AdvancedSettings=(InterpolationType=Average,RotationOrderForFilter=XZY)
    Child=(Type=Bone,Name="deform_ankle_in_base_l")
    Parents=((Item=(Type=Bone,Name="foot_l"),Weight=0.7),(Item=(Type=Bone,Name="calf_twist_01_l"),Weight=0.3))
    Weight=1
    ExecutePin=@n1.ExecutePin
   )
   > &RigUnit_ParentConstraint(
    AdvancedSettings=(InterpolationType=Average,RotationOrderForFilter=XZY)
    Child=(Type=Bone,Name="deform_ankle_out_base_l")
    Parents=((Item=(Type=Bone,Name="foot_l"),Weight=0.8),(Item=(Type=Bone,Name="calf_twist_01_l"),Weight=0.2))
    Weight=1
   )
   > &RigUnit_ParentConstraint(
    AdvancedSettings=(InterpolationType=Average,RotationOrderForFilter=XZY)
    Child=(Type=Bone,Name="deform_ankle_in_base_r")
    Parents=((Item=(Type=Bone,Name="foot_r"),Weight=0.7),(Item=(Type=Bone,Name="calf_twist_01_r"),Weight=0.3))
    Weight=1
   )
   > &RigUnit_ParentConstraint(
    AdvancedSettings=(InterpolationType=Average,RotationOrderForFilter=XZY)
    Child=(Type=Bone,Name="deform_ankle_out_base_r")
    Parents=((Item=(Type=Bone,Name="foot_r"),Weight=0.8),(Item=(Type=Bone,Name="calf_twist_01_r"),Weight=0.2))
    Weight=1
   )
   > GetAlphaFromQuatDelta = RigVMFunctionReferenceNode(
    Rotation Order=ZYX
    child=(Type=Bone,Name="foot_l")
    outputMinMax=(X=0,Y=1)
    parent=(Type=Null,Name="foot_l_null")
    premult=1
    rotationMinMax=(X=0,Y=-40)
    triggerAxis=Z
    twistAxis=(X=1,Y=0,Z=0)
   )
   > &RigUnit_ModifyTransforms(
    ItemToModify=((Item=(Type=Bone,Name="deform_ankle_in_l"),Transform=(Rotation=(X=0,Y=0,Z=0,W=1),Translation=(X=1.2,Y=0,Z=0),Scale3D=(X=1,Y=0.5,Z=1))))
    WeightMinimum=0
    Weight=@GetAlphaFromQuatDelta.alpha
   )
   > GetAlphaFromQuatDelta_1 = RigVMFunctionReferenceNode(
    Rotation Order=ZYX
    child=(Type=Bone,Name="foot_l")
    outputMinMax=(X=0,Y=1)
    parent=(Type=Null,Name="foot_l_null")
    premult=1
    rotationMinMax=(X=0,Y=80)
    triggerAxis=Z
    twistAxis=(X=1,Y=0,Z=0)
   )
   Power = &RigVMFunction_MathDoublePow(B=1.2, A=@GetAlphaFromQuatDelta_1.alpha) {template="Power::Execute(in A,in B,out Result)"}
   > &RigUnit_ModifyTransforms(
    ItemToModify="((Item=(Type=Bone,Name=\"deform_ankle_out_l\"),Transform=(Rotation=(X=0,Y=0,Z=0,W=1),Translation=(X=-15,Y=0,Z=0),Scale3D=(X=1,Y=0.5,Z=1))),(Item=(Type=Bone,Name=\"deform_ankle_in_l\"),Transform=(Rotation=(X=0,Y=0,Z=0,W=1),Translation=(X=-2,Y=0,Z=0),Scale3D=(X= ...(truncated)"
    WeightMinimum=0
    Weight=@Power.Result
   )
   > GetAlphaFromQuatDelta_2 = RigVMFunctionReferenceNode(
    Rotation Order=ZYX
    child=(Type=Bone,Name="foot_r")
    outputMinMax=(X=0,Y=1)
    parent=(Type=Null,Name="foot_r_null")
    premult=1
    rotationMinMax=(X=0,Y=-40)
    triggerAxis=Z
    twistAxis=(X=1,Y=0,Z=0)
   )
   > &RigUnit_ModifyTransforms(
    ItemToModify=((Item=(Type=Bone,Name="deform_knee_in_r"),Transform=(Rotation=(X=0,Y=0,Z=0,W=1),Translation=(X=-1.2,Y=0,Z=0),Scale3D=(X=1,Y=0.5,Z=1))))
    WeightMinimum=0
    Weight=@GetAlphaFromQuatDelta_2.alpha
   )
   > GetAlphaFromQuatDelta_3 = RigVMFunctionReferenceNode(
    Rotation Order=ZYX
    child=(Type=Bone,Name="foot_r")
    outputMinMax=(X=0,Y=1)
    parent=(Type=Null,Name="foot_r_null")
    premult=1
    rotationMinMax=(X=0,Y=80)
    triggerAxis=Z
    twistAxis=(X=1,Y=0,Z=0)
   )
   > n2 = &RigUnit_ModifyTransforms(
    ItemToModify="((Item=(Type=Bone,Name=\"deform_ankle_out_r\"),Transform=(Rotation=(X=0,Y=0,Z=0,W=1),Translation=(X=15,Y=0,Z=0),Scale3D=(X=1,Y=0.5,Z=1))),(Item=(Type=Bone,Name=\"deform_ankle_in_r\"),Transform=(Rotation=(X=0,Y=0,Z=0,W=1),Translation=(X=2,Y=0,Z=0),Scale3D=(X=1, ...(truncated)"
    WeightMinimum=0
    Weight=&RigVMFunction_MathDoublePow(B=1.2, A=@GetAlphaFromQuatDelta_3.alpha) {template="Power::Execute(in A,in B,out Result)"}.Result
   )
   RigVMFunctionReturnNode(ExecutePin=@n2.ExecutePin)
  ]}
 Maximum_1_ContainedGraph {nodes [
   n1 = RigVMFunctionEntryNode()
   Maximum = &RigVMFunction_MathDoubleMax(A=@n1.A, B=@n1.B)
   RigVMFunctionReturnNode(Result=&RigVMFunction_MathDoubleMax(B=@n1.C, A=@Maximum.Result).Result)
  ]}
 Multiply_1_ContainedGraph {nodes [
   n1 = RigVMFunctionEntryNode()
   Multiply = &RigVMFunction_MathDoubleMul(A=@n1.A, B=@n1.B)
   RigVMFunctionReturnNode(Result=&RigVMFunction_MathDoubleMul(B=@n1.C, A=@Multiply.Result).Result)
  ]}
 Multiply_3_ContainedGraph {nodes [
   n1 = RigVMFunctionEntryNode()
   RigVMFunctionReturnNode(Result=&RigVMFunction_MathTransformMul(B=@n1.C, A=&RigVMFunction_MathTransformMul(A=@n1.A, B=@n1.B) {template="Multiply::Execute(in A,in B,out Result)"}.Result) {template="Multiply::Execute(in A,in B,out Result)"}.Result)
  ]}
 New Function_ContainedGraph {nodes [
   RigVMFunctionEntryNode()
   > GetAlphaFromQuatDelta = RigVMFunctionReferenceNode(
    Rotation Order=ZYX
    child=(Type=Bone,Name="hand_l")
    onlyTwist=True
    outputMinMax=(X=0,Y=1)
    parent=(Type=Null,Name="hand_l_null")
    premult=1
    rotationMinMax=(X=120,Y=150)
    twistAxis=(X=1,Y=0,Z=0)
   )
   > &RigUnit_ModifyTransforms(
    ItemToModify="((Item=(Type=Bone,Name=\"lowerarm_twist_01_l\"),Transform=(Rotation=(X=0,Y=-0.0262,Z=0,W=0.9997),Translation=(X=0,Y=0,Z=1),Scale3D=(X=1,Y=1.2,Z=1.2))),(Item=(Type=Bone,Name=\"lowerarm_twist_02_l\"),Transform=(Rotation=(X=0.2588,Y=0,Z=0,W=0.9659),Translation=(X ...(truncated)"
    WeightMinimum=0
    Weight=@GetAlphaFromQuatDelta.alpha
   )
   > RigVMFunctionReturnNode()
  ]}
 New Function_ContainedGraph {nodes [
   RigVMFunctionEntryNode()
   Num = RigVMDispatchNode(Array=RigVMVariableNode() {variable=targets}.Value) {template="DISPATCH_RigVMDispatch_ArrayGetNum(in Array,out Num)"}
   > RigVMDispatchNode(Element=0, Num=@Num.Num, Array=RigVMVariableNode() {variable=outputValues}.Value) {template="DISPATCH_RigVMDispatch_ArrayInit(io Array,in Num,in Element)"}
   > RigVMDispatchNode(Element=0, Array=RigVMVariableNode() {variable=divisors}.Value, Num=@Num.Num) {template="DISPATCH_RigVMDispatch_ArrayInit(io Array,in Num,in Element)"}
   > RigVMDispatchNode(Element=0, Array=RigVMVariableNode() {variable=outLenRemaped}.Value, Num=@Num.Num) {template="DISPATCH_RigVMDispatch_ArrayInit(io Array,in Num,in Element)"}
   n1 = RigVMVariableNode() {variable=child}
   > RigVMVariableNode(Value=&RigVMFunction_MathTransformMul(
     A=(Rotation=(X=0,Y=0,Z=0,W=1),Translation=(X=0,Y=0,Z=0),Scale3D=(X=1,Y=1,Z=1))
     B=&RigUnit_GetTransform(bInitial=False, Item=@n1.Value).Transform
     A.Translation=RigVMVariableNode() {variable=primaryAxis}.Value
    ) {template="Multiply::Execute(in A,in B,out Result)"}.Result.Translation) {variable=target vector}
   n2 = RigVMVariableNode() {variable=parent}
   > RigVMVariableNode(Value=&RigVMFunction_MathTransformMul(
     A=&RigUnit_GetRelativeTransformForItem(bChildInitial=true, bParentInitial=true, Child=@n1.Value, Parent=@n2.Value).RelativeTransform
     B=&RigUnit_GetTransform(bInitial=False, Item=@n2.Value).Transform
    ) {template="Multiply::Execute(in A,in B,out Result)"}.Result) {variable=parentChildLocated transform}
   > /Script/RigVM.RigVMFunction_DebugLineNoSpace(
    Color=(R=0,G=0,B=1,A=1)
    DebugDrawSettings=(DepthPriority=SDPG_Foreground,Lifetime=-1)
    Thickness=0.5
    WorldOffset=(Rotation=(X=0,Y=0,Z=0,W=1),Translation=(X=0,Y=0,Z=0),Scale3D=(X=1,Y=1,Z=1))
    B=RigVMVariableNode() {variable=target vector}.Value
    A=RigVMVariableNode() {variable=parentChildLocated transform}.Value.Translation
    bEnabled=RigVMVariableNode() {variable=drawLines}.Value
   ) {is=struct RigVMUnitNode}
   > For_Each = RigVMDispatchNode(Array=RigVMVariableNode() {variable=targets}.Value) {template="DISPATCH_RigVMDispatch_ArrayIterator(in Array,out Element,out Index,out Count,out Ratio)"}
   > RigVMVariableNode(Value=@For_Each.Index) {variable=index1 iter}
   Multiply_1 = RigVMAggregateNode(
    A=(Rotation=(X=0,Y=0,Z=0,W=1),Translation=(X=0,Y=0,Z=0),Scale3D=(X=1,Y=1,Z=1))
    B=(Rotation=(X=0,Y=0,Z=0,W=1),Translation=(X=0,Y=0,Z=0),Scale3D=(X=1,Y=1,Z=1))
    C=RigVMVariableNode() {variable=parentChildLocated transform}.Value
    A.Translation=RigVMVariableNode() {variable=primaryAxis}.Value
    B.Rotation=@For_Each.Element
   )
   > RigVMFunction_DebugTransformMutableNoSpace = &RigVMFunction_DebugTransformMutableNoSpace(
    Color=(R=1,G=0,B=1,A=0)
    DebugDrawSettings=(DepthPriority=SDPG_Foreground,Lifetime=-1)
    Mode=Box
    Scale=1
    Thickness=0.5
    Transform=(Rotation=(X=0,Y=0,Z=0,W=1),Translation=(X=0,Y=0,Z=0),Scale3D=(X=1,Y=1,Z=1))
    WorldOffset=(Rotation=(X=0,Y=0,Z=0,W=1),Translation=(X=0,Y=0,Z=0),Scale3D=(X=1,Y=1,Z=1))
    bEnabled=RigVMVariableNode() {variable=drawLines}.Value
    Transform.Translation=@Multiply_1.Result.Translation
   )
   n3 = RigVMVariableNode() {variable=parentChildLocated transform}
   Unit = /Script/RigVM.RigVMFunction_MathVectorUnit(Value=/Script/RigVM.RigVMFunction_MathVectorSub(A=RigVMVariableNode() {variable=target vector}.Value, B=@n3.Value.Translation) {is=struct RigVMUnitNode, template="Subtract::Execute(in A,in B,out Result)"}.Result) {is=struct RigVMUnitNode, template="Unit::Execute(in Value,out Result)"}
   Subtract = &RigVMFunction_MathVectorSub(
    A=@Unit.Result
    B=&RigVMFunction_MathVectorUnit(Value=&RigVMFunction_MathVectorSub(A=@Multiply_1.Result.Translation, B=@n3.Value.Translation) {template="Subtract::Execute(in A,in B,out Result)"}.Result) {template="Unit::Execute(in Value,out Result)"}.Result
   ) {template="Subtract::Execute(in A,in B,out Result)"}
   RigVMFunction_MathVectorLength = /Script/RigVM.RigVMFunction_MathVectorLength(Value=@Subtract.Result) {is=struct RigVMUnitNode, template="Length::Execute(in Value,out Result)"}
   Remap = &RigVMFunction_MathDoubleRemap(
    SourceMaximum=0
    TargetMaximum=1
    TargetMinimum=0
    bClamp=False
    SourceMinimum=RigVMVariableNode() {variable=remapMult}.Value
    Value=@RigVMFunction_MathVectorLength.Result
   )
   > Set_At_2 = RigVMDispatchNode(Element=@Remap.Result, Array=RigVMVariableNode() {variable=outLenRemaped}.Value, Index=RigVMVariableNode() {variable=index1 iter}.Value) {template="DISPATCH_RigVMDispatch_ArraySetAtIndex(io Array,in Index,in Element)"}
   Power = &RigVMFunction_MathDoublePow(B=2, A=@RigVMFunction_MathVectorLength.Result) {template="Power::Execute(in A,in B,out Result)"}
   Clamp = &RigVMFunction_MathDoubleClamp(Maximum=10000000000, Value=@Power.Result)
   Divide = &RigVMFunction_MathDoubleDiv(A=1, B=@Clamp.Result)
   > Set_At = RigVMDispatchNode(Element=@Divide.Result, Index=RigVMVariableNode() {variable=index1 iter}.Value, Array=RigVMVariableNode() {variable=divisors}.Value) {template="DISPATCH_RigVMDispatch_ArraySetAtIndex(io Array,in Index,in Element)"}
   n4 = RigVMVariableNode() {variable=divisors}
   For_Each_1 = RigVMDispatchNode(ExecuteContext=@For_Each.Completed, Array=@n4.Value) {template="DISPATCH_RigVMDispatch_ArrayIterator(in Array,out Element,out Index,out Count,out Ratio)"}
   ArraySum = /Script/RigVM.RigVMFunction_MathDoubleArraySum(Array=@n4.Value) {is=struct RigVMUnitNode, template="ArraySum::Execute(in Array,out Sum)"}
   n5 = &RigVMFunction_MathDoubleDiv(B=@ArraySum.Sum, A=@For_Each_1.Element)
   Multiply = &RigVMFunction_MathDoubleMul(
    B=RigVMDispatchNode(Index=@For_Each_1.Index, Array=RigVMVariableNode() {variable=outLenRemaped}.Value) {template="DISPATCH_RigVMDispatch_ArrayGetAtIndex(in Array,in Index,out Element)"}.Element
    A=@n5.Result
   )
   RigVMFunction_AnimEvalRichCurve = &RigVMFunction_AnimEvalRichCurve(
    SourceMinimum=0
    TargetMinimum=0
    Value=RigVMDispatchNode(False=@n5.Result, True=@Multiply.Result, Condition=RigVMVariableNode() {variable=ease}.Value) {template="DISPATCH_RigVMDispatch_If(in Condition,in True,in False,out Result)"}.Result
    Curve=RigVMVariableNode() {variable=remapCurve}.Value
   )
   > Set_At_1 = RigVMDispatchNode(Index=@For_Each_1.Index, Element=@RigVMFunction_AnimEvalRichCurve.Result, Array=RigVMVariableNode() {variable=outputValues}.Value) {template="DISPATCH_RigVMDispatch_ArraySetAtIndex(io Array,in Index,in Element)"}
   RigVMFunctionReturnNode(weights=RigVMVariableNode() {variable=outputValues}.Value, ExecuteContext=@For_Each_1.Completed)
  ]}
 New Function_ContainedGraph {nodes [
   RigVMFunctionEntryNode()
   RigVMFunction_MathBoolToInteger = &RigVMFunction_MathBoolToInteger(Value=RigVMVariableNode() {variable=onlyTwist}.Value)
   RigVMFunction_MathQuaternionSwingTwist = &RigVMFunction_MathQuaternionSwingTwist(
    Input=&RigUnit_GetRelativeTransformForItem(
     bChildInitial=False
     bParentInitial=False
     Child=RigVMVariableNode() {variable=child}.Value
     Parent=RigVMVariableNode() {variable=parent}.Value
    ).RelativeTransform.Rotation
    TwistAxis=RigVMVariableNode() {variable=twistAxis}.Value
   )
   n1 = RigVMDispatchNode(
    Values=((),())
    Index=@RigVMFunction_MathBoolToInteger.Result
    Values.0=@RigVMFunction_MathQuaternionSwingTwist.Swing
    Values.1=@RigVMFunction_MathQuaternionSwingTwist.Twist
   ) {template="DISPATCH_RigVMDispatch_SelectInt32(in Index,in Values,out Result)"}
   > RigVMVariableNode(Value=@n1.Result) {variable=entry_outputQuat}
   RigVMFunction_MathQuaternionToEuler = &RigVMFunction_MathQuaternionToEuler(Value=@n1.Result, RotationOrder=RigVMVariableNode() {variable=Rotation Order}.Value)
   > DISPATCH_RigVMDispatch_Print = RigVMDispatchNode(
    ScreenColor=(R=1,G=1,B=1,A=1)
    ScreenDuration=1.5
    Value=@RigVMFunction_MathQuaternionToEuler.Result
    Enabled=RigVMVariableNode() {variable=debugRotation}.Value
   ) {template="DISPATCH_RigVMDispatch_Print(in Prefix,in Value,in Enabled,in ScreenDuration,in ScreenColor)"}
   Cast = RigVMDispatchNode(Value=RigVMVariableNode() {variable=triggerAxis}.Value) {template="DISPATCH_RigVMDispatch_CastEnumToInt(in Value,out Result)"}
   DISPATCH_RigVMDispatch_SelectInt32 = RigVMDispatchNode(
    Values=(0,0,0)
    Index=@Cast.Result
    Values.0=@RigVMFunction_MathQuaternionToEuler.Result.X
    Values.1=@RigVMFunction_MathQuaternionToEuler.Result.Y
    Values.2=@RigVMFunction_MathQuaternionToEuler.Result.Z
   ) {template="DISPATCH_RigVMDispatch_SelectInt32(in Index,in Values,out Result)"}
   n2 = RigVMVariableNode() {variable=rotationMinMax}
   n3 = RigVMVariableNode() {variable=outputMinMax}
   Remap = &RigVMFunction_MathDoubleRemap(
    bClamp=False
    Value=@DISPATCH_RigVMDispatch_SelectInt32.Result
    SourceMinimum=@n2.Value.X
    SourceMaximum=@n2.Value.Y
    TargetMinimum=@n3.Value.X
    TargetMaximum=@n3.Value.Y
   )
   Absolute = &RigVMFunction_MathDoubleAbs(Value=@Remap.Result)
   n4 = RigVMVariableNode() {variable=absoluteValue}
   Multiply = &RigVMFunction_MathDoubleMul(
    A=&RigVMFunction_MathDoubleClamp(Value=RigVMDispatchNode(True=@Absolute.Result, False=@Remap.Result, Condition=@n4.Value) {template="DISPATCH_RigVMDispatch_If(in Condition,in True,in False,out Result)"}.Result).Result
    B=RigVMVariableNode() {variable=premult}.Value
   )
   > RigVMFunctionReturnNode(alpha=@Multiply.Result, rawQuat=RigVMVariableNode() {variable=entry_outputQuat}.Value)
  ]}
 RigVMFunctionLibrary {nodes [
   GetAlphaFromQuatDelta = RigVMCollapseNode(
    ExecuteContext=()
    Rotation Order=ZYX
    child=(Type=Bone,Name="None")
    outputMinMax=(X=0,Y=1)
    parent=(Type=Bone,Name="None")
    premult=1
    rotationMinMax=(X=0,Y=0)
    twistAxis=(X=1,Y=0,Z=0)
   )
   InverseDistance = RigVMCollapseNode(
    ExecuteContext=()
    child=(Type=Bone,Name="None")
    drawLines=true
    parent=(Type=Bone,Name="None")
    primaryAxis=(X=0,Y=0,Z=0)
    remapCurve=(EditorCurveData=(Keys=((),(Time=0.5,Value=0.5,ArriveTangent=1.203,LeaveTangent=1.203),(Time=1,Value=1))))
    remapMult=4
    targets=()
   )
  ]}
 RigVMFunction_Sequence_1_ContainedGraph {nodes [
   n1 = RigVMFunctionEntryNode()
   RigVMFunction_Sequence = &RigVMFunction_Sequence(ExecuteContext=@n1.ExecuteContext)
   n2 = &RigVMFunction_Sequence(ExecuteContext=@RigVMFunction_Sequence.B)
   n3 = &RigVMFunction_Sequence(ExecuteContext=@n2.B)
   n4 = &RigVMFunction_Sequence(ExecuteContext=@n3.B)
   RigVMFunctionReturnNode(A=@RigVMFunction_Sequence.A, B=@n2.A, C=@n3.A, D=@n4.A, E=@n4.B)
  ]}
 RigVMFunction_Sequence_1_ContainedGraph {nodes [
   n1 = RigVMFunctionEntryNode()
   RigVMFunction_Sequence = &RigVMFunction_Sequence(ExecuteContext=@n1.ExecuteContext)
   n2 = &RigVMFunction_Sequence(ExecuteContext=@RigVMFunction_Sequence.B)
   RigVMFunctionReturnNode(A=@RigVMFunction_Sequence.A, B=@n2.A, C=@n2.B)
  ]}
 RigVMFunction_Sequence_1_ContainedGraph {nodes [
   n1 = RigVMFunctionEntryNode()
   RigVMFunction_Sequence = &RigVMFunction_Sequence(ExecuteContext=@n1.ExecuteContext)
   n2 = &RigVMFunction_Sequence(ExecuteContext=@RigVMFunction_Sequence.B)
   RigVMFunctionReturnNode(A=@RigVMFunction_Sequence.A, B=@n2.A, C=@n2.B)
  ]}
 RigVMModel {nodes [
   RigUnit_BeginExecution = &RigUnit_BeginExecution()
   patchGroinGlute = RigVMCollapseNode(ExecutePin=@RigUnit_BeginExecution.ExecutePin)
   FN_DeformRig_FNC = RigVMFunctionReferenceNode(
    DeformRig_ Curve="(deformRig_l_upperarm_up_alpha_72_#1=False,deformRig_r_upperarm_up_alpha_73_#2=False,deformRig_l_upperarm_fr_alpha_74_#3=False,deformRig_r_upperarm_fr_alpha_75_CB51B8 ...(truncated)"
    DeformRig_Bool="(upperarm_1_#4=True,elbow_3_#5=True,wrist_7_#6=True,groin_8_#7=false,glute_17_#8=false,knee_9_40C408324CA2 ...(truncated)"
    DeformRig_Debug=(upperarm_debug_18_#4=False,wrist_debug_20_#6=False,lat_debug_26_#9=False,groinGlute_debug_27_#7=False)
    DeformRig_Value="(upperarm_up_18_#10=(Rotation=(X=0,Y=0,Z=0,W=1),Translation=(X=0,Y=0,Z=3),Scale3D=(X=0.7,Y=1,Z=1)),upperarm_fwd_fr_19_#11=(Rotation=(X=0,Y=0,Z=0,W=1),Translation=(X=0,Y=10,Z=0),Scale3D=(X=0.5,Y=1,Z= ...(truncated)"
    ExecuteContext=@patchGroinGlute.ExecutePin
   )
   > RigVMFunction_Sequence = RigVMAggregateNode()
   armTwistFix = RigVMCollapseNode(ExecuteContext=@RigVMFunction_Sequence.A)
   skirt = RigVMCollapseNode(ExecutePin=@armTwistFix.ExecuteContext)
   choker = RigVMCollapseNode(ExecutePin=@skirt.ExecutePin)
   necklace = RigVMCollapseNode(ExecutePin=@RigVMFunction_Sequence.B)
   misc = RigVMCollapseNode(ExecutePin=@necklace.ExecutePin)
   belt = RigVMCollapseNode(ExecutePin=@misc.ExecutePin)
   latSetup = RigVMCollapseNode(ExecutePin=@RigVMFunction_Sequence.C)
   ankleSetup = RigVMCollapseNode(ExecutePin=@latSetup.ExecutePin)
  ]}
 RigVMModel Rig {nodes [
   n1 = &RigUnit_PrepareForExecution()
   ItemArray = RigVMDispatchNode(Value=((Type=Bone,Name="head"),(Type=Bone,Name="hand_l"),(Type=Bone,Name="hand_r"),(Type=Bone,Name="clavicle_r"),(Type=Bone,Name="clavicle_l"),(Type=Bone,Name="thigh_l"),(Type=Bone,Name="foot_l"),(Type=Bone,Name="thigh_r"),(Type=Bone,Name="foot_r"))) {template="DISPATCH_RigVMDispatch_Constant(io Value)"}
   DISPATCH_RigVMDispatch_ArrayIterator = RigVMDispatchNode(Array=@ItemArray.Value, ExecuteContext=@n1.ExecutePin) {template="DISPATCH_RigVMDispatch_ArrayIterator(in Array,out Element,out Index,out Count,out Ratio)"}
   Concat = /Script/RigVM.RigVMFunction_NameConcat(B=_null, A=@DISPATCH_RigVMDispatch_ArrayIterator.Element.Name) {is=struct RigVMUnitNode, template="Concat::Execute(in A,in B,out Result)"}
   > &RigUnit_HierarchyAddNull(
    Space=LocalSpace
    Name=@Concat.Result
    Transform=&RigUnit_GetTransform(Item=(Type=Bone,Name="clavicle_l"), Space=LocalSpace, bInitial=true, Item.Name=@DISPATCH_RigVMDispatch_ArrayIterator.Element.Name).Transform
    Parent=/Script/ControlRig.RigUnit_HierarchyGetParent(bDefaultParent=True, Child=@DISPATCH_RigVMDispatch_ArrayIterator.Element) {is=struct RigVMUnitNode}.Parent
   )
   &RigUnit_HierarchyAddNull(
    Name=head_sp_null
    Parent=(Type=Bone,Name="spine_05")
    Space=LocalSpace
    Transform=&RigUnit_GetRelativeTransformForItem(Child=(Type=Bone,Name="head"), Parent=(Type=Bone,Name="spine_05"), bChildInitial=true, bParentInitial=true).RelativeTransform
    ExecutePin=@DISPATCH_RigVMDispatch_ArrayIterator.Completed
   )
   > &RigUnit_HierarchyAddNull(
    Name=spine_02_sp_null
    Parent=(Type=Bone,Name="pelvis")
    Space=LocalSpace
    Transform=&RigUnit_GetRelativeTransformForItem(Child=(Type=Bone,Name="spine_02"), Parent=(Type=Bone,Name="pelvis"), bChildInitial=true, bParentInitial=true).RelativeTransform
   )
  ]}
]
hierarchy [
 r_thighLocal rz_0_ n100 {type=Curve, loc=<0 0 0>, rot=<0 0 0 1>}
 root {loc=<0 0 0>, rot=<0 0 0 1>}
 pelvis {parent=root, loc=<0 -0.5301 98.0571>, rot=<0.0127 0.707 -0.0127 -0.707>}
 spine_01 {parent=pelvis, loc=<6.1208 -0.9328 0>, rot=<0 0 0.034 -0.9994>}
 spine_02 {parent=spine_01, loc=<6.5997 -0.2095 0>, rot=<0 0 0.0238 -0.9997>}
 spine_03 {parent=spine_02, loc=<6.4914 -0.1237 0>, rot=<0 0 0.0149 -0.9999>}
 spine_04 {parent=spine_03, loc=<6.9598 0.0794 0>, rot=<0 0 -0.0132 -0.9999>}
 spine_05 {parent=spine_04, loc=<6.9987 0 0>, rot=<0 0 -0.0015 -1>}
 neck_01 {parent=spine_05, loc=<11.3138 0.1529 0>, rot=<0 0 -0.2003 -0.9797>}
 neck_02 {parent=neck_01, loc=<4.5627 0 0>, rot=<0 0 0.0504 -0.9987>}
 head {parent=neck_02, loc=<5.3914 0 0>, rot=<0 0 0.115 -0.9934>}
 FACIAL_C_FacialRoot {parent=head, loc=<1.4768 -0.0402 0>, rot=<0.5021 -0.4979 -0.5021 -0.4979>}
 FACIAL_C_UnderChin {parent=FACIAL_C_FacialRoot, loc=<0 7.5788 2.8678>, rot=<0.5 0 0 -0.866>}
 FACIAL_L_UnderChin {parent=FACIAL_C_FacialRoot, loc=<2.8165 6.5154 1.9753>, rot=<0.4025 -0.355 0.0823 -0.8398>}
 FACIAL_R_UnderChin {parent=FACIAL_C_FacialRoot, loc=<-2.8166 6.5153 1.9753>, rot=<0.4025 0.355 -0.0823 -0.8398>}
 FACIAL_C_Jaw {parent=FACIAL_C_FacialRoot, loc=<0 1.6464 1.0599>, rot=<0.342 0 0 -0.9397>}
 FACIAL_C_Jawline {parent=FACIAL_C_Jaw, loc=<0 1.8194 6.8806>, rot=<0.2588 0 0 -0.9659>}
 FACIAL_L_Ear {parent=FACIAL_C_FacialRoot, loc=<6.2752 -0.238 -0.7486>, rot=<-0.0262 -0.0007 -0.0262 -0.9993>}
 FACIAL_R_Ear {parent=FACIAL_C_FacialRoot, loc=<-6.2752 -0.238 -0.7486>, rot=<-0.0262 0.0007 0.0262 -0.9993>}
 FACIAL_L_Jawline {parent=FACIAL_C_FacialRoot, loc=<4.2415 4.9303 3.5822>, rot=<0.2092 0.1578 -0.0547 -0.9635>}
 FACIAL_R_Jawline {parent=FACIAL_C_FacialRoot, loc=<-4.2415 4.9303 3.5822>, rot=<0.2092 -0.1578 0.0547 -0.9635>}
 kin_chocker_1 {parent=neck_01, loc=<1.3114 5.0473 -0.0171>, rot=<0.4983 -0.4954 -0.4995 -0.5067>}
 kin_chocker_2 {parent=neck_01, loc=<1.413 2.5336 -3.9904>, rot=<0.6833 -0.6856 -0.1274 -0.2163>}
 kin_chocker_3 {parent=neck_01, loc=<1.45 -1.8358 -4.1631>, rot=<0.6952 -0.6693 0.2415 0.1026>}
 kin_chocker_4 {parent=neck_01, loc=<1.4075 -4.1634 0.001>, rot=<0.5092 -0.4906 0.5097 0.4902>}
 kin_chocker_5 {parent=neck_01, loc=<1.4451 -1.8312 4.1659>, rot=<-0.2407 0.1021 -0.6957 -0.669>}
 kin_chocker_6 {parent=neck_01, loc=<1.4021 2.5357 3.976>, rot=<0.1389 -0.2181 -0.6851 -0.6811>}
 clavicle_l {parent=spine_05, loc=<5.4831 -0.1053 -3.2946>, rot=<0.0465 -0.7289 -0.009 -0.6829>}
 upperarm_l {parent=clavicle_l, loc=<9.8658 -1.2136 -0.4105>, rot=<0.0164 -0.3245 -0.0269 -0.9454>}
 lowerarm_l {parent=upperarm_l, loc=<26.465 -0.8345 0.259>, rot=<0.0036 0.0161 -0.2298 -0.9731>}
 hand_l {parent=lowerarm_l, loc=<22.7355 -0.3203 -0.6105>, rot=<-0.6784 -0.0223 0.0735 -0.7307>}
 thumb_01_l {parent=hand_l, loc=<1.3118 -0.3191 -2.459>, rot=<0.6546 -0.4042 0.0664 -0.6354>}
 thumb_02_l {parent=thumb_01_l, loc=<5.9336 -0.0448 0.0401>, rot=<-0.0037 0.0213 0.298 -0.9543>}
 thumb_03_l {parent=thumb_02_l, loc=<2.2223 0.3682 -0.0907>, rot=<0.0118 -0.013 0.0066 -0.9998>}
 index_metacarpal_l {parent=hand_l, loc=<1.6535 0.0183 -2.96>, rot=<0.0158 -0.0124 -0.0173 -0.9996>}
 index_01_l {parent=index_metacarpal_l, loc=<8.3008 0.0357 -0.0524>, rot=<-0.0044 0.0485 0.0924 -0.9945>}
 index_02_l {parent=index_01_l, loc=<3.3587 -0.3004 0.0452>, rot=<-0.002 0.0027 0.3225 -0.9466>}
 index_03_l {parent=index_02_l, loc=<2.5578 -0.123 -0.0252>, rot=<-0.002 0.0067 0.2558 -0.9667>}
 middle_metacarpal_l {parent=hand_l, loc=<2.1856 -0.1725 -0.5984>, rot=<0.006 0.0168 -0.0329 -0.9993>}
 middle_01_l {parent=middle_metacarpal_l, loc=<7.3275 -0.2447 0.0129>, rot=<-0.0208 0.0042 0.0858 -0.9961>}
 middle_02_l {parent=middle_01_l, loc=<4.0123 0.1401 0.1175>, rot=<-0.0218 0.0258 0.3512 -0.9357>}
 middle_03_l {parent=middle_02_l, loc=<2.6077 0.0534 0.1373>, rot=<0.0082 0.0188 0.1708 -0.9851>}
 ring_metacarpal_l {parent=hand_l, loc=<2.2074 -0.3625 1.3638>, rot=<-0.0284 0.0722 -0.0458 -0.9959>}
 ring_01_l {parent=ring_metacarpal_l, loc=<7.0295 -0.6197 -0.1446>, rot=<-0.0489 -0.0252 0.1892 -0.9804>}
 ring_02_l {parent=ring_01_l, loc=<3.6174 0.4647 -0.1027>, rot=<0 0 0.2806 -0.9598>}
 ring_03_l {parent=ring_02_l, loc=<2.4152 0.114 0.0199>, rot=<0 0 0.1189 -0.9929>}
 pinky_metacarpal_l {parent=hand_l, loc=<1.7616 -0.9753 3.3814>, rot=<-0.0397 0.0814 -0.069 -0.9935>}
 pinky_01_l {parent=pinky_metacarpal_l, loc=<7.1709 -0.4831 -0.1094>, rot=<-0.0654 -0.0035 0.2242 -0.9723>}
 pinky_02_l {parent=pinky_01_l, loc=<2.8274 0.0893 -0.1043>, rot=<0.0037 0.0011 0.2854 -0.9584>}
 pinky_03_l {parent=pinky_02_l, loc=<2.2046 0.0307 -0.0078>, rot=<-0.0002 0.0098 0.1445 -0.9895>}
 deform_wrist_dn_base_l {parent=hand_l, loc=<0 -1.5326 0>, rot=<0 0 0 -1>}
 deform_wrist_dn_l {parent=deform_wrist_dn_base_l, loc=<0 0 0>, rot=<0 0 0 1>}
 deform_wrist_up_base_l {parent=hand_l, loc=<0 2.3009 0>, rot=<0 0 0 -1>}
 deform_wrist_up_l {parent=deform_wrist_up_base_l, loc=<0 0 0>, rot=<0 0 0 1>}
 lowerarm_twist_02_l {parent=lowerarm_l, loc=<10.6707 0 0>, rot=<0 0 0 1>}
 lowerarm_twist_01_l {parent=lowerarm_l, loc=<18.316 0 0>, rot=<0 0 0 1>}
 deform_elbow_in_base_l {parent=lowerarm_l, loc=<0.1834 3.6062 -0.2572>, rot=<0 0 0 -1>}
 deform_elbow_in_l {parent=deform_elbow_in_base_l, loc=<0 0 0>, rot=<0 0 0 1>}
 deform_elbow_out_base_l {parent=lowerarm_l, loc=<0 -4 0>, rot=<0 0 0 1>}
 deform_elbow_out_l {parent=deform_elbow_out_base_l, loc=<0 0 0>, rot=<0 0 0 1>}
 upperarm_twist_01_l {parent=upperarm_l, loc=<7.2135 -0.1134 0>, rot=<0 0 0 1>}
 deform_upperarm_bk_base_l {parent=upperarm_twist_01_l, loc=<-6.4053 -4.3746 0.314>, rot=<0 0 0 -1>}
 deform_upperarm_bk_l {parent=deform_upperarm_bk_base_l, loc=<0 0 0>, rot=<0 0 0 1>}
 deform_upperarm_fr_base_l {parent=upperarm_twist_01_l, loc=<-5.6312 5.092 0.3354>, rot=<0 0 0 -1>}
 deform_upperarm_fr_l {parent=deform_upperarm_fr_base_l, loc=<0 0 0>, rot=<0 0 0 1>}
 deform_upperarm_up_base_l {parent=upperarm_twist_01_l, loc=<-8.891 0.4825 3.9517>, rot=<0 0 0 -1>}
 deform_upperarm_up_l {parent=deform_upperarm_up_base_l, loc=<0 0 0>, rot=<0 0 0 1>}
 upperarm_twist_02_l {parent=upperarm_l, loc=<19.7998 -0.1134 0>, rot=<0 0 0 1>}
 base_strapArm_1_l {parent=upperarm_l, loc=<9.5463 2.3158 2.7732>, rot=<-0.1943 0.6866 0.1927 -0.6736>}
 dyn_strapArm_1_l {parent=base_strapArm_1_l, loc=<0 0 0>, rot=<0 0 0 1>}
 base_strapArm_3_l {parent=upperarm_l, loc=<19.7942 1.9126 3.0224>, rot=<-0.1509 0.6628 0.1638 -0.7149>}
 dyn_strapArm_3_l {parent=base_strapArm_3_l, loc=<0 0 0>, rot=<0 0 0 1>}
 base_strapArm_2_l {parent=upperarm_l, loc=<12.9576 1.8403 3.0815>, rot=<-0.3502 0.6377 0.3323 -0.6002>}
 dyn_strapArm_2_l {parent=base_strapArm_2_l, loc=<0 0 0>, rot=<0 0 0 1>}
 clavicle_r {parent=spine_05, loc=<5.4831 -0.1053 3.2946>, rot=<0.7289 0.0465 -0.6829 0.009>}
 upperarm_r {parent=clavicle_r, loc=<-9.8658 1.2136 0.4105>, rot=<0.0164 -0.3245 -0.0269 -0.9454>}
 lowerarm_r {parent=upperarm_r, loc=<-26.465 0.8345 -0.259>, rot=<0.0036 0.0161 -0.2298 -0.9731>}
 hand_r {parent=lowerarm_r, loc=<-22.7355 0.3203 0.6106>, rot=<-0.6784 -0.0223 0.0735 -0.7307>}
 thumb_01_r {parent=hand_r, loc=<-1.3118 0.3191 2.459>, rot=<0.6546 -0.4042 0.0664 -0.6354>}
 thumb_02_r {parent=thumb_01_r, loc=<-5.9336 0.0448 -0.0401>, rot=<-0.0037 0.0213 0.298 -0.9543>}
 thumb_03_r {parent=thumb_02_r, loc=<-2.2223 -0.3682 0.0907>, rot=<0.0118 -0.013 0.0066 -0.9998>}
 index_metacarpal_r {parent=hand_r, loc=<-1.6535 -0.0183 2.96>, rot=<0.0158 -0.0124 -0.0173 -0.9996>}
 index_01_r {parent=index_metacarpal_r, loc=<-8.3008 -0.0357 0.0524>, rot=<-0.0044 0.0485 0.0924 -0.9945>}
 index_02_r {parent=index_01_r, loc=<-3.3587 0.3004 -0.0452>, rot=<-0.002 0.0027 0.3225 -0.9466>}
 index_03_r {parent=index_02_r, loc=<-2.5578 0.123 0.0252>, rot=<-0.002 0.0067 0.2558 -0.9667>}
 middle_metacarpal_r {parent=hand_r, loc=<-2.1856 0.1725 0.5984>, rot=<0.006 0.0168 -0.0329 -0.9993>}
 middle_01_r {parent=middle_metacarpal_r, loc=<-7.3275 0.2447 -0.0129>, rot=<-0.0208 0.0042 0.0858 -0.9961>}
 middle_02_r {parent=middle_01_r, loc=<-4.0123 -0.1401 -0.1175>, rot=<-0.0218 0.0258 0.3512 -0.9357>}
 middle_03_r {parent=middle_02_r, loc=<-2.6077 -0.0534 -0.1373>, rot=<0.0082 0.0188 0.1708 -0.9851>}
 ring_metacarpal_r {parent=hand_r, loc=<-2.2074 0.3625 -1.3638>, rot=<-0.0284 0.0722 -0.0458 -0.9959>}
 ring_01_r {parent=ring_metacarpal_r, loc=<-7.0295 0.6197 0.1446>, rot=<-0.0489 -0.0252 0.1892 -0.9804>}
 ring_02_r {parent=ring_01_r, loc=<-3.6174 -0.4647 0.1027>, rot=<0 0 0.2806 -0.9598>}
 ring_03_r {parent=ring_02_r, loc=<-2.4152 -0.114 -0.0199>, rot=<0 0 0.1189 -0.9929>}
 pinky_metacarpal_r {parent=hand_r, loc=<-1.7616 0.9753 -3.3814>, rot=<-0.0397 0.0814 -0.069 -0.9935>}
 pinky_01_r {parent=pinky_metacarpal_r, loc=<-7.1709 0.4831 0.1094>, rot=<-0.0654 -0.0035 0.2242 -0.9723>}
 pinky_02_r {parent=pinky_01_r, loc=<-2.8274 -0.0893 0.1043>, rot=<0.0037 0.0011 0.2854 -0.9584>}
 pinky_03_r {parent=pinky_02_r, loc=<-2.2046 -0.0307 0.0078>, rot=<-0.0002 0.0098 0.1445 -0.9895>}
 deform_wrist_dn_base_r {parent=hand_r, loc=<-0.0001 1.5328 0>, rot=<0 0 0 -1>}
 deform_wrist_dn_r {parent=deform_wrist_dn_base_r, loc=<0 0 0>, rot=<0 0 0 1>}
 deform_wrist_up_base_r {parent=hand_r, loc=<-0.0002 -2.3007 -0.0001>, rot=<0 0 0 -1>}
 deform_wrist_up_r {parent=deform_wrist_up_base_r, loc=<0 0 0>, rot=<0 0 0 1>}
 lowerarm_twist_02_r {parent=lowerarm_r, loc=<-10.6707 0 0>, rot=<0 0 0 -1>}
 lowerarm_twist_01_r {parent=lowerarm_r, loc=<-18.316 0 0>, rot=<0 0 0 -1>}
 deform_elbow_in_base_r {parent=lowerarm_r, loc=<-0.1834 -3.6062 0.2572>, rot=<0 0 0 -1>}
 deform_elbow_in_r {parent=deform_elbow_in_base_r, loc=<0 0 0>, rot=<0 0 0 1>}
 deform_elbow_out_base_r {parent=lowerarm_r, loc=<0 4 -0.0001>, rot=<0 0 0 -1>}
 deform_elbow_out_r {parent=deform_elbow_out_base_r, loc=<0 0 0>, rot=<0 0 0 1>}
 upperarm_twist_01_r {parent=upperarm_r, loc=<-7.2135 0.1134 0>, rot=<0 0 0 -1>}
 deform_upperarm_bk_base_r {parent=upperarm_twist_01_r, loc=<6.4053 4.3746 -0.314>, rot=<0 0 0 -1>}
 deform_upperarm_bk_r {parent=deform_upperarm_bk_base_r, loc=<0 0 0>, rot=<0 0 0 1>}
 deform_upperarm_fr_base_r {parent=upperarm_twist_01_r, loc=<5.6312 -5.092 -0.3354>, rot=<0 0 0 -1>}
 deform_upperarm_fr_r {parent=deform_upperarm_fr_base_r, loc=<0 0 0>, rot=<0 0 0 1>}
 deform_upperarm_up_base_r {parent=upperarm_twist_01_r, loc=<8.891 -0.4825 -3.9517>, rot=<0 0 0 -1>}
 deform_upperarm_up_r {parent=deform_upperarm_up_base_r, loc=<0 0 0>, rot=<0 0 0 1>}
 upperarm_twist_02_r {parent=upperarm_r, loc=<-19.7998 0.1134 0>, rot=<0 0 0 -1>}
 base_strapArm_2_r {parent=upperarm_r, loc=<-16.6736 -2.0976 -2.7788>, rot=<-0.6717 -0.1499 0.7078 0.1593>}
 dyn_strapArm_2_r {parent=base_strapArm_2_r, loc=<0 0 0>, rot=<0 0 0 1>}
 base_strapArm_1_r {parent=upperarm_r, loc=<-9.4308 -2.0451 -2.8497>, rot=<-0.6211 -0.3105 0.643 0.3231>}
 dyn_strapArm_1_r {parent=base_strapArm_1_r, loc=<0 0 0>, rot=<0 0 0 1>}
 attach_backpack {parent=spine_05, loc=<-0.123 -13.473 0>, rot=<-0.0093 -0.707 0.0093 -0.707>}
 Backpack_BR {parent=attach_backpack, type=Null, loc=<0 3.5 -0.37>, rot=<0 0 0 1>}
 attach_cape {parent=spine_05, loc=<9.017 -11.714 0>, rot=<0.0616 -0.7044 -0.0616 -0.7044>}
 cape {parent=attach_cape, type=Null, loc=<0 2.45 0.23>, rot=<0 0 0 1>}
 kin_lowerNecklace_l {parent=spine_05, loc=<6.0861 5.5517 -5.1372>, rot=<0.4685 0.6042 0.642 -0.0575>}
 kin_lowerNecklace_r {parent=spine_05, loc=<6.0857 5.5516 5.1372>, rot=<0.0575 -0.642 -0.6042 -0.4685>}
 base_lowerNecklace_1_c {parent=spine_05, loc=<5.113 6.5658 0>, rot=<0.2424 0.6643 0.6643 0.2424>}
 dyn_lowerNecklace_1_c {parent=base_lowerNecklace_1_c, loc=<0 0 0>, rot=<0 0 0 1>}
 dyn_lowerNecklace_2_c {parent=dyn_lowerNecklace_1_c, loc=<4.8453 0 0>, rot=<0 0 0 1>}
 target_lowerNecklace_1_c {parent=dyn_lowerNecklace_1_c, loc=<4.2762 0 0>, rot=<0 0 0 1>}
 base_upperNecklace_c {parent=spine_05, loc=<7.2389 5.2303 0.0381>, rot=<0.215 0.6736 0.6736 0.215>}
 dyn_upperNecklace_c {parent=base_upperNecklace_c, loc=<0 0 0>, rot=<0 0 0 1>}
 deform_pec_aim_base_l {parent=spine_04, loc=<3.2456 9.3309 -1.2237>, rot=<0.0603 -0.4033 0.3318 -0.8507>}
 deform_pec_aim_l {parent=deform_pec_aim_base_l, loc=<0 0 0>, rot=<0 0 0 1>}
 deform_pec_base_l {parent=deform_pec_aim_l, loc=<9.3665 0 0>, rot=<0 0 0 1>}
 deform_pec_l {parent=deform_pec_base_l, loc=<0 0 0>, rot=<0 0 0 1>}
 deform_pec_aim_base_r {parent=spine_04, loc=<3.2452 9.3309 1.2237>, rot=<-0.4033 -0.0603 0.8507 0.3318>}
 deform_pec_aim_r {parent=deform_pec_aim_base_r, loc=<0 0 0>, rot=<0 0 0 1>}
 deform_pec_base_r {parent=deform_pec_aim_r, loc=<-9.3667 -0.0003 -0.0004>, rot=<0 0 0 1>}
 deform_pec_r {parent=deform_pec_base_r, loc=<0 0 0>, rot=<0 0 0 1>}
 coll_breast_bot_c {parent=spine_04, loc=<8.1129 7.2033 0.0816>, rot=<-0.2378 -0.665 -0.6618 -0.2514>}
 target_breast_bot_c {parent=spine_04, loc=<4.1051 10.6241 -0.0218>, rot=<-0.2378 -0.665 -0.6618 -0.2514>}
 deform_lat_base_l {parent=spine_04, loc=<2.0524 0.7365 -9.687>, rot=<0.0294 -0.7065 -0.0294 -0.7065>}
 deform_lat_l {parent=deform_lat_base_l, loc=<0 0 0>, rot=<0 0 0 1>}
 deform_lat_base_r {parent=spine_04, loc=<2.0524 0.7365 9.687>, rot=<0.7065 0.0294 0.7065 -0.0294>}
 deform_lat_r {parent=deform_lat_base_r, loc=<0 0 0>, rot=<0 0 0 1>}
 base_vestZipper_c {parent=spine_03, loc=<-2.4405 11.1299 0>, rot=<-0.0387 0.706 0.706 -0.0387>}
 dyn_vestZipper_c {parent=base_vestZipper_c, loc=<0 0 0>, rot=<0 0 0 1>}
 base_vestZipper_l {parent=spine_03, loc=<0.3968 7.9951 -8.7548>, rot=<0.0525 -0.9575 -0.2832 0.0155>}
 dyn_vestZipper_l {parent=base_vestZipper_l, loc=<0 0 0>, rot=<0 0 0 1>}
 base_vestZipper_r {parent=spine_03, loc=<0.3968 7.9951 8.7548>, rot=<-0.0191 0.3487 0.9356 -0.0513>}
 dyn_vestZipper_r {parent=base_vestZipper_r, loc=<0 0 0>, rot=<0 0 0 1>}
 kin_spine_01_belt {parent=spine_01, loc=<3.8 0 0>, rot=<0 0 0 1>}
 thigh_l {parent=pelvis, loc=<-1.0855 -0.2109 -8.7053>, rot=<0.0009 0.0355 0.0155 -0.9993>}
 calf_l {parent=thigh_l, loc=<-42.5284 -0.0019 -0.7297>, rot=<0.0006 -0.0161 -0.0391 -0.9991>}
 foot_l {parent=calf_l, loc=<-46.7401 -2.5014 -0.592>, rot=<0.0427 -0.0238 0.0353 -0.9982>}
 ball_l {parent=foot_l, loc=<-3.5658 10.8066 -0.1544>, rot=<0.0001 -0.0001 -0.7063 -0.7079>}
 deform_ankle_out_base_l {parent=foot_l, loc=<-2.4772 -4.9805 -0.5462>, rot=<-0.0253 0.0419 -0.526 -0.8491>}
 deform_ankle_out_l {parent=deform_ankle_out_base_l, loc=<0 0 0>, rot=<0 0 0 1>}
 deform_ankle_in_base_l {parent=foot_l, loc=<3.3878 6.8115 0.747>, rot=<-0.0253 0.0419 -0.526 -0.8491>}
 deform_ankle_in_l {parent=deform_ankle_in_base_l, loc=<0 0 0>, rot=<0 0 0 1>}
 calf_twist_02_l {parent=calf_l, loc=<-19.1546 -1.7943 -0.2638>, rot=<0.0237 0.0095 -0.0035 -0.9997>}
 calf_twist_01_l {parent=calf_l, loc=<-35.8483 -1.9197 -0.5794>, rot=<0.0237 0.0095 -0.0035 -0.9997>}
 deform_knee_in_base_l {parent=calf_l, loc=<0.8031 -5.0231 0.2456>, rot=<0 0 0 -1>}
 deform_knee_in_l {parent=deform_knee_in_base_l, loc=<0 0 0>, rot=<0 0 0 1>}
 deform_knee_out_base_l {parent=calf_l, loc=<0 4 0>, rot=<0 0 0 -1>}
 deform_knee_out_l {parent=deform_knee_out_base_l, loc=<0 0 0>, rot=<0 0 0 1>}
 thigh_twist_01_l {parent=thigh_l, loc=<-27.952 0 0>, rot=<0.0013 0 0 -1>}
 deform_groin_base_l {parent=thigh_l, loc=<0.4801 5.6776 -0.2486>, rot=<0 0 0 -1>}
 deform_groin_l {parent=deform_groin_base_l, loc=<0 0 0>, rot=<0 0 0 1>}
 kin_legPouch {parent=thigh_l, loc=<-21.0092 0.5017 -6.947>, rot=<-0.0019 -0.7317 0.0015 -0.6816>}
 thigh_r {parent=pelvis, loc=<-1.0855 -0.2109 8.7053>, rot=<0.0355 -0.0009 0.9993 0.0155>}
 calf_r {parent=thigh_r, loc=<42.5284 0.0019 0.7297>, rot=<0.0006 -0.0161 -0.0391 -0.9991>}
 foot_r {parent=calf_r, loc=<46.7401 2.5014 0.592>, rot=<0.0427 -0.0238 0.0353 -0.9982>}
 ball_r {parent=foot_r, loc=<3.5658 -10.8066 0.1544>, rot=<0.0001 -0.0001 -0.7063 -0.7079>}
 deform_ankle_in_base_r {parent=foot_r, loc=<-3.3878 -6.8115 -0.747>, rot=<-0.0253 0.0419 -0.526 -0.8491>}
 deform_ankle_in_r {parent=deform_ankle_in_base_r, loc=<0 0 0>, rot=<0 0 0 1>}
 deform_ankle_out_base_r {parent=foot_r, loc=<2.4772 4.9806 0.5462>, rot=<-0.0253 0.0419 -0.526 -0.8491>}
 deform_ankle_out_r {parent=deform_ankle_out_base_r, loc=<0 0 0>, rot=<0 0 0 1>}
 calf_twist_02_r {parent=calf_r, loc=<19.1546 1.7943 0.2638>, rot=<0.0237 0.0095 -0.0035 -0.9997>}
 calf_twist_01_r {parent=calf_r, loc=<35.8483 1.9197 0.5794>, rot=<0.0237 0.0095 -0.0035 -0.9997>}
 deform_knee_in_base_r {parent=calf_r, loc=<-0.8031 5.0231 -0.2457>, rot=<0 0 0 -1>}
 deform_knee_in_r {parent=deform_knee_in_base_r, loc=<0 0 0>, rot=<0 0 0 1>}
 deform_knee_out_base_r {parent=calf_r, loc=<0 -4 0>, rot=<0 0 0 -1>}
 deform_knee_out_r {parent=deform_knee_out_base_r, loc=<0 0 0>, rot=<0 0 0 1>}
 thigh_twist_01_r {parent=thigh_r, loc=<27.952 0 0>, rot=<0.0013 0 0 -1>}
 deform_groin_base_r {parent=thigh_r, loc=<-0.4801 -5.6776 0.2486>, rot=<0 0 0 -1>}
 deform_groin_r {parent=deform_groin_base_r, loc=<0 0 0>, rot=<0 0 0 1>}
 deform_glute_base_l {parent=pelvis, loc=<-7.3466 -9.0213 -9.1245>, rot=<0.0172 0.031 -0.4477 -0.8935>}
 deform_glute_l {parent=deform_glute_base_l, loc=<0 0 0>, rot=<0 0 0 1>}
 deform_glute_base_r {parent=pelvis, loc=<-7.3464 -9.0213 9.1245>, rot=<0.031 -0.0172 0.8935 -0.4477>}
 deform_glute_r {parent=deform_glute_base_r, loc=<0 0 0>, rot=<0 0 0 1>}
 base_skirt_1_f_c {parent=pelvis, loc=<0.1953 9.0303 0>, rot=<0.0764 0.703 0.703 0.0764>}
 dyn_skirt_1_f_c {parent=base_skirt_1_f_c, loc=<0 0 0>, rot=<0 0 0 1>}
 dyn_skirt_2_f_c {parent=dyn_skirt_1_f_c, loc=<5.6994 0 0>, rot=<0 0 0 1>}
 base_skirt_1_f_l {parent=pelvis, loc=<0.9579 6.8456 -9.3886>, rot=<-0.0597 -0.8328 -0.5285 -0.1534>}
 dyn_skirt_1_f_l {parent=base_skirt_1_f_l, loc=<0 0 0>, rot=<0 0 0 1>}
 dyn_skirt_2_f_l {parent=dyn_skirt_1_f_l, loc=<5.6994 0 0>, rot=<0 0 0 1>}
 base_skirt_1_s_l {parent=pelvis, loc=<1.7835 -0.8874 -14.3464>, rot=<-0.0012 -0.978 0.0039 -0.2087>}
 dyn_skirt_1_s_l {parent=base_skirt_1_s_l, loc=<0 0 0>, rot=<0 0 0 1>}
 dyn_skirt_2_s_l {parent=dyn_skirt_1_s_l, loc=<5.6994 0 0>, rot=<0 0 0 1>}
 base_skirt_1_b_l {parent=pelvis, loc=<1.8357 -9.0515 -9.4666>, rot=<0.0343 -0.8773 0.4504 -0.162>}
 dyn_skirt_1_b_l {parent=base_skirt_1_b_l, loc=<0 0 0>, rot=<0 0 0 1>}
 dyn_skirt_2_b_l {parent=dyn_skirt_1_b_l, loc=<5.6994 0 0>, rot=<0 0 0 1>}
 base_skirt_1_b_c {parent=pelvis, loc=<1.6447 -11.0625 0>, rot=<0.0935 -0.7009 0.7009 -0.0935>}
 dyn_skirt_1_b_c {parent=base_skirt_1_b_c, loc=<0 0 0>, rot=<0 0 0 1>}
 dyn_skirt_2_b_c {parent=dyn_skirt_1_b_c, loc=<5.6994 0 0>, rot=<0 0 0 1>}
 base_skirt_1_f_r {parent=pelvis, loc=<0.9579 6.8456 9.3886>, rot=<0.1534 0.5285 0.8328 0.0597>}
 dyn_skirt_1_f_r {parent=base_skirt_1_f_r, loc=<0 0 0>, rot=<0 0 0 1>}
 dyn_skirt_2_f_r {parent=dyn_skirt_1_f_r, loc=<5.6994 0 0>, rot=<0 0 0 1>}
 coll_skirt_1 {parent=base_skirt_1_f_r, loc=<3.3278 -5.6888 -0.9414>, rot=<-0.1209 -0.0281 0.0206 -0.9921>}
 coll_skirt_2 {parent=base_skirt_1_f_r, loc=<3.1997 -0.8572 -0.9369>, rot=<-0.0626 -0.0096 -0.0251 -0.9977>}
 coll_skirt_3 {parent=base_skirt_1_f_r, loc=<2.7153 3.8482 -1.0411>, rot=<0.0553 0.0146 -0.1239 -0.9906>}
 target_skirt_1 {parent=base_skirt_1_f_r, loc=<11.8572 -5.9801 -1.4608>, rot=<-0.1209 -0.0281 0.0206 -0.9921>}
 target_skirt_2 {parent=base_skirt_1_f_r, loc=<11.8189 -0.4139 -1.0748>, rot=<-0.0626 -0.0096 -0.0251 -0.9977>}
 target_skirt_3 {parent=base_skirt_1_f_r, loc=<10.4093 5.81 -0.9205>, rot=<0.0553 0.0146 -0.1239 -0.9906>}
 base_skirt_1_s_r {parent=pelvis, loc=<1.7835 -0.8874 14.3464>, rot=<0.2087 -0.0039 0.978 0.0012>}
 dyn_skirt_1_s_r {parent=base_skirt_1_s_r, loc=<0 0 0>, rot=<0 0 0 1>}
 dyn_skirt_2_s_r {parent=dyn_skirt_1_s_r, loc=<5.6994 0 0>, rot=<0 0 0 1>}
 base_skirt_1_b_r {parent=pelvis, loc=<1.8358 -9.0515 9.4666>, rot=<0.162 -0.4504 0.8773 -0.0343>}
 dyn_skirt_1_b_r {parent=base_skirt_1_b_r, loc=<0 0 0>, rot=<0 0 0 1>}
 dyn_skirt_2_b_r {parent=dyn_skirt_1_b_r, loc=<5.6994 0 0>, rot=<0 0 0 1>}
 kin_belt_1 {parent=pelvis, loc=<10.3626 -5.6816 -9.6148>, rot=<-0.6214 -0.6505 0.2305 -0.371>}
 kin_belt_2 {parent=pelvis, loc=<10.055 -0.1487 -11.9652>, rot=<-0.6776 -0.7173 -0.1134 -0.1163>}
 kin_belt_3 {parent=pelvis, loc=<7.6554 7.7322 -6.4631>, rot=<-0.4777 -0.6344 -0.5144 0.3236>}
 base_hangBelt_main_1 {parent=pelvis, loc=<5.5004 0.3333 -0.89>, rot=<0.3995 0.5834 0.5974 -0.3784>}
 dyn_hangBelt_main_1 {parent=base_hangBelt_main_1, loc=<0 0 0>, rot=<0 0 0 1>}
 dyn_hangBelt_main_2 {parent=dyn_hangBelt_main_1, loc=<11.3526 0 0>, rot=<0 0 -0.0647 -0.9979>}
 kin_hangBelt_main_2_F {parent=dyn_hangBelt_main_2, loc=<0.724 -0.0057 7.584>, rot=<0 0 0.0133 -0.9999>}
 base_hangBelt_1 {parent=kin_hangBelt_main_2_F, loc=<0.2591 -1.5375 0.7971>, rot=<0.1676 0.018 0.5671 -0.8062>}
 dyn_hangBelt_1 {parent=base_hangBelt_1, loc=<0 0 0>, rot=<0 0 0 1>}
 dyn_hangBelt_2 {parent=dyn_hangBelt_1, loc=<6.8953 -0.1136 0.039>, rot=<0.0001 0.0172 0.0046 -0.9998>}
 dyn_hangBelt_3 {parent=dyn_hangBelt_2, loc=<5.6952 0.0001 0.1114>, rot=<0 0 0 1>}
 kin_hangBelt_main_2_B {parent=dyn_hangBelt_main_2, loc=<0.0945 -0.9454 -12.3699>, rot=<0 0 0.0133 -0.9999>}
 kin_hangBelt_main_1_F {parent=dyn_hangBelt_main_1, loc=<0.3335 -0.1044 8.8926>, rot=<0 0 -0.0515 -0.9987>}
 kin_hangBelt_main_1_B {parent=dyn_hangBelt_main_1, loc=<0.057 -0.9557 -11.6614>, rot=<0 0 -0.0515 -0.9987>}
 base_keyRing {parent=pelvis, loc=<-1.9016 9.2999 13.1284>, rot=<0.1888 0.4564 0.8574 0.1445>}
 dyn_keyRing {parent=base_keyRing, loc=<0 0 0>, rot=<0 0 0 1>}
]
variables [
 RBAN enabled {type=bool, default=False}
 c spine02RelPelvis rz {type=double, default=0}
 l thighLocal ry float {type=double, default=0}
 l thighLocal rz float {type=double, default=0}
 onGrindRail {type=bool, default=False}
 r thighLocal ry float {type=double, default=0}
 r thighLocal rz float {type=double, default=0}
]
