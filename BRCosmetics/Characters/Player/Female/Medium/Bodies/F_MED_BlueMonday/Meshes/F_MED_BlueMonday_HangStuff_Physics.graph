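# Physics asset for the hanging accessories on the Blue Monday medium female player body: a hanging belt and a key ring.
# Three box bodies form the belt chain from a small base capsule, and a key-ring capsule hangs from its own base capsule.
# Angular limits restrict swing on all four joints and allow zero twist.
# Unconstrained capsules on skirt bones add collision shapes, including one on the front-right dynamic skirt bone.
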PhysicsAsset /BRCosmetics/Characters/Player/Female/Medium/Bodies/F_MED_BlueMonday/Meshes/F_MED_BlueMonday_HangStuff_Physics
bodies {
 base_hangBelt_1 [{shapeType=Capsule, length=1, radius=0.5, shapeName=base_hangBelt_1_capsule}]
 base_keyRing [{shapeType=Capsule, length=1, radius=0.5, shapeName=base_hangBelt_1_capsule}]
 coll_skirt_1 [{
   shapeType=Capsule
   length=5.6996
   radius=5.2359
   shapeName=coll_skirt_1_capsule
   center=<4.4805 -0.5742 -2.3852>
   rotation=<rot -82.3181 -0.0002 -4.8689>
  }]
 coll_skirt_2 [
  {
   shapeType=Capsule
   length=9.1014
   radius=4.0824
   shapeName=coll_skirt_2_capsule
   center=<4.5525 -0.4183 -1.4203>
   rotation=<rot -90 -10.9138 0>
  }
  {shapeType=Capsule, length=7.8139, radius=4.4713, shapeName=coll_skirt_2_capsuleA, center=<4.3826 0.771 -2.2754>, rotation=<rot -90 0 0>}
 ]
 coll_skirt_3 [{
   shapeType=Capsule
   length=10.1276
   radius=4.0824
   shapeName=coll_skirt_3_capsule
   center=<2.9451 0.6814 -1.836>
   rotation=<rot -90 -5.2386 0>
  }]
 dyn_hangBelt_1 [{shapeType=Box, extentX=5.9752, extentY=3.1621, extentZ=1, shapeName=dyn_hangBelt_1_box, center=<3.1946 0 0>}]
 dyn_hangBelt_2 [{shapeType=Box, extentX=5.9752, extentY=3.1621, extentZ=1, shapeName=dyn_hangBelt_1_box, center=<3.1946 0 0>}]
 dyn_hangBelt_3 [{shapeType=Box, extentX=5.9752, extentY=3.1621, extentZ=1, shapeName=dyn_hangBelt_1_box, center=<3.1946 0 0>}]
 dyn_keyRing [{
   shapeType=Capsule
   length=4.2629
   radius=2.5968
   shapeName=dyn_keyRing_capsule
   center=<2.969 -0.3274 1.0825>
   rotation=<rot 0.0001 0 96.9666>
  }]
 dyn_skirt_1_f_r [{
   shapeType=Capsule
   length=6.1069
   radius=8.6018
   shapeName=dyn_skirt_1_f_r_capsule
   center=<2.7137 -0.3752 -8.3527>
   rotation=<rot 1.1419 97.9083 -89.9441>
  }]
}
constraints [
 {bone1Name=dyn_hangBelt_1, bone2Name=base_hangBelt_1, swing1LimitDegrees=15, swing2LimitDegrees=20, twistLimitDegrees=0}
 {bone1Name=dyn_hangBelt_2, bone2Name=dyn_hangBelt_1, swing1LimitDegrees=15, swing2LimitDegrees=20, twistLimitDegrees=0}
 {bone1Name=dyn_hangBelt_3, bone2Name=dyn_hangBelt_2, swing1LimitDegrees=15, swing2LimitDegrees=20, twistLimitDegrees=0}
 {bone1Name=dyn_keyRing, bone2Name=base_keyRing, swing1LimitDegrees=20, swing2LimitDegrees=20, twistLimitDegrees=0}
]
