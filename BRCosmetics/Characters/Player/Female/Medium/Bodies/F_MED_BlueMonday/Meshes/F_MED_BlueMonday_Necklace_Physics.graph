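# Physics asset for the necklace on the F_MED_BlueMonday female player body, defining collision and joints for its upper and lower chains.
# Small capsule base bodies link to spine_05 with free angular motion and anchor the dynamic necklace bodies.
# The upper pendant and two-segment lower chain use boxes, plus an upper capsule, with tight swing limits and zero twist limit.
# Larger capsules on spine_04, spine_05 and the lower breast bone supply torso collision shapes.
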PhysicsAsset /BRCosmetics/Characters/Player/Female/Medium/Bodies/F_MED_BlueMonday/Meshes/F_MED_BlueMonday_Necklace_Physics
bodies {
 base_lowerNecklace_1_c [{shapeType=Capsule, length=0.5, radius=0.25, shapeName=base_lowerNecklace_1_c_capsule}]
 base_upperNecklace_c [{shapeType=Capsule, length=0.5, radius=0.25, shapeName=base_upperNecklace_c_capsule}]
 coll_breast_bot_c [{shapeType=Capsule, length=3.8967, radius=2.9314, shapeName=coll_breast_bot_c_capsule, center=<1.9451 0 -0.9835>, rotation=<rot 90 0 0>}]
 dyn_lowerNecklace_1_c [{shapeType=Box, extentX=2.6959, extentY=7.5, extentZ=0.7637, shapeName=dyn_lowerNecklace_1_c_box, center=<2.8459 0 0>}]
 dyn_lowerNecklace_2_c [{
   shapeType=Box
   extentX=4.6448
   extentY=2.4454
   extentZ=1.6724
   shapeName=dyn_lowerNecklace_2_c_box
   center=<1.5303 0 0.703>
   rotation=<rot 1.581 0 0>
  }]
 dyn_upperNecklace_c [
  {
   shapeType=Capsule
   length=0.505
   radius=0.505
   shapeName=dyn_upperNecklace_c_capsule
   center=<2.9992 0.0385 0.129>
   rotation=<rot 0 85.4111 -83.7645>
  }
  {shapeType=Box, extentX=4.2165, extentY=2.6089, extentZ=0.8455, shapeName=dyn_upperNecklace_c_box, center=<3.3456 0 0.0213>}
 ]
 spine_04 [
  {shapeType=Capsule, length=7.1143, radius=8.1275, shapeName=spine_05_capsule, center=<0.7463 3.9152 0>}
  {shapeType=Capsule, length=3.8048, radius=3.7602, shapeName=spine_04_capsule, center=<2.6439 8.2139 0>}
 ]
 spine_05 [
  {shapeType=Capsule, length=10.251, radius=8, shapeName=spine_05_capsule, center=<0.2311 -0.197 0>}
  {shapeType=Capsule, length=14.2893, radius=4.9698, shapeName=spine_05_capsule, center=<-1.6093 4.1681 0>, rotation=<rot 0 -0.1809 0>}
 ]
}
constraints [
 {bone1Name=dyn_lowerNecklace_2_c, bone2Name=dyn_lowerNecklace_1_c, swing1LimitDegrees=8, swing2LimitDegrees=20, twistLimitDegrees=0}
 {bone1Name=dyn_upperNecklace_c, bone2Name=base_upperNecklace_c, swing1LimitDegrees=10, swing2LimitDegrees=20, twistLimitDegrees=0}
 {bone1Name=dyn_lowerNecklace_1_c, bone2Name=base_lowerNecklace_1_c, swing1LimitDegrees=10, swing2LimitDegrees=25, twistLimitDegrees=0}
 {
  bone1Name=base_lowerNecklace_1_c
  bone2Name=spine_05
  swing1LimitDegrees=45
  swing1Motion=Free
  swing2LimitDegrees=45
  swing2Motion=Free
  twistLimitDegrees=45
  twistMotion=Free
 }
 {
  bone1Name=base_upperNecklace_c
  bone2Name=spine_05
  swing1LimitDegrees=45
  swing1Motion=Free
  swing2LimitDegrees=45
  swing2Motion=Free
  twistLimitDegrees=45
  twistMotion=Free
 }
 {
  bone1Name=spine_05
  bone2Name=spine_04
  swing1LimitDegrees=45
  swing1Motion=Free
  swing2LimitDegrees=45
  swing2Motion=Free
  twistLimitDegrees=45
  twistMotion=Free
 }
 {
  bone1Name=coll_breast_bot_c
  bone2Name=spine_04
  swing1LimitDegrees=45
  swing1Motion=Free
  swing2LimitDegrees=45
  swing2Motion=Free
  twistLimitDegrees=45
  twistMotion=Free
 }
]
