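# Animation Blueprint for the BlueMonday female medium cosmetic body, adding secondary motion on top of the pose copied from its attached parent mesh.
# Facial bone resets feed a Control Rig, then additive bone offsets and AnimDynamics for zippers, straps and skirt.
# RigidBody necklace physics switches assets in the front end.
# Weighted constraints and look-ats position skirt collision bones for hanging-item physics, before final necklace and skirt fixes.
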AnimBlueprint /BRCosmetics/Characters/Player/Female/Medium/Bodies/F_MED_BlueMonday/Meshes/F_MED_BlueMonday_AnimBP
target_skeleton=&/F_MED_BlueMonday_Skeleton
graphs [
 AnimGraph {nodes [
   Output Pose(Result=Use cached pose 'postSimFixes'()) {layerGroup=None}
   n1 = "Transform (Modify) Bone - Bone: FACIAL_C_UnderChin"!(
    ComponentPose=Local To Component(LocalPose=Copy Pose From Mesh() {bUseAttachedParent=true})
    bAlphaBoolEnabled:bool=True
    Translation="0,0,0"
    Rotation="0,0,0"
   ) -> (Pose) {
    alphaInputType=Bool
    rotationMode=BMM_Replace
    rotationSpace=BCS_ParentBoneSpace
    translationMode=BMM_Replace
    translationSpace=BCS_ParentBoneSpace
    alphaBoolBlend {blendOption=Linear}
    boneToModify {boneName=FACIAL_C_UnderChin}
   }
   n2 = "Transform (Modify) Bone - Bone: FACIAL_L_UnderChin"!(ComponentPose=@n1, bAlphaBoolEnabled:bool=True, Translation="0,0,0", Rotation="0,0,0") -> (Pose) {
    alphaInputType=Bool
    rotationMode=BMM_Replace
    rotationSpace=BCS_ParentBoneSpace
    translationMode=BMM_Replace
    translationSpace=BCS_ParentBoneSpace
    alphaBoolBlend {blendOption=Linear}
    boneToModify {boneName=FACIAL_L_UnderChin}
   }
   n3 = "Transform (Modify) Bone - Bone: FACIAL_R_UnderChin"!(ComponentPose=@n2, bAlphaBoolEnabled:bool=True, Translation="0,0,0", Rotation="0,0,0") -> (Pose) {
    alphaInputType=Bool
    rotationMode=BMM_Replace
    rotationSpace=BCS_ParentBoneSpace
    translationMode=BMM_Replace
    translationSpace=BCS_ParentBoneSpace
    alphaBoolBlend {blendOption=Linear}
    boneToModify {boneName=FACIAL_R_UnderChin}
   }
   n4 = "Transform (Modify) Bone - Bone: FACIAL_C_Jaw"!(ComponentPose=@n3, bAlphaBoolEnabled:bool=True, Translation="0,0,0", Rotation="0,0,0") -> (Pose) {
    alphaInputType=Bool
    rotationMode=BMM_Replace
    rotationSpace=BCS_ParentBoneSpace
    translationMode=BMM_Replace
    translationSpace=BCS_ParentBoneSpace
    alphaBoolBlend {blendOption=Linear}
    boneToModify {boneName=FACIAL_C_Jaw}
   }
   n5 = "Transform (Modify) Bone - Bone: FACIAL_L_Ear"!(ComponentPose=@n4, bAlphaBoolEnabled:bool=True, Translation="0,0,0", Rotation="0,0,0") -> (Pose) {
    alphaInputType=Bool
    rotationMode=BMM_Replace
    rotationSpace=BCS_ParentBoneSpace
    translationMode=BMM_Replace
    translationSpace=BCS_ParentBoneSpace
    alphaBoolBlend {blendOption=Linear}
    boneToModify {boneName=FACIAL_L_Ear}
   }
   n6 = "Transform (Modify) Bone - Bone: FACIAL_R_Ear"!(ComponentPose=@n5, bAlphaBoolEnabled:bool=True, Translation="0,0,0", Rotation="0,0,0") -> (Pose) {
    alphaInputType=Bool
    rotationMode=BMM_Replace
    rotationSpace=BCS_ParentBoneSpace
    translationMode=BMM_Replace
    translationSpace=BCS_ParentBoneSpace
    alphaBoolBlend {blendOption=Linear}
    boneToModify {boneName=FACIAL_R_Ear}
   }
   n7 = "Transform (Modify) Bone - Bone: FACIAL_C_Jawline"!(ComponentPose=@n6, bAlphaBoolEnabled:bool=True, Translation="0,0,0", Rotation="0,0,0") -> (Pose) {
    alphaInputType=Bool
    rotationMode=BMM_Replace
    rotationSpace=BCS_ParentBoneSpace
    translationMode=BMM_Replace
    translationSpace=BCS_ParentBoneSpace
    alphaBoolBlend {blendOption=Linear}
    boneToModify {boneName=FACIAL_C_Jawline}
   }
   n8 = "Transform (Modify) Bone - Bone: FACIAL_L_Jawline"!(ComponentPose=@n7, bAlphaBoolEnabled:bool=True, Translation="0,0,0", Rotation="0,0,0") -> (Pose) {
    alphaInputType=Bool
    rotationMode=BMM_Replace
    rotationSpace=BCS_ParentBoneSpace
    translationMode=BMM_Replace
    translationSpace=BCS_ParentBoneSpace
    alphaBoolBlend {blendOption=Linear}
    boneToModify {boneName=FACIAL_L_Jawline}
   }
   n9 = "Transform (Modify) Bone - Bone: FACIAL_R_Jawline"!(ComponentPose=@n8, bAlphaBoolEnabled:bool=True, Translation="0,0,0", Rotation="0,0,0") -> (Pose) {
    alphaInputType=Bool
    rotationMode=BMM_Replace
    rotationSpace=BCS_ParentBoneSpace
    translationMode=BMM_Replace
    translationSpace=BCS_ParentBoneSpace
    alphaBoolBlend {blendOption=Linear}
    boneToModify {boneName=FACIAL_R_Jawline}
   }
   n10 = Control Rig!(Source=Component To Local(ComponentPose=@n9), Alpha, onGrindRail:bool=Get bIsOnGrindRail(), RBAN enabled:bool=Get bRigidBodyEnabled()) -> (Pose) {controlRigAssetReference {blueprintRigClass=&/F_MED_BlueMonday_CtrlRig.F_MED_BlueMonday_CtrlRig_C}}
   Save cached pose 'controlRigCP'(Pose=@n10) {cacheName=controlRigCP}
   n11 = RigidBody!(ComponentPose=Local To Component(LocalPose=Use cached pose 'animDynamicsCP'()), OverrideWorldGravity:Vector="0,0,0") -> (Pose) {
    bOverrideWorldGravity=true
    overridePhysicsAsset=&/F_MED_BlueMonday_Necklace_Physics
    simulationSpace=BaseBoneSpace
    alphaBoolBlend {blendOption=Linear}
    baseBoneRef {boneName=spine_05}
    componentAppliedLinearAccClamp=<3000 3000 3000>
    overrideWorldGravity {z=-980}
    simSpaceSettings {
     dampingAlpha=0
     maxAngularAcceleration=1
     maxAngularVelocity=3
     maxLinearAcceleration=1500
     maxLinearVelocity=500
     worldAlpha=1
     externalLinearDragV=<0.2 0.2 0.2>
    }
   }
   n12 = RigidBody(ComponentPose=Local To Component(LocalPose=Use cached pose 'animDynamicsCP'()), OverrideWorldGravity="0,0,0") {
    bOverrideWorldGravity=true
    overridePhysicsAsset=&/F_MED_BlueMonday_NecklaceFE_Physics
    simulationSpace=BaseBoneSpace
    alphaBoolBlend {blendOption=Linear}
    baseBoneRef {boneName=spine_05}
    componentAppliedLinearAccClamp=<3000 3000 3000>
    overrideWorldGravity {z=-980}
    simSpaceSettings {
     dampingAlpha=0
     maxAngularAcceleration=1
     maxAngularVelocity=3
     maxLinearAcceleration=1500
     maxLinearVelocity=500
     worldAlpha=1
     externalLinearDragV=<0.2 0.2 0.2>
    }
   }
   Save cached pose 'dynamicsCP'(Pose=Blend Poses by bool(BlendPose_0=Component To Local(ComponentPose=@n12), BlendPose_1=Component To Local(ComponentPose=@n11), bActiveValue=Get bIsInFrontEnd())) {cacheName=dynamicsCP}
   n13 = Look At - Bone: kin_lowerNecklace_l(ComponentPose=Local To Component(LocalPose=Use cached pose 'PostDynamicsCP'()), Alpha=1) {
    boneToModify {boneName=kin_lowerNecklace_l}
    lookAtLocation {x=0}
    lookAtTarget {boneReference {boneName=target_lowerNecklace_1_c}}
    lookAt_Axis {axis=<1 0>}
    lookUp_Axis {axis {x=0, z=1}}
   }
   n14 = Look At - Bone: kin_lowerNecklace_r(ComponentPose=@n13, Alpha=1) {
    boneToModify {boneName=kin_lowerNecklace_r}
    lookAtLocation {x=0}
    lookAtTarget {boneReference {boneName=target_lowerNecklace_1_c}}
    lookAt_Axis {axis=<1 0>}
    lookUp_Axis {axis {x=0, z=1}}
   }
   n15 = "Transform (Modify) Bone - Bone: dyn_skirt_1_b_r"!(ComponentPose=@n14, bAlphaBoolEnabled:bool=True, Translation="0,0,0", Rotation="-12,0,0") -> (Pose) {
    alphaInputType=Bool
    rotationMode=BMM_Additive
    rotationSpace=BCS_BoneSpace
    translationMode=BMM_Additive
    translationSpace=BCS_BoneSpace
    alphaBoolBlend {blendInTime=0.5, blendOption=Cubic, blendOutTime=0.5}
    boneToModify {boneName=dyn_skirt_1_b_r}
    rotation {pitch=-12}
   }
   n16 = "Transform (Modify) Bone - Bone: dyn_skirt_1_b_l"!(ComponentPose=@n15, bAlphaBoolEnabled:bool=True, Translation="0,0,0", Rotation="-12,0,0") -> (Pose) {
    alphaInputType=Bool
    rotationMode=BMM_Additive
    rotationSpace=BCS_BoneSpace
    translationMode=BMM_Additive
    translationSpace=BCS_BoneSpace
    alphaBoolBlend {blendInTime=0.5, blendOption=Cubic, blendOutTime=0.5}
    boneToModify {boneName=dyn_skirt_1_b_l}
    rotation {pitch=-12}
   }
   n17 = "Transform (Modify) Bone - Bone: dyn_skirt_1_b_c"!(ComponentPose=@n16, bAlphaBoolEnabled:bool=True, Translation="0,0,0", Rotation="-12,0,0") -> (Pose) {
    alphaInputType=Bool
    rotationMode=BMM_Additive
    rotationSpace=BCS_BoneSpace
    translationMode=BMM_Additive
    translationSpace=BCS_BoneSpace
    alphaBoolBlend {blendInTime=0.5, blendOption=Cubic, blendOutTime=0.5}
    boneToModify {boneName=dyn_skirt_1_b_c}
    rotation {pitch=-12}
   }
   Save cached pose 'postSimFixes'(Pose=Component To Local(ComponentPose=@n17)) {cacheName=postSimFixes}
   n18 = "Constraint - attach_cape (P)"(ComponentPose=Local To Component(LocalPose=Use cached pose 'controlRigCP'()), Alpha=1, ConstraintWeights_0=1) {
    alphaBoolBlend {blendOption=Linear}
    boneToModify {boneName=attach_cape}
    constraintSetup [{targetBone {boneName=spine_05}}]
    constraintWeights [1]
   }
   n19 = "Transform (Modify) Bone - Bone: attach_cape"!(ComponentPose=@n18, bAlphaBoolEnabled:bool=True, Translation="0,-1,0", Rotation="0,0,16") -> (Pose) {
    alphaInputType=Bool
    rotationMode=BMM_Additive
    rotationSpace=BCS_BoneSpace
    translationMode=BMM_Additive
    translationSpace=BCS_BoneSpace
    alphaBoolBlend {blendInTime=0.5, blendOption=CircularOut, blendOutTime=0.5}
    boneToModify {boneName=attach_cape}
    rotation {roll=16}
    translation {y=-1}
   }
   n20 = "Transform (Modify) Bone - Bone: attach_cape"(ComponentPose=@n19, bAlphaBoolEnabled=True, Translation="0,-2,0", Rotation="0,0,20") {
    alphaInputType=Bool
    rotationMode=BMM_Additive
    rotationSpace=BCS_BoneSpace
    translationMode=BMM_Additive
    translationSpace=BCS_BoneSpace
    alphaBoolBlend {blendInTime=0.5, blendOption=CircularOut, blendOutTime=0.5}
    boneToModify {boneName=attach_cape}
    rotation {roll=20}
    translation {y=-2}
   }
   n21 = "Transform (Modify) Bone - Bone: attach_cape"(ComponentPose=@n20, bAlphaBoolEnabled=True, Translation="0,0,0", Rotation="0,0,10") {
    alphaInputType=Bool
    rotationMode=BMM_Additive
    rotationSpace=BCS_BoneSpace
    translationMode=BMM_Additive
    translationSpace=BCS_BoneSpace
    alphaBoolBlend {blendInTime=0.5, blendOption=Cubic, blendOutTime=0.5}
    boneToModify {boneName=attach_cape}
    rotation {roll=10}
   }
   n22 = "Transform (Modify) Bone - Bone: base_hangBelt_1"!(ComponentPose=@n21, bAlphaBoolEnabled:bool=True, Translation="0.8,1.7,0.8", Rotation="0,-30,0") -> (Pose) {
    alphaInputType=Bool
    rotationMode=BMM_Additive
    rotationSpace=BCS_BoneSpace
    translationMode=BMM_Additive
    translationSpace=BCS_BoneSpace
    alphaBoolBlend {blendInTime=0.5, blendOption=Cubic, blendOutTime=0.5}
    boneToModify {boneName=base_hangBelt_1}
    rotation {yaw=-30}
    translation=<0.8 1.7 0.8>
   }
   n23 = "Transform (Modify) Bone - Bone: base_lowerNecklace_1_c"!(ComponentPose=@n22, bAlphaBoolEnabled:bool=True, Translation="2,0,2", Rotation="50,-0.008,0") -> (Pose) {
    alphaInputType=Bool
    rotationMode=BMM_Additive
    rotationSpace=BCS_BoneSpace
    translationMode=BMM_Additive
    translationSpace=BCS_BoneSpace
    alphaBoolBlend {blendInTime=0.5, blendOption=Cubic, blendOutTime=0.5}
    boneToModify {boneName=base_lowerNecklace_1_c}
    rotation {pitch=50, yaw=-0.008}
    translation {x=2, z=2}
   }
   n24 = "Transform (Modify) Bone - Bone: base_upperNecklace_c"!(ComponentPose=@n23, bAlphaBoolEnabled:bool=True, Translation="1,0,1", Rotation="50,-0.008,-0.018") -> (Pose) {
    alphaInputType=Bool
    rotationMode=BMM_Additive
    rotationSpace=BCS_BoneSpace
    translationMode=BMM_Additive
    translationSpace=BCS_BoneSpace
    alphaBoolBlend {blendInTime=0.5, blendOption=Cubic, blendOutTime=0.5}
    boneToModify {boneName=base_upperNecklace_c}
    rotation=<rot 50 -0.008 -0.018>
    translation {x=1, z=1}
   }
   n25 = "Transform (Modify) Bone - Bone: upperarm_twist_01_l"!(ComponentPose=@n24, bAlphaBoolEnabled:bool=True, Translation="0,0,0", Rotation="0,0,0", Scale="1,1,0.85") -> (Pose) {
    alphaInputType=Bool
    rotationMode=BMM_Additive
    rotationSpace=BCS_BoneSpace
    scaleMode=BMM_Additive
    scaleSpace=BCS_BoneSpace
    translationMode=BMM_Additive
    translationSpace=BCS_BoneSpace
    alphaBoolBlend {blendInTime=0.5, blendOption=Cubic, blendOutTime=0.5}
    boneToModify {boneName=upperarm_twist_01_l}
    scale {z=0.85}
   }
   n26 = "Transform (Modify) Bone - Bone: upperarm_twist_01_l"(ComponentPose=@n25, bAlphaBoolEnabled=True, Translation="0,0,1", Rotation="0,0,0", Scale="1,1,0.9") {
    alphaInputType=Bool
    rotationMode=BMM_Additive
    rotationSpace=BCS_BoneSpace
    scaleMode=BMM_Additive
    scaleSpace=BCS_BoneSpace
    translationMode=BMM_Additive
    translationSpace=BCS_BoneSpace
    alphaBoolBlend {blendInTime=0.5, blendOption=Cubic, blendOutTime=0.5}
    boneToModify {boneName=upperarm_twist_01_l}
    scale {z=0.9}
    translation {z=1}
   }
   n27 = "Transform (Modify) Bone - Bone: upperarm_twist_01_r"!(ComponentPose=@n26, bAlphaBoolEnabled:bool=True, Translation="0,0,-1", Rotation="0,0,0", Scale="1,1,0.9") -> (Pose) {
    alphaInputType=Bool
    rotationMode=BMM_Additive
    rotationSpace=BCS_BoneSpace
    scaleMode=BMM_Additive
    scaleSpace=BCS_BoneSpace
    translationMode=BMM_Additive
    translationSpace=BCS_BoneSpace
    alphaBoolBlend {blendInTime=0.5, blendOption=Cubic, blendOutTime=0.5}
    boneToModify {boneName=upperarm_twist_01_r}
    scale {z=0.9}
    translation {z=-1}
   }
   Save cached pose 'generalAdjuistementsCP'(Pose=Component To Local(ComponentPose=@n27)) {cacheName=generalAdjuistementsCP}
   n28 = "Constraint - coll_skirt_1 (PPPPPPPP)"!(
    ComponentPose=Local To Component(LocalPose=Use cached pose 'dynamicsCP'())
    Alpha=1
    ConstraintWeights_0=0.3207
    ConstraintWeights_1:float=0.3073
    ConstraintWeights_2:float=0.1033
    ConstraintWeights_3:float=0.0872
    ConstraintWeights_4:float=0.0863
    ConstraintWeights_5:float=0.0427
    ConstraintWeights_6:float=0.0386
    ConstraintWeights_7:float=0.014
   ) -> (Pose) {
    boneToModify {boneName=coll_skirt_1}
    constraintSetup [
     {targetBone {boneName=dyn_skirt_1_f_r}}
     {targetBone {boneName=dyn_skirt_1_f_c}}
     {targetBone {boneName=pelvis}}
     {targetBone {boneName=dyn_skirt_2_f_c}}
     {targetBone {boneName=dyn_skirt_2_f_r}}
     {targetBone {boneName=deform_groin_r}}
     {targetBone {boneName=thigh_twist_01_r}}
     {targetBone {boneName=dyn_skirt_1_f_l}}
    ]
    constraintWeights [1,1,1,1,1,1,1,1]
   }
   n29 = "Constraint - coll_skirt_2 (PPPPPPPP)"!(
    ComponentPose=@n28
    Alpha=1
    ConstraintWeights_0=0.3638
    ConstraintWeights_1:float=0.2032
    ConstraintWeights_2:float=0.1437
    ConstraintWeights_3:float=0.1372
    ConstraintWeights_4:float=0.064
    ConstraintWeights_5:float=0.0517
    ConstraintWeights_6:float=0.0198
    ConstraintWeights_7:float=0.0166
   ) -> (Pose) {
    boneToModify {boneName=coll_skirt_2}
    constraintSetup [
     {targetBone {boneName=dyn_skirt_1_f_r}}
     {targetBone {boneName=pelvis}}
     {targetBone {boneName=deform_groin_r}}
     {targetBone {boneName=thigh_twist_01_r}}
     {targetBone {boneName=dyn_skirt_1_s_r}}
     {targetBone {boneName=dyn_skirt_2_f_r}}
     {targetBone {boneName=dyn_skirt_1_f_c}}
     {targetBone {boneName=spine_01}}
    ]
    constraintWeights [1,1,1,1,1,1,1,1]
   }
   n30 = "Constraint - coll_skirt_3 (PPPPPPPP)"!(
    ComponentPose=@n29
    Alpha=1
    ConstraintWeights_0=0.4573
    ConstraintWeights_1:float=0.1664
    ConstraintWeights_2:float=0.1417
    ConstraintWeights_3:float=0.0764
    ConstraintWeights_4:float=0.0538
    ConstraintWeights_5:float=0.0482
    ConstraintWeights_6:float=0.0449
    ConstraintWeights_7:float=0.0113
   ) -> (Pose) {
    boneToModify {boneName=coll_skirt_3}
    constraintSetup [
     {targetBone {boneName=dyn_skirt_1_f_r}}
     {targetBone {boneName=dyn_skirt_1_s_r}}
     {targetBone {boneName=dyn_skirt_2_f_r}}
     {targetBone {boneName=thigh_twist_01_r}}
     {targetBone {boneName=pelvis}}
     {targetBone {boneName=dyn_skirt_2_s_r}}
     {targetBone {boneName=deform_groin_r}}
     {targetBone {boneName=dyn_skirt_1_f_c}}
    ]
    constraintWeights [1,1,1,1,1,1,1,1]
   }
   n31 = "Constraint - target_skirt_1 (PPPP)"!(
    ComponentPose=@n30
    Alpha=1
    ConstraintWeights_0=0.3925
    ConstraintWeights_1:float=0.2597
    ConstraintWeights_2:float=0.2067
    ConstraintWeights_3:float=0.141
   ) -> (Pose) {
    boneToModify {boneName=target_skirt_1}
    constraintSetup [
     {targetBone {boneName=dyn_skirt_2_f_r}}
     {targetBone {boneName=dyn_skirt_2_f_c}}
     {targetBone {boneName=dyn_skirt_1_f_r}}
     {targetBone {boneName=dyn_skirt_1_f_c}}
    ]
    constraintWeights [1,1,1,1]
   }
   n32 = "Constraint - target_skirt_2 (PPPPPP)"!(
    ComponentPose=@n31
    Alpha=1
    ConstraintWeights_0=0.5651
    ConstraintWeights_1:float=0.2647
    ConstraintWeights_2:float=0.0849
    ConstraintWeights_3:float=0.0459
    ConstraintWeights_4:float=0.0251
    ConstraintWeights_5:float=0.0142
   ) -> (Pose) {
    boneToModify {boneName=target_skirt_2}
    constraintSetup [
     {targetBone {boneName=dyn_skirt_2_f_r}}
     {targetBone {boneName=dyn_skirt_1_f_r}}
     {targetBone {boneName=dyn_skirt_2_f_c}}
     {targetBone {boneName=dyn_skirt_1_f_c}}
     {targetBone {boneName=dyn_skirt_2_s_r}}
     {targetBone {boneName=dyn_skirt_1_s_r}}
    ]
    constraintWeights [1,1,1,1,1,1]
   }
   n33 = "Constraint - target_skirt_3 (PPPP)"!(
    ComponentPose=@n32
    Alpha=1
    ConstraintWeights_0=0.4585
    ConstraintWeights_1:float=0.2564
    ConstraintWeights_2:float=0.1823
    ConstraintWeights_3:float=0.1028
   ) -> (Pose) {
    boneToModify {boneName=target_skirt_3}
    constraintSetup [
     {targetBone {boneName=dyn_skirt_2_f_r}}
     {targetBone {boneName=dyn_skirt_2_s_r}}
     {targetBone {boneName=dyn_skirt_1_f_r}}
     {targetBone {boneName=dyn_skirt_1_s_r}}
    ]
    constraintWeights [1,1,1,1]
   }
   n34 = Look At - Bone: coll_skirt_1(ComponentPose=@n33, Alpha=1) {
    boneToModify {boneName=coll_skirt_1}
    lookAtLocation {x=0}
    lookAtTarget {boneReference {boneName=target_skirt_1}}
    lookAt_Axis {axis=<1 0>}
    lookUp_Axis {axis {x=0, z=1}}
   }
   n35 = Look At - Bone: coll_skirt_2(ComponentPose=@n34, Alpha=1) {
    boneToModify {boneName=coll_skirt_2}
    lookAtLocation {x=0}
    lookAtTarget {boneReference {boneName=target_skirt_2}}
    lookAt_Axis {axis=<1 0>}
    lookUp_Axis {axis {x=0, z=1}}
   }
   n36 = Look At - Bone: coll_skirt_3(ComponentPose=@n35, Alpha=1) {
    boneToModify {boneName=coll_skirt_3}
    lookAtLocation {x=0}
    lookAtTarget {boneReference {boneName=target_skirt_3}}
    lookAt_Axis {axis=<1 0>}
    lookUp_Axis {axis {x=0, z=1}}
   }
   n37 = "Transform (Modify) Bone - Bone: dyn_hangBelt_main_1"!(ComponentPose=@n36, Alpha=1, Rotation="0,5,0") -> (Pose) {
    rotationMode=BMM_Additive
    rotationSpace=BCS_BoneSpace
    scaleMode=BMM_Additive
    scaleSpace=BCS_BoneSpace
    translationMode=BMM_Additive
    translationSpace=BCS_BoneSpace
    alphaBoolBlend {blendInTime=0.5, blendOutTime=0.5}
    boneToModify {boneName=dyn_hangBelt_main_1}
    rotation {yaw=5}
   }
   n38 = Anim Dynamics - Bone: dyn_hangBelt_main_2!(ComponentPose=@n37, Alpha=0, GravityOverride:Vector="0,0,-980") -> (Pose) {
    angularDampingOverride=0.99
    angularSpringConstant=3000
    bAngularSpring=true
    bOverrideAngularDamping=true
    bUseGravityOverride=true
    gravityScale=0.35
    simulationSpace=RootRelative
    alphaBoolBlend {blendOption=Linear}
    boundBone {boneName=dyn_hangBelt_main_2}
    componentAppliedLinearAccClamp=<1500 1500 1500>
    componentLinearAccScale=<0.5 0.5 0.4>
    componentLinearVelScale=<0.5 0.5 0.4>
    gravityOverride {z=-980}
    physicsBodyDefinitions [{
      boundBone {boneName=dyn_hangBelt_main_2}
      boxExtents=<20 20 20>
      constraintSetup {angularLimitsMax=<0 5 15>, angularLimitsMin=<0 -5 0>, angularTarget=<1 0 0>}
      localJointOffset=<15 0 0>
     }]
   }
   n39 = "Constraint - base_keyRing (PPP)"!(ComponentPose=@n38, Alpha=1, ConstraintWeights_0=0.8497, ConstraintWeights_1:float=0.1382, ConstraintWeights_2:float=0.0121) -> (Pose) {
    boneToModify {boneName=base_keyRing}
    constraintSetup [{targetBone {boneName=kin_hangBelt_main_2_F}}, {targetBone {boneName=kin_hangBelt_main_1_F}}, {targetBone {boneName=kin_hangBelt_main_2_B}}]
    constraintWeights [1,1,1]
   }
   n40 = "Transform (Modify) Bone - Bone: base_keyRing"!(ComponentPose=@n39, Alpha=1, Rotation="38.5,7,42") -> (Pose) {
    rotationMode=BMM_Additive
    rotationSpace=BCS_BoneSpace
    scaleMode=BMM_Additive
    scaleSpace=BCS_BoneSpace
    translationMode=BMM_Additive
    translationSpace=BCS_BoneSpace
    alphaBoolBlend {blendInTime=0.5, blendOutTime=0.5}
    boneToModify {boneName=base_keyRing}
    rotation=<rot 38.5 7 42>
   }
   n41 = RigidBody(ComponentPose=@n40, OverrideWorldGravity="0,0,0") {
    bOverrideWorldGravity=true
    overridePhysicsAsset=&/F_MED_BlueMonday_HangStuff_Physics
    simulationSpace=BaseBoneSpace
    alphaBoolBlend {blendOption=Linear}
    baseBoneRef {boneName=coll_skirt_2}
    componentAppliedLinearAccClamp=<3000 3000 3000>
    overrideWorldGravity {z=-980}
    simSpaceSettings {
     dampingAlpha=0
     maxAngularAcceleration=1
     maxAngularVelocity=3
     maxLinearAcceleration=1200
     maxLinearVelocity=400
     worldAlpha=1
     externalLinearDragV=<0.2 0.2 0.2>
    }
   }
   Save cached pose 'PostDynamicsCP'(Pose=Component To Local(ComponentPose=@n41)) {cacheName=PostDynamicsCP}
   n42 = Anim Dynamics - Bone: dyn_vestZipper_c!(
    ComponentPose=Local To Component(LocalPose=Use cached pose 'generalAdjuistementsCP'())
    Alpha=1
    GravityOverride:Vector="0,0,-980"
    SimSpaceSettings:Anim Phys Sim Space Settings Structure=(SimSpaceAngularAlpha=0,MaxAngularVelocity=10000,MaxAngularAcceleration=10000,ExternalAngularVelocity=(X=0,Y=0,Z=0))
   ) -> (Pose) {
    angularDampingOverride=1
    angularSpringConstant=1000
    bAngularSpring=true
    bOverrideAngularDamping=true
    bUseGravityOverride=true
    gravityScale=0.35
    simulationSpace=RootRelative
    alphaBoolBlend {blendOption=Linear}
    boundBone {boneName=dyn_vestZipper_c}
    componentAppliedLinearAccClamp=<5000 5000 50000>
    componentLinearAccScale=<0.5 0.5 0.5>
    componentLinearVelScale=<0.5 0.5 0.5>
    gravityOverride {z=-980}
    physicsBodyDefinitions [{
      boundBone {boneName=dyn_vestZipper_c}
      boxExtents=<10 10 10>
      constraintSetup {angularLimitsMax=<0 10 15>, angularLimitsMin=<0 -30 -15>, angularTarget=<1 0 0>}
      localJointOffset=<5 0 0>
     }]
   }
   n43 = Anim Dynamics - Bone: dyn_vestZipper_l!(
    ComponentPose=@n42
    Alpha=1
    GravityOverride:Vector="0,0,-980"
    SimSpaceSettings:Anim Phys Sim Space Settings Structure=(SimSpaceAngularAlpha=0,MaxAngularVelocity=10000,MaxAngularAcceleration=10000,ExternalAngularVelocity=(X=0,Y=0,Z=0))
   ) -> (Pose) {
    angularDampingOverride=1
    angularSpringConstant=1000
    bAngularSpring=true
    bOverrideAngularDamping=true
    bUseGravityOverride=true
    gravityScale=0.35
    simulationSpace=RootRelative
    alphaBoolBlend {blendOption=Linear}
    boundBone {boneName=dyn_vestZipper_l}
    componentAppliedLinearAccClamp=<5000 5000 50000>
    componentLinearAccScale=<0.5 0.5 0.5>
    componentLinearVelScale=<0.5 0.5 0.5>
    gravityOverride {z=-980}
    physicsBodyDefinitions [{
      boundBone {boneName=dyn_vestZipper_l}
      boxExtents=<10 10 10>
      constraintSetup {angularLimitsMax=<0 10 15>, angularLimitsMin=<0 -30 -15>, angularTarget=<1 0 0>}
      localJointOffset=<5 0 0>
     }]
   }
   n44 = Anim Dynamics - Bone: dyn_vestZipper_r!(
    ComponentPose=@n43
    Alpha=1
    GravityOverride:Vector="0,0,-980"
    SimSpaceSettings:Anim Phys Sim Space Settings Structure=(SimSpaceAngularAlpha=0,MaxAngularVelocity=10000,MaxAngularAcceleration=10000,ExternalAngularVelocity=(X=0,Y=0,Z=0))
   ) -> (Pose) {
    angularDampingOverride=1
    angularSpringConstant=1000
    bAngularSpring=true
    bOverrideAngularDamping=true
    bUseGravityOverride=true
    gravityScale=0.35
    simulationSpace=RootRelative
    alphaBoolBlend {blendOption=Linear}
    boundBone {boneName=dyn_vestZipper_r}
    componentAppliedLinearAccClamp=<5000 5000 50000>
    componentLinearAccScale=<0.5 0.5 0.5>
    componentLinearVelScale=<0.5 0.5 0.5>
    gravityOverride {z=-980}
    physicsBodyDefinitions [{
      boundBone {boneName=dyn_vestZipper_r}
      boxExtents=<10 10 10>
      constraintSetup {angularLimitsMax=<0 10 15>, angularLimitsMin=<0 -30 -15>, angularTarget=<1 0 0>}
      localJointOffset=<5 0 0>
     }]
   }
   n45 = Anim Dynamics - Bone: dyn_strapArm_1_l!(
    ComponentPose=@n44
    Alpha=1
    SimSpaceSettings:Anim Phys Sim Space Settings Structure=(SimSpaceAngularAlpha=0,MaxAngularVelocity=10000,MaxAngularAcceleration=10000,ExternalAngularVelocity=(X=0,Y=0,Z=0))
   ) -> (Pose) {
    angularDampingOverride=1
    angularSpringConstant=1000
    bAngularSpring=true
    bOverrideAngularDamping=true
    bUseGravityOverride=true
    gravityScale=0.35
    simulationSpace=RootRelative
    alphaBoolBlend {blendOption=Linear}
    boundBone {boneName=dyn_strapArm_1_l}
    componentAppliedLinearAccClamp=<3000 3000>
    componentLinearAccScale=<0.5 0.5 0.25>
    componentLinearVelScale=<0.5 0.5 0.25>
    gravityOverride {z=-980}
    physicsBodyDefinitions [{
      boundBone {boneName=dyn_strapArm_1_l}
      boxExtents=<20 20 20>
      constraintSetup {angularLimitsMax=<0 15 25>, angularLimitsMin=<0 -15 -25>, angularTarget=<1 0 0>}
      localJointOffset=<2 0 0>
     }]
   }
   n46 = Anim Dynamics - Bone: dyn_strapArm_2_l!(
    ComponentPose=@n45
    Alpha=1
    SimSpaceSettings:Anim Phys Sim Space Settings Structure=(SimSpaceAngularAlpha=0,MaxAngularVelocity=10000,MaxAngularAcceleration=10000,ExternalAngularVelocity=(X=0,Y=0,Z=0))
   ) -> (Pose) {
    angularDampingOverride=1
    angularSpringConstant=1000
    bAngularSpring=true
    bOverrideAngularDamping=true
    bUseGravityOverride=true
    gravityScale=0.35
    simulationSpace=RootRelative
    alphaBoolBlend {blendOption=Linear}
    boundBone {boneName=dyn_strapArm_2_l}
    componentAppliedLinearAccClamp=<3000 3000>
    componentLinearAccScale=<0.5 0.5 0.25>
    componentLinearVelScale=<0.5 0.5 0.25>
    gravityOverride {z=-980}
    physicsBodyDefinitions [{
      boundBone {boneName=dyn_strapArm_2_l}
      boxExtents=<20 20 20>
      constraintSetup {angularLimitsMax=<0 15 10>, angularLimitsMin=<0 -15 -18>, angularTarget=<1 0 0>}
      localJointOffset=<2 0 0>
     }]
   }
   n47 = Anim Dynamics - Bone: dyn_strapArm_3_l!(
    ComponentPose=@n46
    Alpha=1
    SimSpaceSettings:Anim Phys Sim Space Settings Structure=(SimSpaceAngularAlpha=0,MaxAngularVelocity=10000,MaxAngularAcceleration=10000,ExternalAngularVelocity=(X=0,Y=0,Z=0))
   ) -> (Pose) {
    angularDampingOverride=1
    angularSpringConstant=1000
    bAngularSpring=true
    bOverrideAngularDamping=true
    bUseGravityOverride=true
    gravityScale=0.35
    simulationSpace=RootRelative
    alphaBoolBlend {blendOption=Linear}
    boundBone {boneName=dyn_strapArm_3_l}
    componentAppliedLinearAccClamp=<3000 3000>
    componentLinearAccScale=<0.5 0.5 0.25>
    componentLinearVelScale=<0.5 0.5 0.25>
    gravityOverride {z=-980}
    physicsBodyDefinitions [{
      boundBone {boneName=dyn_strapArm_3_l}
      boxExtents=<20 20 20>
      constraintSetup {angularLimitsMax=<0 15 25>, angularLimitsMin=<0 -15 -25>, angularTarget=<1 0 0>}
      localJointOffset=<2 0 0>
     }]
   }
   n48 = Anim Dynamics - Bone: dyn_strapArm_1_r!(
    ComponentPose=@n47
    Alpha=1
    SimSpaceSettings:Anim Phys Sim Space Settings Structure=(SimSpaceAngularAlpha=0,MaxAngularVelocity=10000,MaxAngularAcceleration=10000,ExternalAngularVelocity=(X=0,Y=0,Z=0))
   ) -> (Pose) {
    angularDampingOverride=1
    angularSpringConstant=1000
    bAngularSpring=true
    bOverrideAngularDamping=true
    bUseGravityOverride=true
    gravityScale=0.35
    simulationSpace=RootRelative
    alphaBoolBlend {blendOption=Linear}
    boundBone {boneName=dyn_strapArm_1_r}
    componentAppliedLinearAccClamp=<3000 3000>
    componentLinearAccScale=<0.5 0.5 0.25>
    componentLinearVelScale=<0.5 0.5 0.25>
    gravityOverride {z=-980}
    physicsBodyDefinitions [{
      boundBone {boneName=dyn_strapArm_1_r}
      boxExtents=<20 20 20>
      constraintSetup {angularLimitsMax=<0 15 15>, angularLimitsMin=<0 -15 -15>, angularTarget=<1 0 0>}
      localJointOffset=<2 0 0>
     }]
   }
   n49 = Anim Dynamics - Bone: dyn_strapArm_2_r!(
    ComponentPose=@n48
    Alpha=1
    SimSpaceSettings:Anim Phys Sim Space Settings Structure=(SimSpaceAngularAlpha=0,MaxAngularVelocity=10000,MaxAngularAcceleration=10000,ExternalAngularVelocity=(X=0,Y=0,Z=0))
   ) -> (Pose) {
    angularDampingOverride=1
    angularSpringConstant=1000
    bAngularSpring=true
    bOverrideAngularDamping=true
    bUseGravityOverride=true
    gravityScale=0.35
    simulationSpace=RootRelative
    alphaBoolBlend {blendOption=Linear}
    boundBone {boneName=dyn_strapArm_2_r}
    componentAppliedLinearAccClamp=<3000 3000>
    componentLinearAccScale=<0.5 0.5 0.25>
    componentLinearVelScale=<0.5 0.5 0.25>
    gravityOverride {z=-980}
    physicsBodyDefinitions [{
      boundBone {boneName=dyn_strapArm_2_r}
      boxExtents=<20 20 20>
      constraintSetup {angularLimitsMax=<0 15 25>, angularLimitsMin=<0 -15 -30>, angularTarget=<1 0 0>}
      localJointOffset=<2 0 0>
     }]
   }
   n50 = Anim Dynamics - Bone: dyn_skirt_1_f_c!(
    ComponentPose=@n49
    Alpha=1
    GravityOverride:Vector="0,0,-980"
    SimSpaceSettings:Anim Phys Sim Space Settings Structure=(SimSpaceAngularAlpha=0,MaxAngularVelocity=10000,MaxAngularAcceleration=10000,ExternalAngularVelocity=(X=0,Y=0,Z=0))
   ) -> (Pose) {
    angularDampingOverride=0.99
    angularSpringConstant=3000
    bAngularSpring=true
    bOverrideAngularDamping=true
    bUseGravityOverride=true
    gravityScale=0.35
    simulationSpace=RootRelative
    alphaBoolBlend {blendOption=Linear}
    boundBone {boneName=dyn_skirt_1_f_c}
    componentAppliedLinearAccClamp=<1000 1000 1000>
    componentLinearAccScale=<0.7 0.7 0.7>
    componentLinearVelScale=<0.7 0.7 0.7>
    gravityOverride {z=-980}
    physicsBodyDefinitions [{
      boundBone {boneName=dyn_skirt_1_f_c}
      boxExtents=<12 12 12>
      constraintSetup {angularLimitsMax=<0 0 0>, angularLimitsMin=<0 -10 0>, angularTarget=<1 0 0>}
      localJointOffset=<6 0 0>
     }]
   }
   n51 = Anim Dynamics - Bone: dyn_skirt_1_b_l!(
    ComponentPose=@n50
    Alpha=1
    GravityOverride:Vector="0,0,-980"
    SimSpaceSettings:Anim Phys Sim Space Settings Structure=(SimSpaceAngularAlpha=0,MaxAngularVelocity=10000,MaxAngularAcceleration=10000,ExternalAngularVelocity=(X=0,Y=0,Z=0))
   ) -> (Pose) {
    angularDampingOverride=0.99
    angularSpringConstant=3000
    bAngularSpring=true
    bOverrideAngularDamping=true
    bUseGravityOverride=true
    gravityScale=0.35
    simulationSpace=RootRelative
    alphaBoolBlend {blendOption=Linear}
    boundBone {boneName=dyn_skirt_1_b_l}
    componentAppliedLinearAccClamp=<1000 1000 1000>
    componentLinearAccScale=<0.7 0.7 0.7>
    componentLinearVelScale=<0.7 0.7 0.7>
    gravityOverride {z=-980}
    physicsBodyDefinitions [{
      boundBone {boneName=dyn_skirt_1_b_l}
      boxExtents=<12 12 12>
      constraintSetup {angularLimitsMax=<0 0 0>, angularLimitsMin=<0 -10 0>, angularTarget=<1 0 0>}
      localJointOffset=<6 0 0>
     }]
   }
   n52 = Anim Dynamics - Bone: dyn_skirt_1_f_l!(
    ComponentPose=@n51
    Alpha=1
    GravityOverride:Vector="0,0,-980"
    SimSpaceSettings:Anim Phys Sim Space Settings Structure=(SimSpaceAngularAlpha=0,MaxAngularVelocity=10000,MaxAngularAcceleration=10000,ExternalAngularVelocity=(X=0,Y=0,Z=0))
   ) -> (Pose) {
    angularDampingOverride=0.99
    angularSpringConstant=3000
    bAngularSpring=true
    bOverrideAngularDamping=true
    bUseGravityOverride=true
    gravityScale=0.35
    simulationSpace=RootRelative
    alphaBoolBlend {blendOption=Linear}
    boundBone {boneName=dyn_skirt_1_f_l}
    componentAppliedLinearAccClamp=<1000 1000 1000>
    componentLinearAccScale=<0.7 0.7 0.7>
    componentLinearVelScale=<0.7 0.7 0.7>
    gravityOverride {z=-980}
    physicsBodyDefinitions [{
      boundBone {boneName=dyn_skirt_1_f_l}
      boxExtents=<12 12 12>
      constraintSetup {angularLimitsMax=<0 0 0>, angularLimitsMin=<0 -10 0>, angularTarget=<1 0 0>}
      localJointOffset=<6 0 0>
     }]
   }
   n53 = Anim Dynamics - Bone: dyn_skirt_1_b_l(
    ComponentPose=@n52
    Alpha=1
    GravityOverride="0,0,-980"
    SimSpaceSettings=(SimSpaceAngularAlpha=0,MaxAngularVelocity=10000,MaxAngularAcceleration=10000,ExternalAngularVelocity=(X=0,Y=0,Z=0))
   ) {
    angularDampingOverride=0.99
    angularSpringConstant=3000
    bAngularSpring=true
    bOverrideAngularDamping=true
    bUseGravityOverride=true
    gravityScale=0.35
    simulationSpace=RootRelative
    alphaBoolBlend {blendOption=Linear}
    boundBone {boneName=dyn_skirt_1_b_l}
    componentAppliedLinearAccClamp=<1000 1000 1000>
    componentLinearAccScale=<0.7 0.7 0.7>
    componentLinearVelScale=<0.7 0.7 0.7>
    gravityOverride {z=-980}
    physicsBodyDefinitions [{
      boundBone {boneName=dyn_skirt_1_b_l}
      boxExtents=<12 12 12>
      constraintSetup {angularLimitsMax=<0 0 0>, angularLimitsMin=<0 -10 0>, angularTarget=<1 0 0>}
      localJointOffset=<6 0 0>
     }]
   }
   n54 = Anim Dynamics - Bone: dyn_skirt_1_s_l!(
    ComponentPose=@n53
    Alpha=1
    GravityOverride:Vector="0,0,-980"
    SimSpaceSettings:Anim Phys Sim Space Settings Structure=(SimSpaceAngularAlpha=0,MaxAngularVelocity=10000,MaxAngularAcceleration=10000,ExternalAngularVelocity=(X=0,Y=0,Z=0))
   ) -> (Pose) {
    angularDampingOverride=0.99
    angularSpringConstant=3000
    bAngularSpring=true
    bOverrideAngularDamping=true
    bUseGravityOverride=true
    gravityScale=0.35
    simulationSpace=RootRelative
    alphaBoolBlend {blendOption=Linear}
    boundBone {boneName=dyn_skirt_1_s_l}
    componentAppliedLinearAccClamp=<1000 1000 1000>
    componentLinearAccScale=<0.7 0.7 0.7>
    componentLinearVelScale=<0.7 0.7 0.7>
    gravityOverride {z=-980}
    physicsBodyDefinitions [{
      boundBone {boneName=dyn_skirt_1_s_l}
      boxExtents=<12 12 12>
      constraintSetup {angularLimitsMax=<0 0 0>, angularLimitsMin=<0 -10 0>, angularTarget=<1 0 0>}
      localJointOffset=<6 0 0>
     }]
   }
   n55 = Anim Dynamics - Bone: dyn_skirt_1_f_r!(
    ComponentPose=@n54
    Alpha=1
    GravityOverride:Vector="0,0,-980"
    SimSpaceSettings:Anim Phys Sim Space Settings Structure=(SimSpaceAngularAlpha=0,MaxAngularVelocity=10000,MaxAngularAcceleration=10000,ExternalAngularVelocity=(X=0,Y=0,Z=0))
   ) -> (Pose) {
    angularDampingOverride=0.99
    angularSpringConstant=3000
    bAngularSpring=true
    bOverrideAngularDamping=true
    bUseGravityOverride=true
    gravityScale=0.35
    simulationSpace=RootRelative
    alphaBoolBlend {blendOption=Linear}
    boundBone {boneName=dyn_skirt_1_f_r}
    componentAppliedLinearAccClamp=<1000 1000 1000>
    componentLinearAccScale=<0.7 0.7 0.7>
    componentLinearVelScale=<0.7 0.7 0.7>
    gravityOverride {z=-980}
    physicsBodyDefinitions [{
      boundBone {boneName=dyn_skirt_1_f_r}
      boxExtents=<12 12 12>
      constraintSetup {angularLimitsMax=<0 0 0>, angularLimitsMin=<0 -10 0>, angularTarget=<1 0 0>}
      localJointOffset=<6 0 0>
     }]
   }
   n56 = Anim Dynamics - Bone: dyn_skirt_1_b_r!(
    ComponentPose=@n55
    Alpha=1
    GravityOverride:Vector="0,0,-980"
    SimSpaceSettings:Anim Phys Sim Space Settings Structure=(SimSpaceAngularAlpha=0,MaxAngularVelocity=10000,MaxAngularAcceleration=10000,ExternalAngularVelocity=(X=0,Y=0,Z=0))
   ) -> (Pose) {
    angularDampingOverride=0.99
    angularSpringConstant=3000
    bAngularSpring=true
    bOverrideAngularDamping=true
    bUseGravityOverride=true
    gravityScale=0.35
    simulationSpace=RootRelative
    alphaBoolBlend {blendOption=Linear}
    boundBone {boneName=dyn_skirt_1_b_r}
    componentAppliedLinearAccClamp=<1000 1000 1000>
    componentLinearAccScale=<0.7 0.7 0.7>
    componentLinearVelScale=<0.7 0.7 0.7>
    gravityOverride {z=-980}
    physicsBodyDefinitions [{
      boundBone {boneName=dyn_skirt_1_b_r}
      boxExtents=<12 12 12>
      constraintSetup {angularLimitsMax=<0 0 0>, angularLimitsMin=<0 -10 0>, angularTarget=<1 0 0>}
      localJointOffset=<6 0 0>
     }]
   }
   n57 = Anim Dynamics - Bone: dyn_skirt_1_s_r!(
    ComponentPose=@n56
    Alpha=1
    GravityOverride:Vector="0,0,-980"
    SimSpaceSettings:Anim Phys Sim Space Settings Structure=(SimSpaceAngularAlpha=0,MaxAngularVelocity=10000,MaxAngularAcceleration=10000,ExternalAngularVelocity=(X=0,Y=0,Z=0))
   ) -> (Pose) {
    angularDampingOverride=0.99
    angularSpringConstant=3000
    bAngularSpring=true
    bOverrideAngularDamping=true
    bUseGravityOverride=true
    gravityScale=0.35
    simulationSpace=RootRelative
    alphaBoolBlend {blendOption=Linear}
    boundBone {boneName=dyn_skirt_1_s_r}
    componentAppliedLinearAccClamp=<1000 1000 1000>
    componentLinearAccScale=<0.7 0.7 0.7>
    componentLinearVelScale=<0.7 0.7 0.7>
    gravityOverride {z=-980}
    physicsBodyDefinitions [{
      boundBone {boneName=dyn_skirt_1_s_r}
      boxExtents=<12 12 12>
      constraintSetup {angularLimitsMax=<0 0 0>, angularLimitsMin=<0 -10 0>, angularTarget=<1 0 0>}
      localJointOffset=<6 0 0>
     }]
   }
   n58 = Anim Dynamics - Bone: dyn_skirt_2_f_c!(
    ComponentPose=@n57
    Alpha=1
    GravityOverride:Vector="0,0,-980"
    SimSpaceSettings:Anim Phys Sim Space Settings Structure=(SimSpaceAngularAlpha=0,MaxAngularVelocity=10000,MaxAngularAcceleration=10000,ExternalAngularVelocity=(X=0,Y=0,Z=0))
   ) -> (Pose) {
    angularDampingOverride=0.99
    angularSpringConstant=3000
    bAngularSpring=true
    bOverrideAngularDamping=true
    bUseGravityOverride=true
    gravityScale=0.35
    simulationSpace=RootRelative
    alphaBoolBlend {blendOption=Linear}
    boundBone {boneName=dyn_skirt_2_f_c}
    componentAppliedLinearAccClamp=<1000 1000 1000>
    componentLinearAccScale=<0.7 0.7 0.7>
    componentLinearVelScale=<0.7 0.7 0.7>
    gravityOverride {z=-980}
    physicsBodyDefinitions [{
      boundBone {boneName=dyn_skirt_2_f_c}
      boxExtents=<12 12 12>
      constraintSetup {angularLimitsMax=<0 0 0>, angularLimitsMin=<0 -15 0>, angularTarget=<1 0 0>}
      localJointOffset=<9 0 0>
     }]
   }
   n59 = Anim Dynamics - Bone: dyn_skirt_2_b_l!(
    ComponentPose=@n58
    Alpha=1
    GravityOverride:Vector="0,0,-980"
    SimSpaceSettings:Anim Phys Sim Space Settings Structure=(SimSpaceAngularAlpha=0,MaxAngularVelocity=10000,MaxAngularAcceleration=10000,ExternalAngularVelocity=(X=0,Y=0,Z=0))
   ) -> (Pose) {
    angularDampingOverride=0.99
    angularSpringConstant=3000
    bAngularSpring=true
    bOverrideAngularDamping=true
    bUseGravityOverride=true
    gravityScale=0.35
    simulationSpace=RootRelative
    alphaBoolBlend {blendOption=Linear}
    boundBone {boneName=dyn_skirt_2_b_l}
    componentAppliedLinearAccClamp=<1000 1000 1000>
    componentLinearAccScale=<0.7 0.7 0.7>
    componentLinearVelScale=<0.7 0.7 0.7>
    gravityOverride {z=-980}
    physicsBodyDefinitions [{
      boundBone {boneName=dyn_skirt_2_b_l}
      boxExtents=<12 12 12>
      constraintSetup {angularLimitsMax=<0 0 0>, angularLimitsMin=<0 -15 0>, angularTarget=<1 0 0>}
      localJointOffset=<9 0 0>
     }]
   }
   n60 = Anim Dynamics - Bone: dyn_skirt_2_f_l!(
    ComponentPose=@n59
    Alpha=1
    GravityOverride:Vector="0,0,-980"
    SimSpaceSettings:Anim Phys Sim Space Settings Structure=(SimSpaceAngularAlpha=0,MaxAngularVelocity=10000,MaxAngularAcceleration=10000,ExternalAngularVelocity=(X=0,Y=0,Z=0))
   ) -> (Pose) {
    angularDampingOverride=0.99
    angularSpringConstant=3000
    bAngularSpring=true
    bOverrideAngularDamping=true
    bUseGravityOverride=true
    gravityScale=0.35
    simulationSpace=RootRelative
    alphaBoolBlend {blendOption=Linear}
    boundBone {boneName=dyn_skirt_2_f_l}
    componentAppliedLinearAccClamp=<1000 1000 1000>
    componentLinearAccScale=<0.7 0.7 0.7>
    componentLinearVelScale=<0.7 0.7 0.7>
    gravityOverride {z=-980}
    physicsBodyDefinitions [{
      boundBone {boneName=dyn_skirt_2_f_l}
      boxExtents=<12 12 12>
      constraintSetup {angularLimitsMax=<0 0 0>, angularLimitsMin=<0 -15 0>, angularTarget=<1 0 0>}
      localJointOffset=<9 0 0>
     }]
   }
   n61 = Anim Dynamics - Bone: dyn_skirt_2_b_l(
    ComponentPose=@n60
    Alpha=1
    GravityOverride="0,0,-980"
    SimSpaceSettings=(SimSpaceAngularAlpha=0,MaxAngularVelocity=10000,MaxAngularAcceleration=10000,ExternalAngularVelocity=(X=0,Y=0,Z=0))
   ) {
    angularDampingOverride=0.99
    angularSpringConstant=3000
    bAngularSpring=true
    bOverrideAngularDamping=true
    bUseGravityOverride=true
    gravityScale=0.35
    simulationSpace=RootRelative
    alphaBoolBlend {blendOption=Linear}
    boundBone {boneName=dyn_skirt_2_b_l}
    componentAppliedLinearAccClamp=<1000 1000 1000>
    componentLinearAccScale=<0.7 0.7 0.7>
    componentLinearVelScale=<0.7 0.7 0.7>
    gravityOverride {z=-980}
    physicsBodyDefinitions [{
      boundBone {boneName=dyn_skirt_2_b_l}
      boxExtents=<12 12 12>
      constraintSetup {angularLimitsMax=<0 0 0>, angularLimitsMin=<0 -15 0>, angularTarget=<1 0 0>}
      localJointOffset=<9 0 0>
     }]
   }
   n62 = Anim Dynamics - Bone: dyn_skirt_2_s_l!(
    ComponentPose=@n61
    Alpha=1
    GravityOverride:Vector="0,0,-980"
    SimSpaceSettings:Anim Phys Sim Space Settings Structure=(SimSpaceAngularAlpha=0,MaxAngularVelocity=10000,MaxAngularAcceleration=10000,ExternalAngularVelocity=(X=0,Y=0,Z=0))
   ) -> (Pose) {
    angularDampingOverride=0.99
    angularSpringConstant=3000
    bAngularSpring=true
    bOverrideAngularDamping=true
    bUseGravityOverride=true
    gravityScale=0.35
    simulationSpace=RootRelative
    alphaBoolBlend {blendOption=Linear}
    boundBone {boneName=dyn_skirt_2_s_l}
    componentAppliedLinearAccClamp=<1000 1000 1000>
    componentLinearAccScale=<0.7 0.7 0.7>
    componentLinearVelScale=<0.7 0.7 0.7>
    gravityOverride {z=-980}
    physicsBodyDefinitions [{
      boundBone {boneName=dyn_skirt_2_s_l}
      boxExtents=<12 12 12>
      constraintSetup {angularLimitsMax=<0 0 0>, angularLimitsMin=<0 -15 0>, angularTarget=<1 0 0>}
      localJointOffset=<9 0 0>
     }]
   }
   n63 = Anim Dynamics - Bone: dyn_skirt_2_f_r!(
    ComponentPose=@n62
    Alpha=1
    GravityOverride:Vector="0,0,-980"
    SimSpaceSettings:Anim Phys Sim Space Settings Structure=(SimSpaceAngularAlpha=0,MaxAngularVelocity=10000,MaxAngularAcceleration=10000,ExternalAngularVelocity=(X=0,Y=0,Z=0))
   ) -> (Pose) {
    angularDampingOverride=0.99
    angularSpringConstant=3000
    bAngularSpring=true
    bOverrideAngularDamping=true
    bUseGravityOverride=true
    gravityScale=0.35
    simulationSpace=RootRelative
    alphaBoolBlend {blendOption=Linear}
    boundBone {boneName=dyn_skirt_2_f_r}
    componentAppliedLinearAccClamp=<1000 1000 1000>
    componentLinearAccScale=<0.7 0.7 0.7>
    componentLinearVelScale=<0.7 0.7 0.7>
    gravityOverride {z=-980}
    physicsBodyDefinitions [{
      boundBone {boneName=dyn_skirt_2_f_r}
      boxExtents=<12 12 12>
      constraintSetup {angularLimitsMax=<0 0 0>, angularLimitsMin=<0 -15 0>, angularTarget=<1 0 0>}
      localJointOffset=<9 0 0>
     }]
   }
   n64 = Anim Dynamics - Bone: dyn_skirt_2_b_r!(
    ComponentPose=@n63
    Alpha=1
    GravityOverride:Vector="0,0,-980"
    SimSpaceSettings:Anim Phys Sim Space Settings Structure=(SimSpaceAngularAlpha=0,MaxAngularVelocity=10000,MaxAngularAcceleration=10000,ExternalAngularVelocity=(X=0,Y=0,Z=0))
   ) -> (Pose) {
    angularDampingOverride=0.99
    angularSpringConstant=3000
    bAngularSpring=true
    bOverrideAngularDamping=true
    bUseGravityOverride=true
    gravityScale=0.35
    simulationSpace=RootRelative
    alphaBoolBlend {blendOption=Linear}
    boundBone {boneName=dyn_skirt_2_b_r}
    componentAppliedLinearAccClamp=<1000 1000 1000>
    componentLinearAccScale=<0.7 0.7 0.7>
    componentLinearVelScale=<0.7 0.7 0.7>
    gravityOverride {z=-980}
    physicsBodyDefinitions [{
      boundBone {boneName=dyn_skirt_2_b_r}
      boxExtents=<12 12 12>
      constraintSetup {angularLimitsMax=<0 0 0>, angularLimitsMin=<0 -15 0>, angularTarget=<1 0 0>}
      localJointOffset=<9 0 0>
     }]
   }
   n65 = Anim Dynamics - Bone: dyn_skirt_2_s_r!(
    ComponentPose=@n64
    Alpha=1
    GravityOverride:Vector="0,0,-980"
    SimSpaceSettings:Anim Phys Sim Space Settings Structure=(SimSpaceAngularAlpha=0,MaxAngularVelocity=10000,MaxAngularAcceleration=10000,ExternalAngularVelocity=(X=0,Y=0,Z=0))
   ) -> (Pose) {
    angularDampingOverride=0.99
    angularSpringConstant=3000
    bAngularSpring=true
    bOverrideAngularDamping=true
    bUseGravityOverride=true
    gravityScale=0.35
    simulationSpace=RootRelative
    alphaBoolBlend {blendOption=Linear}
    boundBone {boneName=dyn_skirt_2_s_r}
    componentAppliedLinearAccClamp=<1000 1000 1000>
    componentLinearAccScale=<0.7 0.7 0.7>
    componentLinearVelScale=<0.7 0.7 0.7>
    gravityOverride {z=-980}
    physicsBodyDefinitions [{
      boundBone {boneName=dyn_skirt_2_s_r}
      boxExtents=<12 12 12>
      constraintSetup {angularLimitsMax=<0 0 0>, angularLimitsMin=<0 -15 0>, angularTarget=<1 0 0>}
      localJointOffset=<9 0 0>
     }]
   }
   Save cached pose 'animDynamicsCP'(Pose=Component To Local(ComponentPose=@n65)) {cacheName=animDynamicsCP}
  ]}
 EventGraph {nodes [
   Event BlueprintInitializeAnimation()
   n1 = TryGetPawnOwner()
   > Is Valid(InputObject=@n1)
   > n2 = Cast To FortPlayerPawnAthena(Object=@n1)
   > Set PlayerPawn(PlayerPawn:Fort Player Pawn Athena Object Reference=@n2) -> (Output_Get:Fort Player Pawn Athena Object Reference)
   n3 = Event OnCharacterCustomizationLoaded()
   > Set HeadMesh(HeadMesh:Skeletal Mesh Component Budgeted Object Reference=Get Array Item with EFort Custom Part Type(Array=@n3.CustomMeshParts, PartType=Head)) -> (Output_Get:Skeletal Mesh Component Budgeted Object Reference)
   > Set BodyMesh(BodyMesh=GetPartSkeletalMeshComponent())
   n4 = Get BodyMesh() -> (BodyMesh:Skeletal Mesh Component Object Reference)
   > n5 = Get Dynamic Material Instance(SkeletalMeshComponent=@n4)
   > Set Mat Instance Dynamic(Mat Instance Dynamic=@n5)
   > n6 = GetOrCreateDynamicOverlayMaterial(MeshComponent=@n4, MaterialName=None)
   > Set Overlay Mat Instance Dynamic(Overlay Mat Instance Dynamic=@n6)
   > n7 = Branch(Condition=Get bIsInFrontEnd())
   > Set SimSpaceSettings output(SimSpaceSettings output:Anim Phys Sim Space Settings Structure=Get SimSpaceSettings frontEnd() -> (SimSpaceSettings frontEnd:Anim Phys Sim Space Settings Structure)) -> (Output_Get:Anim Phys Sim Space Settings Structure)
   > Set SimSpaceSettings skirt output(SimSpaceSettings skirt output:Anim Phys Sim Space Settings Structure=Get SimSpaceSettings skirt frontEnd() -> (SimSpaceSettings skirt frontEnd:Anim Phys Sim Space Settings Structure)) -> (Output_Get:Anim Phys Sim Space Settings Structure)
   Set SimSpaceSettings output(
    execute=@n7.else
    SimSpaceSettings output=Get SimSpaceSettings default() -> (SimSpaceSettings default:Anim Phys Sim Space Settings Structure)
   )
   > Set SimSpaceSettings skirt output(SimSpaceSettings skirt output=Get SimSpaceSettings default())
   Event OnParachuteClosed()
   > Set isUsingHangGlidingInParachute(isUsingHangGlidingInParachute:bool=false) -> (Output_Get:bool)
   Event OnParachuteOpened()
   > Switch on EGliderType(Selection=Get PlayerGliderType() -> (PlayerGliderType=HangGlider))
   > Set isUsingHangGlidingInParachute(isUsingHangGlidingInParachute=true)
   n8 = Event BlueprintUpdateAnimation()
   > n9 = Sequence!(execute) -> (then_0, then_1, then_2:Exec, then_3:Exec)
   n10 = "Gravity Override State Function (Parameter Asset) (Nativized)"(
    execute=@n9.then_1
    CharacterPartAnimInstance=Self-Reference()
    GameTimeInSeconds=GetGameTimeInSeconds()
    Parameters=&/F_MED_BlueMonday_Parameters
   )
   > Set GravityOverride(GravityOverride=@n10)
   n11 = Get bIsTacticalSprinting()
   n12 = Get bIsJumping()
   n13 = Get bIsFalling()
   n14 = Branch(
    execute=@n9.then_2
    Condition=OR Boolean!(A=AND Boolean(A=@n12, B=@n11), B=AND Boolean(A=@n11, B=@n13), C:bool=Get bIsTacticalSprinting(), D:bool=Get bIsKneeSliding()) -> (ReturnValue)
   )
   > Set alphaSkirtRigidBody float(alphaSkirtRigidBody float:double=0.5) -> (Output_Get:double)
   n15 = Branch(execute=@n14.else, Condition=Get bIsBaseballSliding())
   > Set alphaSkirtRigidBody float(alphaSkirtRigidBody float=0)
   n16 = Branch(execute=@n15.else, Condition=OR Boolean!(A=Get bIsJumping(), B=Get bIsFalling()) -> (ReturnValue))
   > Set alphaSkirtRigidBody float(alphaSkirtRigidBody float=0.6)
   Set alphaSkirtRigidBody float(execute=@n16.else, alphaSkirtRigidBody float=1)
   n17 = Sequence!(execute=@n9.then_3) -> (then_0, then_1, then_2)
   n18 = BreakTransform(InTransform=GetSocketTransform(self=GetPartSkeletalMeshComponent(), InSocketName=dyn_skirt_1_b_r, TransformSpace=RTS_ParentBoneSpace))
   > Set dyn_skirt_1_b_r_delta(dyn_skirt_1_b_r_delta:double=Remap Float(Input=BreakRotator(InRot=@n18.Rotation).Pitch, OldRange=(X=0,Y=10), NewRange=(X=0,Y=1), Clamp=true) -> (Output=~)) -> (Output_Get:double)
   Sequence!(execute=@n17.then_2) -> (then_0, then_1)
   > Is Valid(InputObject=Get PlayerPawn() -> (PlayerPawn:Fort Player Pawn Athena Object Reference))
   > n19 = Branch(Condition=Get bIsInFrontEnd())
   Set WPO Globals Function(
    execute=@n19.else
    Material=Get Mat Instance Dynamic()
    Overlay Material=Get Overlay Mat Instance Dynamic()
    Skeletal Mesh=Get BodyMesh()
    "Local Noise Transform  Location"=Get Local Noise Transform Location() -> (Local Noise Transform Location:Vector)
    Local Noise Offset=Get Local Noise Offset()
   )
   > Set WPO Regional Function!(
    execute
    Material:Material Instance Dynamic Object Reference=Get Mat Instance Dynamic()
    Overlay Material:Material Instance Dynamic Object Reference=Get Overlay Mat Instance Dynamic()
    Skeletal Mesh:Skeletal Mesh Component Object Reference=Get BodyMesh()
    Velocity Transform:Name=foot_l
    Velocity Interp Speed:double=Get Velocity Interp Speed()
    Collision Transform:Name=calf_l
    Collision Transform Axis:EAxis Enum
    Invert Collision Transform Axis:bool=true
    Delta Time X:double=@n8.DeltaTimeX
    Previous Velocity Transform Location:"Vector (by ref)"=Get R1 Velocity Transform Location()
    Previous Velocity Transform Velocity:"Vector (by ref)"=Get R1 Velocity Transform Velocity()
    Region:int=1
   ) -> (then)
   > Set WPO Regional Function(
    Material=Get Mat Instance Dynamic()
    Overlay Material=Get Overlay Mat Instance Dynamic()
    Skeletal Mesh=Get BodyMesh()
    Velocity Transform=foot_r
    Velocity Interp Speed=Get Velocity Interp Speed()
    Collision Transform=calf_r
    Invert Collision Transform Axis=true
    Delta Time X=@n8.DeltaTimeX
    Previous Velocity Transform Location=Get R2 Velocity Transform Location()
    Previous Velocity Transform Velocity=Get R2 Velocity Transform Velocity()
    Region=2
   )
   > n20 = Branch(Condition=OR Boolean(A=Get bIsSkydiveFloating(), B=Get bIsParachuteOpened()))
   > Set WPO Pho Wind Intensity(WPO Pho Wind Intensity=0.8)
   > n21 = Set WPO Pho Turbulence(WPO Pho Turbulence:double=3.5) -> (Output_Get:double)
   SetScalarParameterValue(
    execute=[@n23.then, @n21.then]
    self=[Get Overlay Mat Instance Dynamic(), Get Mat Instance Dynamic()]
    ParameterName=ClothSim_WindStrength
    Value=Get WPO Pho Wind Intensity()
   )
   n22 = Get WPO Pho Turbulence() -> (WPO Pho Turbulence:double)
   > SetScalarParameterValue(self=[Get Overlay Mat Instance Dynamic(), Get Mat Instance Dynamic()], ParameterName=ClothSim_R1_TurbulenceStrength, Value=@n22)
   > SetScalarParameterValue(self=[Get Mat Instance Dynamic(), Get Overlay Mat Instance Dynamic()], ParameterName=ClothSim_R2_TurbulenceStrength, Value=@n22)
   Set WPO Pho Wind Intensity(execute=@n20.else, WPO Pho Wind Intensity=0)
   > n23 = Set WPO Pho Turbulence(WPO Pho Turbulence=0)
   Set r_thighLocal rz_0_ n100(execute=@n17.then_1, r_thighLocal rz_0_ n100:double=GetCurveValue(CurveName=r_thighLocal rz_0_ n100)) -> (Output_Get:double)
   n24 = Get HeadMesh() -> (HeadMesh:Skeletal Mesh Component Budgeted Object Reference)
   Is Valid(exec=@n9.then_0, InputObject=@n24)
   n25 = GetSocketBoneName!(self:Skinned Mesh Component Object Reference=@n24, InSocketName:Name=FACIAL_C_Jaw) -> (ReturnValue:Name=None)
   > n26 = Set isUsingFacialBones(isUsingFacialBones:bool="Not Equal Exactly (String)"(A="To String (Name)"(InName=@n25), B=None)) -> (Output_Get:bool)
   > Branch(Condition=@n26)
   n27 = Get HeadMesh()
   n28 = BreakTransform(InTransform=GetSocketTransform(self=@n27, InSocketName=FACIAL_C_UnderChin, TransformSpace=RTS_ParentBoneSpace))
   > Set FACIAL_C_UnderChin_Translation(FACIAL_C_UnderChin_Translation:Vector=@n28) -> (Output_Get:Vector)
   > Set FACIAL_C_UnderChin_Rotation(FACIAL_C_UnderChin_Rotation:Rotator=@n28.Rotation) -> (Output_Get:Rotator)
   n29 = BreakTransform(InTransform=GetSocketTransform(self=@n27, InSocketName=FACIAL_L_UnderChin, TransformSpace=RTS_ParentBoneSpace))
   > Set FACIAL_L_UnderChin_Translation(FACIAL_L_UnderChin_Translation:Vector=@n29) -> (Output_Get:Vector)
   > Set FACIAL_L_UnderChin_Rotation(FACIAL_L_UnderChin_Rotation:Rotator=@n29.Rotation) -> (Output_Get:Rotator)
   n30 = BreakTransform(InTransform=GetSocketTransform(self=@n27, InSocketName=FACIAL_R_UnderChin, TransformSpace=RTS_ParentBoneSpace))
   > Set FACIAL_R_UnderChin_Translation(FACIAL_R_UnderChin_Translation:Vector=@n30) -> (Output_Get:Vector)
   > Set FACIAL_R_UnderChin_Rotation(FACIAL_R_UnderChin_Rotation:Rotator=@n30.Rotation) -> (Output_Get:Rotator)
   n31 = BreakTransform(InTransform=GetSocketTransform(self=@n27, InSocketName=FACIAL_C_Jaw, TransformSpace=RTS_ParentBoneSpace))
   > Set FACIAL_C_Jaw_Tranform(FACIAL_C_Jaw_Tranform:Vector=@n31) -> (Output_Get:Vector)
   > Set FACIAL_C_Jaw_Rotation(FACIAL_C_Jaw_Rotation:Rotator=@n31.Rotation) -> (Output_Get:Rotator)
   n32 = BreakTransform(InTransform=GetSocketTransform(self=@n27, InSocketName=FACIAL_C_Jawline, TransformSpace=RTS_ParentBoneSpace))
   > Set FACIAL_C_Jawline_Translation(FACIAL_C_Jawline_Translation:Vector=@n32) -> (Output_Get:Vector)
   > Set FACIAL_C_Jawline_Rotation(FACIAL_C_Jawline_Rotation:Rotator=@n32.Rotation) -> (Output_Get:Rotator)
   n33 = BreakTransform(InTransform=GetSocketTransform(self=@n27, InSocketName=FACIAL_L_Jawline, TransformSpace=RTS_ParentBoneSpace))
   > Set FACIAL_L_Jawline_Translation(FACIAL_L_Jawline_Translation:Vector=@n33) -> (Output_Get:Vector)
   > Set FACIAL_L_Jawline_Rotation(FACIAL_L_Jawline_Rotation:Rotator=@n33.Rotation) -> (Output_Get:Rotator)
   n34 = BreakTransform(InTransform=GetSocketTransform(self=@n27, InSocketName=FACIAL_R_Jawline, TransformSpace=RTS_ParentBoneSpace))
   > Set FACIAL_R_Jawline_Translation(FACIAL_R_Jawline_Translation:Vector=@n34) -> (Output_Get:Vector)
   > Set FACIAL_R_Jawline_Rotation(FACIAL_R_Jawline_Rotation:Rotator=@n34.Rotation) -> (Output_Get:Rotator)
   n35 = BreakTransform(InTransform=GetSocketTransform(self=@n27, InSocketName=FACIAL_L_Ear, TransformSpace=RTS_ParentBoneSpace))
   > Set FACIAL_L_Ear_Translation(FACIAL_L_Ear_Translation:Vector=@n35) -> (Output_Get:Vector)
   > Set FACIAL_L_Ear_Rotation(FACIAL_L_Ear_Rotation:Rotator=@n35.Rotation) -> (Output_Get:Rotator)
   n36 = BreakTransform(InTransform=GetSocketTransform(self=@n27, InSocketName=FACIAL_R_Ear, TransformSpace=RTS_ParentBoneSpace))
   > Set FACIAL_R_Ear_Translation(FACIAL_R_Ear_Translation:Vector=@n36) -> (Output_Get:Vector)
   > Set FACIAL_R_Ear_Rotation(FACIAL_R_Ear_Rotation:Rotator=@n36.Rotation) -> (Output_Get:Rotator)
  ]}
]
variables [
 alphaSkirtRigidBody float {type=number}
 BodyMesh {type=SkeletalMeshComponent}
 dyn_skirt_1_b_r_delta {type=number}
 FACIAL_C_Jaw_Rotation {type=Rotator}
 FACIAL_C_Jaw_Tranform {type=Vector}
 FACIAL_C_Jawline_Rotation {type=Rotator}
 FACIAL_C_Jawline_Translation {type=Vector}
 FACIAL_C_UnderChin_Rotation {type=Rotator}
 FACIAL_C_UnderChin_Translation {type=Vector}
 FACIAL_L_Ear_Rotation {type=Rotator}
 FACIAL_L_Ear_Translation {type=Vector}
 FACIAL_L_Jawline_Rotation {type=Rotator}
 FACIAL_L_Jawline_Translation {type=Vector}
 FACIAL_L_UnderChin_Rotation {type=Rotator}
 FACIAL_L_UnderChin_Translation {type=Vector}
 FACIAL_R_Ear_Rotation {type=Rotator}
 FACIAL_R_Ear_Translation {type=Vector}
 FACIAL_R_Jawline_Rotation {type=Rotator}
 FACIAL_R_Jawline_Translation {type=Vector}
 FACIAL_R_UnderChin_Rotation {type=Rotator}
 FACIAL_R_UnderChin_Translation {type=Vector}
 GravityOverride {type=Vector}
 HeadMesh {type=SkeletalMeshComponentBudgeted}
 isUsingFacialBones {type=boolean}
 isUsingHangGlidingInParachute {type=boolean}
 Local Noise Offset {type=Vector}
 Local Noise Transform Location {type=Vector}
 Mat Instance Dynamic {type=MaterialInstanceDynamic}
 Overlay Mat Instance Dynamic {type=MaterialInstanceDynamic}
 PlayerPawn {type=FortPlayerPawnAthena}
 R1 Velocity Transform Location {type=Vector}
 R1 Velocity Transform Velocity {type=Vector}
 R2 Velocity Transform Location {type=Vector}
 R2 Velocity Transform Velocity {type=Vector}
 r_thighLocal rz_0_ n100 {type=number}
 SimSpaceSettings default {
  type=AnimPhysSimSpaceSettings
  default {maxAngularAcceleration=10000, maxAngularVelocity=10000, simSpaceAngularAlpha=0, externalAngularVelocity=<0 0 0>}
 }
 SimSpaceSettings frontEnd {
  type=AnimPhysSimSpaceSettings
  default {maxAngularAcceleration=30, maxAngularVelocity=6, simSpaceAngularAlpha=0.6, externalAngularVelocity=<0 0 0>}
 }
 SimSpaceSettings output {
  type=AnimPhysSimSpaceSettings
  default {maxAngularAcceleration=10000, maxAngularVelocity=10000, simSpaceAngularAlpha=0, externalAngularVelocity=<0 0 0>}
 }
 SimSpaceSettings skirt frontEnd {
  type=AnimPhysSimSpaceSettings
  default {maxAngularAcceleration=30, maxAngularVelocity=6, simSpaceAngularAlpha=0.9, externalAngularVelocity=<0 0 0>}
 }
 SimSpaceSettings skirt output {
  type=AnimPhysSimSpaceSettings
  default {maxAngularAcceleration=10000, maxAngularVelocity=10000, simSpaceAngularAlpha=0, externalAngularVelocity=<0 0 0>}
 }
 Velocity Interp Speed {type=number}
 WPO Pho Turbulence {type=number}
 WPO Pho Wind Intensity {type=number}
]
