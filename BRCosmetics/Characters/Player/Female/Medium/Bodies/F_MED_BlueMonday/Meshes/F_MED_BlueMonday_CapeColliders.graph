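# Physics asset of cape colliders for the female medium BlueMonday body cosmetic, giving capsule shapes for cloth collision.
# Capsules cover pelvis, spine, clavicles, feet and head, with extra crosswise capsules on the middle spine; arms, legs and neck have shapeless bodies.
# Back and side skirt bones carry capsules in two-link chains from the pelvis.
# All constraints leave swing and twist free, linking bodies along the skeleton hierarchy.
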PhysicsAsset /BRCosmetics/Characters/Player/Female/Medium/Bodies/F_MED_BlueMonday/Meshes/F_MED_BlueMonday_CapeColliders
bodies {
 calf_l []
 calf_r []
 clavicle_l [{shapeType=Capsule, length=6.5623, radius=7.5194, center=<3.9566 -1.0862 -2.1127>, rotation=<rot -77.5566 177.3915 7.7314>}]
 clavicle_r [{shapeType=Capsule, length=6.5623, radius=7.5194, center=<-3.9566 1.0862 2.1127>, rotation=<rot -77.5566 177.3915 -172.2686>}]
 dyn_skirt_1_b_c [{
   shapeType=Capsule
   length=2.0127
   radius=5.0985
   shapeName=dyn_skirt_1_b_c_capsule
   center=<4.6507 0.0801 -0.7057>
   rotation=<rot 0 -177.9334 -89.5018>
  }]
 dyn_skirt_1_b_l [{
   shapeType=Capsule
   length=3.1457
   radius=5.2209
   shapeName=dyn_skirt_1_b_l_capsule
   center=<4.5738 0.4202 -1.5519>
   rotation=<rot 0 172.4511 -87.4618>
  }]
 dyn_skirt_1_b_r [{
   shapeType=Capsule
   length=2.7297
   radius=5.1776
   shapeName=dyn_skirt_1_b_r_capsule
   center=<4.3751 -0.2178 -1.4851>
   rotation=<rot 0 3.8544 -85.5002>
  }]
 dyn_skirt_1_s_l [{
   shapeType=Capsule
   length=2.3534
   radius=5.353
   shapeName=dyn_skirt_1_s_l_capsule
   center=<4.882 0.0627 -2.0109>
   rotation=<rot 0 -3.4423 -88.1661>
  }]
 dyn_skirt_1_s_r [{
   shapeType=Capsule
   length=3.0031
   radius=5.4815
   shapeName=dyn_skirt_1_s_r_capsule
   center=<4.3391 -0.4633 -2.0841>
   rotation=<rot 0 -171.8428 -86.8773>
  }]
 dyn_skirt_2_b_c [{
   shapeType=Capsule
   length=3.1251
   radius=5.9974
   shapeName=dyn_skirt_2_b_c_capsule
   center=<5.0119 0.3531 -0.6568>
   rotation=<rot 0 177.0722 -89.1538>
  }]
 dyn_skirt_2_b_l [{
   shapeType=Capsule
   length=4.053
   radius=5.7782
   shapeName=dyn_skirt_2_b_l_capsule
   center=<5.1572 0.478 -1.5567>
   rotation=<rot 0 174.7211 -83.7067>
  }]
 dyn_skirt_2_b_r [{
   shapeType=Capsule
   length=4.4618
   radius=5.7441
   shapeName=dyn_skirt_2_b_r_capsule
   center=<4.9566 0.48 -1.6047>
   rotation=<rot 0 2.1019 -81.8006>
  }]
 dyn_skirt_2_s_l [{
   shapeType=Capsule
   length=3.1445
   radius=5.2747
   shapeName=dyn_skirt_2_s_l_capsule
   center=<4.8698 0.2174 -1.8422>
   rotation=<rot 0 177.4758 -89.2974>
  }]
 dyn_skirt_2_s_r [{
   shapeType=Capsule
   length=3.7413
   radius=5.5321
   shapeName=dyn_skirt_2_s_r_capsule
   center=<4.7941 -0.2408 -1.8981>
   rotation=<rot 0 -174.6447 -89.5055>
  }]
 foot_l [{shapeType=Capsule, length=6.8347, radius=12.2763, center=<-4.1811 6.0027 -0.338>, rotation=<rot 0.0388 1.4701 -89.3982>}]
 foot_r [{shapeType=Capsule, length=6.8347, radius=12.2763, center=<4.1811 -6.0027 0.338>, rotation=<rot 0.0388 1.4701 90.6018>}]
 head [{shapeType=Capsule, length=0, radius=18, center=<9.0697 0.0746 0>, rotation=<rot 0 -90 0>}]
 lowerarm_l []
 lowerarm_r []
 neck_02 []
 pelvis [{shapeType=Capsule, length=9.6042, radius=11.1441, center=<-2.4471 -2.5289 -0.1981>}]
 spine_01 [{shapeType=Capsule, length=4.3888, radius=14.3638, center=<-8.9021 -0.8973 0>}]
 spine_02 [
  {shapeType=Capsule, length=6.4519, radius=9.7317, center=<-2.6885 -1.2641 0>}
  {shapeType=Capsule, length=20.6875, radius=9.3125, center=<-1.8317 1.9362 0>, rotation=<rot 90 0 0>}
 ]
 spine_03 [
  {shapeType=Capsule, length=6.1863, radius=8.4854, center=<0.0574 0.4416 0>}
  {shapeType=Capsule, length=20.7188, radius=9.2812, center=<-1.2163 0.549 0>, rotation=<rot 90 180 -175.0732>}
 ]
 spine_04 [
  {shapeType=Capsule, length=9.0076, radius=10.3916, center=<5.6234 -0.6779 0>}
  {shapeType=Capsule, length=10.0312, radius=9.2812, center=<-1.2326 1.308 0>, rotation=<rot 90 0 0>}
 ]
 spine_05 [{shapeType=Capsule, length=9.678, radius=12.3571, center=<0.6969 1.91 0>}]
 thigh_l []
 thigh_r []
 upperarm_l []
 upperarm_r []
}
constraints [
 {
  bone1Name=spine_01
  bone2Name=pelvis
  swing1LimitDegrees=45
  swing1Motion=Free
  swing2LimitDegrees=45
  swing2Motion=Free
  twistLimitDegrees=45
  twistMotion=Free
 }
 {
  bone1Name=spine_02
  bone2Name=spine_01
  swing1LimitDegrees=45
  swing1Motion=Free
  swing2LimitDegrees=45
  swing2Motion=Free
  twistLimitDegrees=45
  twistMotion=Free
 }
 {
  bone1Name=spine_03
  bone2Name=spine_02
  swing1LimitDegrees=45
  swing1Motion=Free
  swing2LimitDegrees=45
  swing2Motion=Free
  twistLimitDegrees=45
  twistMotion=Free
 }
 {
  bone1Name=spine_04
  bone2Name=spine_03
  swing1LimitDegrees=45
  swing1Motion=Free
  swing2LimitDegrees=45
  swing2Motion=Free
  twistLimitDegrees=45
  twistMotion=Free
 }
 {
  bone1Name=spine_05
  bone2Name=spine_04
  swing1LimitDegrees=45
  swing1Motion=Free
  swing2LimitDegrees=45
  swing2Motion=Free
  twistLimitDegrees=45
  twistMotion=Free
 }
 {
  bone1Name=clavicle_l
  bone2Name=spine_05
  swing1LimitDegrees=45
  swing1Motion=Free
  swing2LimitDegrees=45
  swing2Motion=Free
  twistLimitDegrees=45
  twistMotion=Free
 }
 {
  bone1Name=upperarm_l
  bone2Name=clavicle_l
  swing1LimitDegrees=45
  swing1Motion=Free
  swing2LimitDegrees=45
  swing2Motion=Free
  twistLimitDegrees=45
  twistMotion=Free
 }
 {
  bone1Name=lowerarm_l
  bone2Name=upperarm_l
  swing1LimitDegrees=45
  swing1Motion=Free
  swing2LimitDegrees=45
  swing2Motion=Free
  twistLimitDegrees=45
  twistMotion=Free
 }
 {
  bone1Name=neck_02
  bone2Name=spine_05
  swing1LimitDegrees=45
  swing1Motion=Free
  swing2LimitDegrees=45
  swing2Motion=Free
  twistLimitDegrees=45
  twistMotion=Free
 }
 {
  bone1Name=head
  bone2Name=neck_02
  swing1LimitDegrees=45
  swing1Motion=Free
  swing2LimitDegrees=45
  swing2Motion=Free
  twistLimitDegrees=45
  twistMotion=Free
 }
 {
  bone1Name=thigh_l
  bone2Name=pelvis
  swing1LimitDegrees=45
  swing1Motion=Free
  swing2LimitDegrees=45
  swing2Motion=Free
  twistLimitDegrees=45
  twistMotion=Free
 }
 {
  bone1Name=foot_l
  bone2Name=calf_l
  swing1LimitDegrees=45
  swing1Motion=Free
  swing2LimitDegrees=45
  swing2Motion=Free
  twistLimitDegrees=45
  twistMotion=Free
 }
 {
  bone1Name=lowerarm_r
  bone2Name=upperarm_r
  swing1LimitDegrees=45
  swing1Motion=Free
  swing2LimitDegrees=45
  swing2Motion=Free
  twistLimitDegrees=45
  twistMotion=Free
 }
 {
  bone1Name=upperarm_r
  bone2Name=clavicle_r
  swing1LimitDegrees=45
  swing1Motion=Free
  swing2LimitDegrees=45
  swing2Motion=Free
  twistLimitDegrees=45
  twistMotion=Free
 }
 {
  bone1Name=clavicle_r
  bone2Name=spine_05
  swing1LimitDegrees=45
  swing1Motion=Free
  swing2LimitDegrees=45
  swing2Motion=Free
  twistLimitDegrees=45
  twistMotion=Free
 }
 {
  bone1Name=thigh_r
  bone2Name=pelvis
  swing1LimitDegrees=45
  swing1Motion=Free
  swing2LimitDegrees=45
  swing2Motion=Free
  twistLimitDegrees=45
  twistMotion=Free
 }
 {
  bone1Name=foot_r
  bone2Name=calf_r
  swing1LimitDegrees=45
  swing1Motion=Free
  swing2LimitDegrees=45
  swing2Motion=Free
  twistLimitDegrees=45
  twistMotion=Free
 }
 {
  bone1Name=calf_r
  bone2Name=thigh_r
  swing1LimitDegrees=45
  swing1Motion=Free
  swing2LimitDegrees=45
  swing2Motion=Free
  twistLimitDegrees=45
  twistMotion=Free
 }
 {
  bone1Name=calf_l
  bone2Name=thigh_l
  swing1LimitDegrees=45
  swing1Motion=Free
  swing2LimitDegrees=45
  swing2Motion=Free
  twistLimitDegrees=45
  twistMotion=Free
 }
 {
  bone1Name=dyn_skirt_1_b_c
  bone2Name=pelvis
  swing1LimitDegrees=45
  swing1Motion=Free
  swing2LimitDegrees=45
  swing2Motion=Free
  twistLimitDegrees=45
  twistMotion=Free
 }
 {
  bone1Name=dyn_skirt_2_b_c
  bone2Name=dyn_skirt_1_b_c
  swing1LimitDegrees=45
  swing1Motion=Free
  swing2LimitDegrees=45
  swing2Motion=Free
  twistLimitDegrees=45
  twistMotion=Free
 }
 {
  bone1Name=dyn_skirt_1_b_l
  bone2Name=pelvis
  swing1LimitDegrees=45
  swing1Motion=Free
  swing2LimitDegrees=45
  swing2Motion=Free
  twistLimitDegrees=45
  twistMotion=Free
 }
 {
  bone1Name=dyn_skirt_2_b_l
  bone2Name=dyn_skirt_1_b_l
  swing1LimitDegrees=45
  swing1Motion=Free
  swing2LimitDegrees=45
  swing2Motion=Free
  twistLimitDegrees=45
  twistMotion=Free
 }
 {
  bone1Name=dyn_skirt_1_b_r
  bone2Name=pelvis
  swing1LimitDegrees=45
  swing1Motion=Free
  swing2LimitDegrees=45
  swing2Motion=Free
  twistLimitDegrees=45
  twistMotion=Free
 }
 {
  bone1Name=dyn_skirt_2_b_r
  bone2Name=dyn_skirt_1_b_r
  swing1LimitDegrees=45
  swing1Motion=Free
  swing2LimitDegrees=45
  swing2Motion=Free
  twistLimitDegrees=45
  twistMotion=Free
 }
 {
  bone1Name=dyn_skirt_1_s_l
  bone2Name=pelvis
  swing1LimitDegrees=45
  swing1Motion=Free
  swing2LimitDegrees=45
  swing2Motion=Free
  twistLimitDegrees=45
  twistMotion=Free
 }
 {
  bone1Name=dyn_skirt_2_s_l
  bone2Name=dyn_skirt_1_s_l
  swing1LimitDegrees=45
  swing1Motion=Free
  swing2LimitDegrees=45
  swing2Motion=Free
  twistLimitDegrees=45
  twistMotion=Free
 }
 {
  bone1Name=dyn_skirt_1_s_r
  bone2Name=pelvis
  swing1LimitDegrees=45
  swing1Motion=Free
  swing2LimitDegrees=45
  swing2Motion=Free
  twistLimitDegrees=45
  twistMotion=Free
 }
 {
  bone1Name=dyn_skirt_2_s_r
  bone2Name=dyn_skirt_1_s_r
  swing1LimitDegrees=45
  swing1Motion=Free
  swing2LimitDegrees=45
  swing2Motion=Free
  twistLimitDegrees=45
  twistMotion=Free
 }
]
