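# Physics asset for the BlueMonday female medium character's FaceAcc mesh part, setting up physics for left and right ponytails.
# Each side chains five dyn_ponytail capsules from a small base body attached to spine_05, with limited swing and tighter twist at the first link.
# Capsules on the upper spine, clavicles and upper arms form a body proxy, linked by free constraints, presumably for the hair to collide against.
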PhysicsAsset /BRCosmetics/Characters/Player/Female/Medium/Bodies/F_MED_BlueMonday/Meshes/Parts/F_MED_BlueMonday_FaceAcc_Physics
bodies {
 base_ponytail_1_l [{shapeType=Capsule, length=0.505, radius=0.505, shapeName=base_ponytail_1_l_capsule, rotation=<rot 0 -90 0>}]
 base_ponytail_1_r [{shapeType=Capsule, length=0.505, radius=0.505, shapeName=base_ponytail_1_r_capsule, rotation=<rot 0 -90 0>}]
 clavicle_l [{
   shapeType=Capsule
   length=8.943
   radius=6.9185
   shapeName=clavicle_r_capsule
   center=<7.9121 2.0223 -0.7177>
   rotation=<rot 0.1592 -84.2339 -91.5779>
  }]
 clavicle_r [{
   shapeType=Capsule
   length=8.943
   radius=6.9185
   shapeName=clavicle_r_capsule
   center=<-7.9121 -2.0223 0.7177>
   rotation=<rot 0.1592 -84.2339 -91.5779>
  }]
 dyn_ponytail_1_l [{shapeType=Capsule, length=1.2275, radius=2.9172, shapeName=dyn_ponytail_1_l_capsule, center=<2.4871 0 0>, rotation=<rot -90 0 0>}]
 dyn_ponytail_1_r [{shapeType=Capsule, length=1.2275, radius=2.9172, shapeName=dyn_ponytail_1_l_capsule, center=<2.4871 0 0>, rotation=<rot -90 0 0>}]
 dyn_ponytail_2_l [{shapeType=Capsule, length=3.3923, radius=1.7152, shapeName=dyn_ponytail_2_l_capsule, center=<2.9781 -0.1833 0.3137>, rotation=<rot -90 0 0>}]
 dyn_ponytail_2_r [{shapeType=Capsule, length=3.3923, radius=1.7152, shapeName=dyn_ponytail_2_l_capsule, center=<2.9781 -0.1833 0.3137>, rotation=<rot -90 0 0>}]
 dyn_ponytail_3_l [{shapeType=Capsule, length=3.7652, radius=1.3423, shapeName=dyn_ponytail_2_l_capsule, center=<2.9781 0 0.2461>, rotation=<rot -90 0 0>}]
 dyn_ponytail_3_r [{shapeType=Capsule, length=3.7652, radius=1.3423, shapeName=dyn_ponytail_2_l_capsule, center=<2.9781 0 0.2461>, rotation=<rot -90 0 0>}]
 dyn_ponytail_4_l [{
   shapeType=Capsule
   length=3.8701
   radius=1.2374
   shapeName=dyn_ponytail_2_l_capsule
   center=<2.9781 0 0.2409>
   rotation=<rot -87.9236 0.0005 -0.0005>
  }]
 dyn_ponytail_4_r [{
   shapeType=Capsule
   length=3.8701
   radius=1.2374
   shapeName=dyn_ponytail_2_l_capsule
   center=<2.9781 0 0.2409>
   rotation=<rot -87.9236 -0.0005 0.0005>
  }]
 dyn_ponytail_5_l [{
   shapeType=Capsule
   length=3.7732
   radius=1.3343
   shapeName=dyn_ponytail_2_l_capsule
   center=<3.4404 0 0.3112>
   rotation=<rot -89.7812 179.9927 -179.9927>
  }]
 dyn_ponytail_5_r [{
   shapeType=Capsule
   length=3.7732
   radius=1.3343
   shapeName=dyn_ponytail_2_l_capsule
   center=<3.4404 0 0.3112>
   rotation=<rot -89.7812 179.9927 -179.9927>
  }]
 spine_03 [{shapeType=Capsule, length=11.6552, radius=7.767, shapeName=spine_03_capsule, center=<5.2322 0.6238 0>, rotation=<rot 0 -90 0>}]
 spine_04 [{shapeType=Capsule, length=12.6054, radius=7.5158, shapeName=spine_04_capsule, center=<3.8288 -0.4637 0>, rotation=<rot 0 -90 0>}]
 spine_05 [{shapeType=Capsule, length=12.8296, radius=7.8465, shapeName=spine_05_capsule, center=<3.8394 0.6628 0>, rotation=<rot 0 -90 0>}]
 upperarm_l [{shapeType=Capsule, length=15.9972, radius=6.6474, shapeName=upperarm_l_capsule, center=<10.0696 0.7159 0>, rotation=<rot 90 0 0>}]
 upperarm_r [{shapeType=Capsule, length=15.9972, radius=6.6474, shapeName=upperarm_r_capsule, center=<-10.0696 -0.7159 0>, rotation=<rot 90 180 0>}]
}
constraints [
 {
  bone1Name=spine_05
  bone2Name=spine_04
  swing1LimitDegrees=45
  swing1Motion=Free
  swing2LimitDegrees=45
  swing2Motion=Free
  twistLimitDegrees=45
  twistMotion=Free
 }
 {
  bone1Name=spine_04
  bone2Name=spine_03
  swing1LimitDegrees=45
  swing1Motion=Free
  swing2LimitDegrees=45
  swing2Motion=Free
  twistLimitDegrees=45
  twistMotion=Free
 }
 {
  bone1Name=base_ponytail_1_l
  bone2Name=spine_05
  swing1LimitDegrees=45
  swing1Motion=Free
  swing2LimitDegrees=45
  swing2Motion=Free
  twistLimitDegrees=45
  twistMotion=Free
 }
 {bone1Name=dyn_ponytail_1_l, bone2Name=base_ponytail_1_l, swing1LimitDegrees=20, swing2LimitDegrees=20, twistLimitDegrees=2}
 {bone1Name=dyn_ponytail_2_l, bone2Name=dyn_ponytail_1_l, swing1LimitDegrees=10, swing2LimitDegrees=30, twistLimitDegrees=10}
 {bone1Name=dyn_ponytail_3_l, bone2Name=dyn_ponytail_2_l, swing1LimitDegrees=10, swing2LimitDegrees=30, twistLimitDegrees=10}
 {bone1Name=dyn_ponytail_4_l, bone2Name=dyn_ponytail_3_l, swing1LimitDegrees=10, swing2LimitDegrees=25, twistLimitDegrees=10}
 {bone1Name=dyn_ponytail_5_l, bone2Name=dyn_ponytail_4_l, swing1LimitDegrees=10, swing2LimitDegrees=20, twistLimitDegrees=10}
 {
  bone1Name=base_ponytail_1_r
  bone2Name=spine_05
  swing1LimitDegrees=45
  swing1Motion=Free
  swing2LimitDegrees=45
  swing2Motion=Free
  twistLimitDegrees=45
  twistMotion=Free
 }
 {bone1Name=dyn_ponytail_1_r, bone2Name=base_ponytail_1_r, swing1LimitDegrees=20, swing2LimitDegrees=20, twistLimitDegrees=2}
 {bone1Name=dyn_ponytail_2_r, bone2Name=dyn_ponytail_1_r, swing1LimitDegrees=10, swing2LimitDegrees=30, twistLimitDegrees=10}
 {bone1Name=dyn_ponytail_3_r, bone2Name=dyn_ponytail_2_r, swing1LimitDegrees=10, swing2LimitDegrees=30, twistLimitDegrees=10}
 {bone1Name=dyn_ponytail_4_r, bone2Name=dyn_ponytail_3_r, swing1LimitDegrees=10, swing2LimitDegrees=25, twistLimitDegrees=10}
 {bone1Name=dyn_ponytail_5_r, bone2Name=dyn_ponytail_4_r, swing1LimitDegrees=10, swing2LimitDegrees=20, twistLimitDegrees=10}
 {
  bone1Name=clavicle_l
  bone2Name=spine_05
  swing1LimitDegrees=45
  swing1Motion=Free
  swing2LimitDegrees=45
  swing2Motion=Free
  twistLimitDegrees=45
  twistMotion=Free
 }
 {
  bone1Name=clavicle_r
  bone2Name=spine_05
  swing1LimitDegrees=45
  swing1Motion=Free
  swing2LimitDegrees=45
  swing2Motion=Free
  twistLimitDegrees=45
  twistMotion=Free
 }
 {
  bone1Name=upperarm_l
  bone2Name=clavicle_l
  swing1LimitDegrees=45
  swing1Motion=Free
  swing2LimitDegrees=45
  swing2Motion=Free
  twistLimitDegrees=45
  twistMotion=Free
 }
 {
  bone1Name=upperarm_r
  bone2Name=clavicle_r
  swing1LimitDegrees=45
  swing1Motion=Free
  swing2LimitDegrees=45
  swing2Motion=Free
  twistLimitDegrees=45
  twistMotion=Free
 }
]
