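# Post-process style hair physics AnimBP for the F_MED_BlueMonday FaceAcc cosmetic part, driving ponytail, bangs and strands.
# Copies the parent mesh pose through a Control Rig, then applies additive ponytail bone offsets.
# A bool blend selects that pose or its RigidBody simulation, followed by chained spring-based AnimDynamics for bangs and left strands.
# Events set gravity, movement-dependent RigidBody alpha, backbling state and front-end or default sim-space settings.
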AnimBlueprint /BRCosmetics/Characters/Player/Female/Medium/Bodies/F_MED_BlueMonday/Meshes/Parts/F_MED_BlueMonday_FaceAcc_AnimBP
target_skeleton=&/F_MED_BlueMonday_FaceAcc_Skeleton
graphs [
 AnimGraph {nodes [
   n1 = RigidBody!(ComponentPose=Local To Component(LocalPose=Use cached pose 'preSimTweaks'()), Alpha, OverrideWorldGravity:Vector="0,0,0") -> (Pose) {
    bOverrideWorldGravity=true
    overridePhysicsAsset=&/F_MED_BlueMonday_FaceAcc_Physics
    simulationSpace=BaseBoneSpace
    alphaBoolBlend {blendOption=Linear}
    alphaScaleBiasClamp {bInterpResult=true, interpSpeedDecreasing=3, interpSpeedIncreasing=3}
    baseBoneRef {boneName=spine_05}
    componentAppliedLinearAccClamp=<3000 3000 3000>
    simSpaceSettings {
     maxAngularAcceleration=1
     maxAngularVelocity=3
     maxLinearAcceleration=1000
     maxLinearVelocity=500
     velocityScaleZ=0.8
     worldAlpha=1
     externalLinearDragV=<0.2 0.2 0.2>
    }
   }
   n2 = Component To Local(ComponentPose=@n1)
   n3 = Use cached pose 'preSimTweaks'()
   n4 = Anim Dynamics - Bone: dyn_bangs_1_l!(
    ComponentPose=Local To Component(LocalPose=Blend Poses by bool(BlendPose_0=@n3, BlendPose_1=@n2))
    Alpha=1
    GravityOverride:Vector="0,0,-980"
    SimSpaceSettings:Anim Phys Sim Space Settings Structure=(SimSpaceAngularAlpha=0,MaxAngularVelocity=10000,MaxAngularAcceleration=10000,ExternalAngularVelocity=(X=0,Y=0,Z=0))
   ) -> (Pose) {
    angularDampingOverride=1
    angularSpringConstant=1000
    bAngularSpring=true
    bOverrideAngularDamping=true
    bUseGravityOverride=true
    gravityScale=0.35
    simulationSpace=RootRelative
    alphaBoolBlend {blendOption=Linear}
    boundBone {boneName=dyn_bangs_1_l}
    componentAppliedLinearAccClamp=<3000 3000 3000>
    componentLinearAccScale=<0.4 0.4 0.4>
    componentLinearVelScale=<0.4 0.4 0.4>
    gravityOverride {z=-980}
    physicsBodyDefinitions [{
      boundBone {boneName=dyn_bangs_1_l}
      boxExtents=<20 20 20>
      constraintSetup {angularLimitsMax=<0 1 0>, angularLimitsMin=<0 -15 0>, angularTarget=<1 0 0>}
      localJointOffset=<5 0 0>
     }]
   }
   n5 = Anim Dynamics - Bone: dyn_bangs_1_c!(
    ComponentPose=@n4
    Alpha=1
    GravityOverride:Vector="0,0,-980"
    SimSpaceSettings:Anim Phys Sim Space Settings Structure=(SimSpaceAngularAlpha=0,MaxAngularVelocity=10000,MaxAngularAcceleration=10000,ExternalAngularVelocity=(X=0,Y=0,Z=0))
   ) -> (Pose) {
    angularDampingOverride=1
    angularSpringConstant=1000
    bAngularSpring=true
    bOverrideAngularDamping=true
    bUseGravityOverride=true
    gravityScale=0.35
    simulationSpace=RootRelative
    alphaBoolBlend {blendOption=Linear}
    boundBone {boneName=dyn_bangs_1_c}
    componentAppliedLinearAccClamp=<3000 3000 3000>
    componentLinearAccScale=<0.4 0.4 0.4>
    componentLinearVelScale=<0.4 0.4 0.4>
    gravityOverride {z=-980}
    physicsBodyDefinitions [{
      boundBone {boneName=dyn_bangs_1_c}
      boxExtents=<20 20 20>
      constraintSetup {angularLimitsMax=<0 5 7>, angularLimitsMin=<0 -15 -7>, angularTarget=<1 0 0>}
      localJointOffset=<5 0 0>
     }]
   }
   n6 = Anim Dynamics - Bone: dyn_bangs_1_r!(
    ComponentPose=@n5
    Alpha=1
    GravityOverride:Vector="0,0,-980"
    SimSpaceSettings:Anim Phys Sim Space Settings Structure=(SimSpaceAngularAlpha=0,MaxAngularVelocity=10000,MaxAngularAcceleration=10000,ExternalAngularVelocity=(X=0,Y=0,Z=0))
   ) -> (Pose) {
    angularDampingOverride=1
    angularSpringConstant=1000
    bAngularSpring=true
    bOverrideAngularDamping=true
    bUseGravityOverride=true
    gravityScale=0.35
    simulationSpace=RootRelative
    alphaBoolBlend {blendOption=Linear}
    boundBone {boneName=dyn_bangs_1_r}
    componentAppliedLinearAccClamp=<3000 3000 3000>
    componentLinearAccScale=<0.4 0.4 0.4>
    componentLinearVelScale=<0.4 0.4 0.4>
    gravityOverride {z=-980}
    physicsBodyDefinitions [{
      boundBone {boneName=dyn_bangs_1_r}
      boxExtents=<20 20 20>
      constraintSetup {angularLimitsMax=<0 5 2>, angularLimitsMin=<0 -15 -7>, angularTarget=<1 0 0>}
      localJointOffset=<5 0 0>
     }]
   }
   n7 = Anim Dynamics - Bone: dyn_hairStrand_1_l!(
    ComponentPose=@n6
    Alpha=1
    GravityOverride:Vector="0,0,-980"
    SimSpaceSettings:Anim Phys Sim Space Settings Structure=(SimSpaceAngularAlpha=0,MaxAngularVelocity=10000,MaxAngularAcceleration=10000,ExternalAngularVelocity=(X=0,Y=0,Z=0))
   ) -> (Pose) {
    angularDampingOverride=1
    angularSpringConstant=1000
    bAngularSpring=true
    bOverrideAngularDamping=true
    bUseGravityOverride=true
    gravityScale=0.35
    simulationSpace=RootRelative
    alphaBoolBlend {blendOption=Linear}
    boundBone {boneName=dyn_hairStrand_1_l}
    componentAppliedLinearAccClamp=<3000 3000 3000>
    componentLinearAccScale=<0.4 0.4 0.4>
    componentLinearVelScale=<0.4 0.4 0.4>
    gravityOverride {z=-980}
    physicsBodyDefinitions [{
      boundBone {boneName=dyn_hairStrand_1_l}
      boxExtents=<20 20 20>
      constraintSetup {angularLimitsMax=<0 10 15>, angularLimitsMin=<0 -10 -15>, angularTarget=<1 0 0>}
      localJointOffset=<10 0 0>
     }]
   }
   n8 = Anim Dynamics - Bone: dyn_hairStrand_2_l!(
    ComponentPose=@n7
    Alpha=1
    GravityOverride:Vector="0,0,-980"
    SimSpaceSettings:Anim Phys Sim Space Settings Structure=(SimSpaceAngularAlpha=0,MaxAngularVelocity=10000,MaxAngularAcceleration=10000,ExternalAngularVelocity=(X=0,Y=0,Z=0))
   ) -> (Pose) {
    angularDampingOverride=1
    angularSpringConstant=1000
    bAngularSpring=true
    bOverrideAngularDamping=true
    bUseGravityOverride=true
    gravityScale=0.35
    simulationSpace=RootRelative
    alphaBoolBlend {blendOption=Linear}
    boundBone {boneName=dyn_hairStrand_2_l}
    componentAppliedLinearAccClamp=<3000 3000 3000>
    componentLinearAccScale=<0.4 0.4 0.4>
    componentLinearVelScale=<0.4 0.4 0.4>
    gravityOverride {z=-980}
    physicsBodyDefinitions [{
      boundBone {boneName=dyn_hairStrand_2_l}
      boxExtents=<20 20 20>
      constraintSetup {angularLimitsMax=<0 10 15>, angularLimitsMin=<0 -10 -15>, angularTarget=<1 0 0>}
      localJointOffset=<5 0 0>
     }]
   }
   n9 = Anim Dynamics - Bone: dyn_hairStrand_3_l!(
    ComponentPose=@n8
    Alpha=1
    GravityOverride:Vector="0,0,-980"
    SimSpaceSettings:Anim Phys Sim Space Settings Structure=(SimSpaceAngularAlpha=0,MaxAngularVelocity=10000,MaxAngularAcceleration=10000,ExternalAngularVelocity=(X=0,Y=0,Z=0))
   ) -> (Pose) {
    angularDampingOverride=1
    angularSpringConstant=1000
    bAngularSpring=true
    bOverrideAngularDamping=true
    bUseGravityOverride=true
    gravityScale=0.35
    simulationSpace=RootRelative
    alphaBoolBlend {blendOption=Linear}
    boundBone {boneName=dyn_hairStrand_3_l}
    componentAppliedLinearAccClamp=<3000 3000 3000>
    componentLinearAccScale=<0.4 0.4 0.4>
    componentLinearVelScale=<0.4 0.4 0.4>
    gravityOverride {z=-980}
    physicsBodyDefinitions [{
      boundBone {boneName=dyn_hairStrand_3_l}
      boxExtents=<20 20 20>
      constraintSetup {angularLimitsMax=<0 10 15>, angularLimitsMin=<0 -10 -15>, angularTarget=<1 0 0>}
      localJointOffset=<5 0 0>
     }]
   }
   n10 = Anim Dynamics - Bone: dyn_hairStrand_4_A_l!(
    ComponentPose=@n9
    Alpha=1
    GravityOverride:Vector="0,0,-980"
    SimSpaceSettings:Anim Phys Sim Space Settings Structure=(SimSpaceAngularAlpha=0,MaxAngularVelocity=10000,MaxAngularAcceleration=10000,ExternalAngularVelocity=(X=0,Y=0,Z=0))
   ) -> (Pose) {
    angularDampingOverride=1
    angularSpringConstant=1000
    bAngularSpring=true
    bOverrideAngularDamping=true
    bUseGravityOverride=true
    gravityScale=0.35
    simulationSpace=RootRelative
    alphaBoolBlend {blendOption=Linear}
    boundBone {boneName=dyn_hairStrand_4_A_l}
    componentAppliedLinearAccClamp=<3000 3000 3000>
    componentLinearAccScale=<0.4 0.4 0.4>
    componentLinearVelScale=<0.4 0.4 0.4>
    gravityOverride {z=-980}
    physicsBodyDefinitions [{
      boundBone {boneName=dyn_hairStrand_4_A_l}
      boxExtents=<30 30 30>
      constraintSetup {angularLimitsMax=<0 10 10>, angularLimitsMin=<0 -10 -10>, angularTarget=<1 0 0>}
      localJointOffset=<10 0 0>
     }]
   }
   n11 = Anim Dynamics - Bone: dyn_hairStrand_5_l!(
    ComponentPose=@n10
    Alpha=1
    GravityOverride:Vector="0,0,-980"
    SimSpaceSettings:Anim Phys Sim Space Settings Structure=(SimSpaceAngularAlpha=0,MaxAngularVelocity=10000,MaxAngularAcceleration=10000,ExternalAngularVelocity=(X=0,Y=0,Z=0))
   ) -> (Pose) {
    angularDampingOverride=1
    angularSpringConstant=1000
    bAngularSpring=true
    bOverrideAngularDamping=true
    bUseGravityOverride=true
    gravityScale=0.35
    simulationSpace=RootRelative
    alphaBoolBlend {blendOption=Linear}
    boundBone {boneName=dyn_hairStrand_5_l}
    componentAppliedLinearAccClamp=<3000 3000 3000>
    componentLinearAccScale=<0.4 0.4 0.4>
    componentLinearVelScale=<0.4 0.4 0.4>
    gravityOverride {z=-980}
    physicsBodyDefinitions [{
      boundBone {boneName=dyn_hairStrand_5_l}
      boxExtents=<20 20 20>
      constraintSetup {angularLimitsMax=<0 10 15>, angularLimitsMin=<0 -10 -15>, angularTarget=<1 0 0>}
      localJointOffset=<5 0 0>
     }]
   }
   n12 = Anim Dynamics - Bone: base_hairStrand_6_l!(
    ComponentPose=@n11
    Alpha=1
    GravityOverride:Vector="0,0,-980"
    SimSpaceSettings:Anim Phys Sim Space Settings Structure=(SimSpaceAngularAlpha=0,MaxAngularVelocity=10000,MaxAngularAcceleration=10000,ExternalAngularVelocity=(X=0,Y=0,Z=0))
   ) -> (Pose) {
    angularDampingOverride=1
    angularSpringConstant=1000
    bAngularSpring=true
    bOverrideAngularDamping=true
    bUseGravityOverride=true
    gravityScale=0.35
    simulationSpace=RootRelative
    alphaBoolBlend {blendOption=Linear}
    boundBone {boneName=base_hairStrand_6_l}
    componentAppliedLinearAccClamp=<3000 3000 3000>
    componentLinearAccScale=<0.4 0.4 0.4>
    componentLinearVelScale=<0.4 0.4 0.4>
    gravityOverride {z=-980}
    physicsBodyDefinitions [{
      boundBone {boneName=base_hairStrand_6_l}
      boxExtents=<20 20 20>
      constraintSetup {angularLimitsMax=<0 5 10>, angularLimitsMin=<0 -5 -15>, angularTarget=<1 0 0>}
      localJointOffset=<5 0 0>
     }]
   }
   Output Pose(Result=Component To Local(ComponentPose=@n12)) {layerGroup=None}
   n13 = Control Rig(Source=Copy Pose From Mesh() {bUseAttachedParent=true}) {controlRigAssetReference {blueprintRigClass=&/F_MED_BlueMonday_FaceAcc_CtrlRig.F_MED_BlueMonday_FaceAcc_CtrlRig_C}}
   Save cached pose 'controlRigCP'(Pose=@n13) {cacheName=controlRigCP}
   n14 = "Transform (Modify) Bone - Bone: base_ponytail_1_r"!(
    ComponentPose=Local To Component(LocalPose=Use cached pose 'controlRigCP'())
    bAlphaBoolEnabled:bool=True
    Translation="0,0,0"
    Rotation="0,25,0"
   ) -> (Pose) {
    alphaInputType=Bool
    rotationMode=BMM_Additive
    rotationSpace=BCS_BoneSpace
    translationMode=BMM_Additive
    translationSpace=BCS_BoneSpace
    alphaBoolBlend {blendInTime=0.5, blendOption=Cubic, blendOutTime=0.5}
    boneToModify {boneName=base_ponytail_1_r}
    rotation {yaw=25}
   }
   n15 = "Transform (Modify) Bone - Bone: base_ponytail_1_r"(ComponentPose=@n14, bAlphaBoolEnabled=True, Translation="0,0,0", Rotation="0,-15,0") {
    alphaInputType=Bool
    rotationMode=BMM_Additive
    rotationSpace=BCS_BoneSpace
    translationMode=BMM_Additive
    translationSpace=BCS_BoneSpace
    boneToModify {boneName=base_ponytail_1_r}
    rotation {yaw=-15}
   }
   Save cached pose 'preSimTweaks'(Pose=Component To Local(ComponentPose=@n15)) {cacheName=preSimTweaks}
  ]}
 EventGraph {nodes [
   Event BlueprintUpdateAnimation()
   > n1 = Sequence()
   > n2 = "Gravity Override State Function (Parameter Asset) (Nativized)"(
    CharacterPartAnimInstance=Self-Reference()
    GameTimeInSeconds=GetGameTimeInSeconds()
    Parameters=/BRCosmetics/Characters/Player/Female/Medium/Bodies/F_MED_BlueMonday/Meshes/F_MED_BlueMonday_Parameters
   )
   > Set GravityOverride(GravityOverride=@n2)
   n3 = Branch(execute=@n1.then_1, Condition=Get bIsPlayingMeleeAnim())
   > Set alpha hairRigidBody(alpha hairRigidBody:double=0.6) -> (Output_Get:double)
   n4 = Branch(execute=@n3.else, Condition=Get bIsDBNOCarrying() -> (bIsDBNOCarrying:bool))
   > Set alpha hairRigidBody(alpha hairRigidBody=0.3)
   n5 = Get bIsTacticalSprinting()
   n6 = Get bIsJumping()
   n7 = Get bIsFalling()
   n8 = Branch(execute=@n4.else, Condition=OR Boolean(A=AND Boolean(A=@n6, B=@n5), B=AND Boolean(A=@n5, B=@n7)))
   > Set alpha hairRigidBody(alpha hairRigidBody=0.8)
   n9 = Branch(execute=@n8.else, Condition=Get bIsTacticalSprinting())
   > Set alpha hairRigidBody(alpha hairRigidBody=0.7)
   Set alpha hairRigidBody(execute=@n9.else, alpha hairRigidBody=1)
   Event OnCharacterCustomizationLoaded()
   n10 = CheckForBackbling()
   > Set isWearingBlackbling(isWearingBlackbling:bool=OR Boolean!(A=@n10, B=@n10.bIsBackblingACape, C:bool=@n10.bIsBackblingLong) -> (ReturnValue)) -> (Output_Get:bool)
   > n11 = Branch(Condition=Get bIsInFrontEnd())
   > Set SimSpaceSettings output(SimSpaceSettings output:Anim Phys Sim Space Settings Structure=Get SimSpaceSettings frontEnd() -> (SimSpaceSettings frontEnd:Anim Phys Sim Space Settings Structure)) -> (Output_Get:Anim Phys Sim Space Settings Structure)
   Set SimSpaceSettings output(
    execute=@n11.else
    SimSpaceSettings output=Get SimSpaceSettings default() -> (SimSpaceSettings default:Anim Phys Sim Space Settings Structure)
   )
  ]}
]
variables [
 alpha hairRigidBody {type=number}
 GravityOverride {type=Vector}
 isWearingBlackbling {type=boolean}
 SimSpaceSettings default {
  type=AnimPhysSimSpaceSettings
  default {maxAngularAcceleration=10000, maxAngularVelocity=10000, simSpaceAngularAlpha=0, externalAngularVelocity=<0 0 0>}
 }
 SimSpaceSettings frontEnd {
  type=AnimPhysSimSpaceSettings
  default {maxAngularAcceleration=56, maxAngularVelocity=8, simSpaceAngularAlpha=0.75, externalAngularVelocity=<0 0 0>}
 }
 SimSpaceSettings output {
  type=AnimPhysSimSpaceSettings
  default {maxAngularAcceleration=10000, maxAngularVelocity=10000, simSpaceAngularAlpha=0, externalAngularVelocity=<0 0 0>}
 }
]
